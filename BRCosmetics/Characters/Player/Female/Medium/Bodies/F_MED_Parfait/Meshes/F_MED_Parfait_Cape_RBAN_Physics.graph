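# Physics asset for the Parfait female-medium player cosmetic's cape, named for RigidBody anim node (RBAN) simulation.
# Four cape chains (outer and mid, left and right) use box and capsule bodies with tight angular limits below their roots.
# Angular-free constraints cross-link outer and mid chains at their two lowest segments on each side.
# Torso, clavicle and leg shapes, plus two back-skirt chains, share the asset, with angular-free joints; arm bodies have no shapes.
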PhysicsAsset /BRCosmetics/Characters/Player/Female/Medium/Bodies/F_MED_Parfait/Meshes/F_MED_Parfait_Cape_RBAN_Physics
bodies {
 calf_l [{shapeType=Capsule, length=34.0625, radius=10.2863, center=<-27.2533 -2.5217 -0.0003>, rotation=<rot 90 180 178.4284>}]
 calf_r [{shapeType=Capsule, length=34.0625, radius=10.2863, center=<26.9506 2.5134 0.0003>, rotation=<rot 90 -179.8508 178.5776>}]
 cape_01_l [{shapeType=Capsule, length=1, radius=1, shapeName=cape_01_l_capsule}]
 cape_01_r [{shapeType=Capsule, length=1, radius=1, shapeName=cape_01_r_capsule}]
 cape_02_l [{shapeType=Capsule, length=1, radius=1, shapeName=cape_02_l_capsule}]
 cape_02_r [{shapeType=Capsule, length=1, radius=1, shapeName=cape_02_r_capsule}]
 cape_03_l [{shapeType=Box, extentX=22.5625, extentY=2.6875, extentZ=29.0244, center=<11.5978 0.0001 0>}]
 cape_03_r [{shapeType=Box, extentX=22.5625, extentY=2.6875, extentZ=29.0244, center=<11.5978 0.0001 0>}]
 cape_04_l [{shapeType=Capsule, length=21.75, radius=7, shapeName=cape_05_l_capsule, center=<5 0 0>}]
 cape_04_r [{shapeType=Capsule, length=21.75, radius=7, shapeName=cape_05_r_capsule, center=<5 0 0>}]
 cape_05_l [{shapeType=Capsule, length=21.75, radius=7, shapeName=cape_05_l_capsule, center=<5 0 0>}]
 cape_05_r [{shapeType=Capsule, length=21.75, radius=7, shapeName=cape_05_r_capsule, center=<5 0 0>}]
 cape_06_l [{shapeType=Capsule, length=21.75, radius=7, shapeName=cape_06_l_capsule, center=<5 0 0>}]
 cape_06_r [{shapeType=Capsule, length=21.75, radius=7, shapeName=cape_06_r_capsule, center=<5 0 0>}]
 cape_07_l [{shapeType=Capsule, length=21.75, radius=7, shapeName=cape_07_l_capsule, center=<5 0 0>}]
 cape_07_r [{shapeType=Capsule, length=21.75, radius=7, shapeName=cape_07_r_capsule, center=<5 0 0>}]
 cape_mid_01_l [{shapeType=Capsule, length=1, radius=1, shapeName=cape_mid_01_l_capsule}]
 cape_mid_01_r [{shapeType=Capsule, length=1, radius=1, shapeName=cape_mid_01_r_capsule}]
 cape_mid_02_l [{shapeType=Box, extentX=22.5625, extentY=2.6875, extentZ=29.0244, center=<11.5978 0.0001 0>}]
 cape_mid_02_r [{shapeType=Box, extentX=22.5625, extentY=2.6875, extentZ=29.0244, center=<11.5978 0.0001 0>}]
 cape_mid_03_l [{shapeType=Capsule, length=21.75, radius=7, shapeName=cape_05_r_capsule, center=<5 0 0>}]
 cape_mid_03_r [{shapeType=Capsule, length=21.75, radius=7, shapeName=cape_05_r_capsule, center=<5 0 0>}]
 cape_mid_04_l [{shapeType=Capsule, length=21.75, radius=7, shapeName=cape_05_r_capsule, center=<5 0 0>}]
 cape_mid_04_r [{shapeType=Capsule, length=21.75, radius=7, shapeName=cape_05_r_capsule, center=<5 0 0>}]
 cape_mid_05_l [{shapeType=Capsule, length=21.75, radius=7, shapeName=cape_06_r_capsule, center=<5 0 0>}]
 cape_mid_05_r [{shapeType=Capsule, length=21.75, radius=7, shapeName=cape_06_r_capsule, center=<5 0 0>}]
 cape_mid_06_l [{shapeType=Capsule, length=21.75, radius=7, shapeName=cape_07_r_capsule, center=<5 0 0>}]
 cape_mid_06_r [{shapeType=Capsule, length=21.75, radius=7, shapeName=cape_07_r_capsule, center=<5 0 0>}]
 clavicle_cape_l [{shapeType=Capsule, length=1, radius=1, shapeName=clavicle_cape_l_capsule}]
 clavicle_cape_r [{shapeType=Capsule, length=1, radius=1, shapeName=clavicle_cape_r_capsule}]
 clavicle_l [{shapeType=Capsule, length=7.6964, radius=6.3853, center=<4.0147 -0.8354 -1.7419>, rotation=<rot -77.5566 177.3915 7.7314>}]
 clavicle_r [{shapeType=Capsule, length=7.6964, radius=8.2957, center=<-4.0147 0.8354 1.7419>, rotation=<rot -77.5565 177.3915 -172.2685>}]
 lowerarm_l []
 lowerarm_r []
 pelvis [{shapeType=Capsule, length=8.9028, radius=11.1596, center=<-1.9544 -2.2029 -0.1981>}]
 skirt_back_a_03 [{shapeType=Capsule, length=26.25, radius=3.75, shapeName=skirt_back_a_03_capsule, center=<3 2.4987 0>}]
 skirt_back_a_04 [{shapeType=Capsule, length=26.25, radius=3.75, shapeName=skirt_back_a_04_capsule, center=<3 2.4987 0>}]
 skirt_back_a_05 [{shapeType=Capsule, length=26.25, radius=3.75, shapeName=skirt_back_a_05_capsule, center=<3 2.4987 0>}]
 skirt_back_b_03 [{shapeType=Capsule, length=26.25, radius=3.75, shapeName=skirt_back_b_03_capsule, center=<3 2.4987 0>}]
 skirt_back_b_04 [{shapeType=Capsule, length=26.25, radius=3.75, shapeName=skirt_back_b_04_capsule, center=<3 2.4987 0>}]
 skirt_back_b_05 [{shapeType=Capsule, length=26.25, radius=3.75, shapeName=skirt_back_b_05_capsule, center=<3 2.4987 0>}]
 spine_01 [{shapeType=Capsule, length=13.2671, radius=10.4854, center=<-2.8183 -1.0873 0>}]
 spine_02 [
  {shapeType=Capsule, length=12.5, radius=10.9854, center=<1.179 2.2974 0>}
  {shapeType=Capsule, length=20.6875, radius=9.3125, center=<-1.8317 2.606 0>, rotation=<rot 90 0 0>}
 ]
 spine_03 [
  {shapeType=Capsule, length=9.25, radius=14.2354, center=<-0.0747 2.3725 0>}
  {shapeType=Capsule, length=20.7188, radius=9.2812, center=<-1.2326 2.6059 0>, rotation=<rot 90 0 0>}
 ]
 spine_04 [
  {shapeType=Capsule, length=16.5796, radius=9.8916, center=<6.9796 -1.0098 0>}
  {shapeType=Capsule, length=10.0312, radius=9.2812, center=<-1.2326 1.308 0>, rotation=<rot 90 0 0>}
 ]
 spine_05 [{shapeType=Capsule, length=11.7359, radius=9.3291, center=<1.5198 -0.8941 0>}]
 thigh_l [{shapeType=Capsule, length=24.5, radius=11.25, shapeName=thigh_l_capsule, center=<-21.2572 0 0>, rotation=<rot 90 0 0>}]
 thigh_r [{shapeType=Capsule, length=28.5, radius=12, shapeName=thigh_r_capsule, center=<21.2457 0 0>, rotation=<rot 90 0 0>}]
 upperarm_l []
 upperarm_r []
}
constraints [
 {
  bone1Name=cape_01_l
  bone2Name=clavicle_cape_l
  swing1LimitDegrees=45
  swing1Motion=Free
  swing2LimitDegrees=45
  swing2Motion=Free
  twistLimitDegrees=45
  twistMotion=Free
 }
 {
  bone1Name=cape_02_l
  bone2Name=cape_01_l
  swing1LimitDegrees=45
  swing1Motion=Free
  swing2LimitDegrees=45
  swing2Motion=Free
  twistLimitDegrees=45
  twistMotion=Free
 }
 {bone1Name=cape_03_l, bone2Name=cape_02_l, swing1LimitDegrees=15, swing2LimitDegrees=5, twistLimitDegrees=5}
 {bone1Name=cape_04_l, bone2Name=cape_03_l, swing1LimitDegrees=15, swing2LimitDegrees=5, twistLimitDegrees=5}
 {bone1Name=cape_05_l, bone2Name=cape_04_l, swing1LimitDegrees=20, swing2LimitDegrees=2, twistLimitDegrees=5}
 {bone1Name=cape_06_l, bone2Name=cape_05_l, swing1LimitDegrees=20, swing2LimitDegrees=2, twistLimitDegrees=5}
 {bone1Name=cape_07_l, bone2Name=cape_06_l, swing1LimitDegrees=20, swing2LimitDegrees=2, twistLimitDegrees=5}
 {
  bone1Name=cape_01_r
  bone2Name=clavicle_cape_r
  swing1LimitDegrees=45
  swing1Motion=Free
  swing2LimitDegrees=45
  swing2Motion=Free
  twistLimitDegrees=45
  twistMotion=Free
 }
 {
  bone1Name=cape_02_r
  bone2Name=cape_01_r
  swing1LimitDegrees=45
  swing1Motion=Free
  swing2LimitDegrees=45
  swing2Motion=Free
  twistLimitDegrees=45
  twistMotion=Free
 }
 {bone1Name=cape_03_r, bone2Name=cape_02_r, swing1LimitDegrees=15, swing2LimitDegrees=5, twistLimitDegrees=5}
 {bone1Name=cape_04_r, bone2Name=cape_03_r, swing1LimitDegrees=15, swing2LimitDegrees=5, twistLimitDegrees=5}
 {bone1Name=cape_05_r, bone2Name=cape_04_r, swing1LimitDegrees=20, swing2LimitDegrees=2, twistLimitDegrees=5}
 {bone1Name=cape_06_r, bone2Name=cape_05_r, swing1LimitDegrees=20, swing2LimitDegrees=2, twistLimitDegrees=5}
 {bone1Name=cape_07_r, bone2Name=cape_06_r, swing1LimitDegrees=20, swing2LimitDegrees=2, twistLimitDegrees=5}
 {
  bone1Name=clavicle_cape_l
  bone2Name=clavicle_l
  swing1LimitDegrees=45
  swing1Motion=Free
  swing2LimitDegrees=45
  swing2Motion=Free
  twistLimitDegrees=45
  twistMotion=Free
 }
 {
  bone1Name=clavicle_cape_r
  bone2Name=clavicle_r
  swing1LimitDegrees=45
  swing1Motion=Free
  swing2LimitDegrees=45
  swing2Motion=Free
  twistLimitDegrees=45
  twistMotion=Free
 }
 {
  bone1Name=spine_01
  bone2Name=pelvis
  swing1LimitDegrees=45
  swing1Motion=Free
  swing2LimitDegrees=45
  swing2Motion=Free
  twistLimitDegrees=45
  twistMotion=Free
 }
 {
  bone1Name=spine_02
  bone2Name=spine_01
  swing1LimitDegrees=45
  swing1Motion=Free
  swing2LimitDegrees=45
  swing2Motion=Free
  twistLimitDegrees=45
  twistMotion=Free
 }
 {
  bone1Name=spine_03
  bone2Name=spine_02
  swing1LimitDegrees=45
  swing1Motion=Free
  swing2LimitDegrees=45
  swing2Motion=Free
  twistLimitDegrees=45
  twistMotion=Free
 }
 {
  bone1Name=spine_04
  bone2Name=spine_03
  swing1LimitDegrees=45
  swing1Motion=Free
  swing2LimitDegrees=45
  swing2Motion=Free
  twistLimitDegrees=45
  twistMotion=Free
 }
 {
  bone1Name=spine_05
  bone2Name=spine_04
  swing1LimitDegrees=45
  swing1Motion=Free
  swing2LimitDegrees=45
  swing2Motion=Free
  twistLimitDegrees=45
  twistMotion=Free
 }
 {
  bone1Name=thigh_l
  bone2Name=pelvis
  swing1LimitDegrees=45
  swing1Motion=Free
  swing2LimitDegrees=45
  swing2Motion=Free
  twistLimitDegrees=45
  twistMotion=Free
 }
 {
  bone1Name=calf_l
  bone2Name=thigh_l
  swing1LimitDegrees=45
  swing1Motion=Free
  swing2LimitDegrees=45
  swing2Motion=Free
  twistLimitDegrees=45
  twistMotion=Free
 }
 {
  bone1Name=thigh_r
  bone2Name=pelvis
  swing1LimitDegrees=45
  swing1Motion=Free
  swing2LimitDegrees=45
  swing2Motion=Free
  twistLimitDegrees=45
  twistMotion=Free
 }
 {
  bone1Name=calf_r
  bone2Name=thigh_r
  swing1LimitDegrees=45
  swing1Motion=Free
  swing2LimitDegrees=45
  swing2Motion=Free
  twistLimitDegrees=45
  twistMotion=Free
 }
 {bone1Name=cape_mid_02_l, bone2Name=cape_mid_01_l, swing1LimitDegrees=15, swing2LimitDegrees=5, twistLimitDegrees=5}
 {bone1Name=cape_mid_03_l, bone2Name=cape_mid_02_l, swing1LimitDegrees=15, swing2LimitDegrees=5, twistLimitDegrees=5}
 {bone1Name=cape_mid_04_l, bone2Name=cape_mid_03_l, swing1LimitDegrees=20, swing2LimitDegrees=2, twistLimitDegrees=5}
 {bone1Name=cape_mid_05_l, bone2Name=cape_mid_04_l, swing1LimitDegrees=20, swing2LimitDegrees=2, twistLimitDegrees=5}
 {bone1Name=cape_mid_06_l, bone2Name=cape_mid_05_l, swing1LimitDegrees=20, swing2LimitDegrees=2, twistLimitDegrees=5}
 {bone1Name=cape_mid_02_r, bone2Name=cape_mid_01_r, swing1LimitDegrees=15, swing2LimitDegrees=5, twistLimitDegrees=5}
 {bone1Name=cape_mid_03_r, bone2Name=cape_mid_02_r, swing1LimitDegrees=15, swing2LimitDegrees=5, twistLimitDegrees=5}
 {bone1Name=cape_mid_04_r, bone2Name=cape_mid_03_r, swing1LimitDegrees=20, swing2LimitDegrees=2, twistLimitDegrees=5}
 {bone1Name=cape_mid_05_r, bone2Name=cape_mid_04_r, swing1LimitDegrees=20, swing2LimitDegrees=2, twistLimitDegrees=5}
 {bone1Name=cape_mid_06_r, bone2Name=cape_mid_05_r, swing1LimitDegrees=20, swing2LimitDegrees=2, twistLimitDegrees=5}
 {
  bone1Name=clavicle_l
  bone2Name=spine_05
  swing1LimitDegrees=45
  swing1Motion=Free
  swing2LimitDegrees=45
  swing2Motion=Free
  twistLimitDegrees=45
  twistMotion=Free
 }
 {
  bone1Name=upperarm_l
  bone2Name=clavicle_l
  swing1LimitDegrees=45
  swing1Motion=Free
  swing2LimitDegrees=45
  swing2Motion=Free
  twistLimitDegrees=45
  twistMotion=Free
 }
 {
  bone1Name=lowerarm_l
  bone2Name=upperarm_l
  swing1LimitDegrees=45
  swing1Motion=Free
  swing2LimitDegrees=45
  swing2Motion=Free
  twistLimitDegrees=45
  twistMotion=Free
 }
 {
  bone1Name=clavicle_r
  bone2Name=spine_05
  swing1LimitDegrees=45
  swing1Motion=Free
  swing2LimitDegrees=45
  swing2Motion=Free
  twistLimitDegrees=45
  twistMotion=Free
 }
 {
  bone1Name=lowerarm_r
  bone2Name=upperarm_r
  swing1LimitDegrees=45
  swing1Motion=Free
  swing2LimitDegrees=45
  swing2Motion=Free
  twistLimitDegrees=45
  twistMotion=Free
 }
 {
  bone1Name=upperarm_r
  bone2Name=clavicle_r
  swing1LimitDegrees=45
  swing1Motion=Free
  swing2LimitDegrees=45
  swing2Motion=Free
  twistLimitDegrees=45
  twistMotion=Free
 }
 {
  bone1Name=cape_07_l
  bone2Name=cape_mid_06_l
  swing1LimitDegrees=45
  swing1Motion=Free
  swing2LimitDegrees=45
  swing2Motion=Free
  twistLimitDegrees=45
  twistMotion=Free
 }
 {
  bone1Name=cape_06_l
  bone2Name=cape_mid_05_l
  swing1LimitDegrees=45
  swing1Motion=Free
  swing2LimitDegrees=45
  swing2Motion=Free
  twistLimitDegrees=45
  twistMotion=Free
 }
 {
  bone1Name=cape_06_r
  bone2Name=cape_mid_05_r
  swing1LimitDegrees=45
  swing1Motion=Free
  swing2LimitDegrees=45
  swing2Motion=Free
  twistLimitDegrees=45
  twistMotion=Free
 }
 {
  bone1Name=cape_07_r
  bone2Name=cape_mid_06_r
  swing1LimitDegrees=45
  swing1Motion=Free
  swing2LimitDegrees=45
  swing2Motion=Free
  twistLimitDegrees=45
  twistMotion=Free
 }
 {
  bone1Name=skirt_back_a_03
  bone2Name=pelvis
  swing1LimitDegrees=45
  swing1Motion=Free
  swing2LimitDegrees=45
  swing2Motion=Free
  twistLimitDegrees=45
  twistMotion=Free
 }
 {
  bone1Name=skirt_back_a_04
  bone2Name=skirt_back_a_03
  swing1LimitDegrees=45
  swing1Motion=Free
  swing2LimitDegrees=45
  swing2Motion=Free
  twistLimitDegrees=45
  twistMotion=Free
 }
 {
  bone1Name=skirt_back_a_05
  bone2Name=skirt_back_a_04
  swing1LimitDegrees=45
  swing1Motion=Free
  swing2LimitDegrees=45
  swing2Motion=Free
  twistLimitDegrees=45
  twistMotion=Free
 }
 {
  bone1Name=skirt_back_b_03
  bone2Name=pelvis
  swing1LimitDegrees=45
  swing1Motion=Free
  swing2LimitDegrees=45
  swing2Motion=Free
  twistLimitDegrees=45
  twistMotion=Free
 }
 {
  bone1Name=skirt_back_b_04
  bone2Name=skirt_back_b_03
  swing1LimitDegrees=45
  swing1Motion=Free
  swing2LimitDegrees=45
  swing2Motion=Free
  twistLimitDegrees=45
  twistMotion=Free
 }
 {
  bone1Name=skirt_back_b_05
  bone2Name=skirt_back_b_04
  swing1LimitDegrees=45
  swing1Motion=Free
  swing2LimitDegrees=45
  swing2Motion=Free
  twistLimitDegrees=45
  twistMotion=Free
 }
]
