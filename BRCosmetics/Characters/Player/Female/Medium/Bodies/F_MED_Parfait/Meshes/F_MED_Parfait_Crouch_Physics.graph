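# Physics asset for the crouch variant of the F_MED_Parfait player body, covering its skirt and loincloth bone chains.
# Pelvis, thigh and calf shapes form the lower body, with extra boxes on thighs and calves.
# Limited-swing chains hang from the pelvis through angularly locked loincloth masters and free skirt bases, while skirt connectors attach to the thighs.
# Angularly free constraints cross-link neighboring chain tips, pairing loincloths, side panels and back panels.
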PhysicsAsset /BRCosmetics/Characters/Player/Female/Medium/Bodies/F_MED_Parfait/Meshes/F_MED_Parfait_Crouch_Physics
bodies {
 calf_l [
  {
   shapeType=Capsule
   length=30.5
   radius=6.5
   shapeName=calf_r_capsule
   center=<-20.795 -1.0725 -0.8611>
   rotation=<rot -0.024 88.3082 87.0673>
  }
  {
   shapeType=Box
   extentX=16.5
   extentY=38.5
   extentZ=13.5
   shapeName=calf_r_box
   center=<-20.9608 -6.6931 -0.9532>
   rotation=<rot 4.0512 70.0164 175.8646>
  }
 ]
 calf_r [
  {shapeType=Capsule, length=30.5, radius=6.5, shapeName=calf_r_capsule, center=<20.795 1.0725 0.8611>, rotation=<rot -0.024 88.3082 -92.9327>}
  {
   shapeType=Box
   extentX=16.5
   extentY=38.5
   extentZ=13.5
   shapeName=calf_r_box
   center=<20.9608 6.6931 0.9532>
   rotation=<rot 4.0512 70.0164 -4.1354>
  }
 ]
 calf_twist_02_l [{
   shapeType=Capsule
   length=32.3735
   radius=6.8268
   shapeName=calf_twist_02_r_capsule
   center=<-8.0443 -8.5679 -0.4318>
   rotation=<rot -0.6816 -94.3312 94.5737>
  }]
 calf_twist_02_r [{
   shapeType=Capsule
   length=32.3735
   radius=6.8268
   shapeName=calf_twist_02_r_capsule
   center=<8.0443 8.5679 0.4318>
   rotation=<rot -0.6816 -94.3312 -85.4263>
  }]
 loincloth_fl_01 [{shapeType=Capsule, length=4, radius=3, shapeName=loincloth_fl_01_capsule, center=<3 0 0>, rotation=<rot 0 0 -4.0082>}]
 loincloth_fl_02 [{shapeType=Capsule, length=5, radius=2.5, shapeName=loincloth_fl_02_capsule, center=<3 0.0547 -0.6773>, rotation=<rot 0 0.0001 -4.6165>}]
 loincloth_fl_03 [{shapeType=Capsule, length=5, radius=2.5, shapeName=loincloth_fl_03_capsule, center=<3 0.0411 -0.4842>, rotation=<rot 0 0 -4.8493>}]
 loincloth_fl_04 [{shapeType=Capsule, length=7, radius=2.5, shapeName=loincloth_fl_04_capsule, center=<3 0.0696 -0.7652>, rotation=<rot 0 0.0001 -5.1997>}]
 loincloth_fl_05 [{shapeType=Capsule, length=7.25, radius=2.75, shapeName=loincloth_fl_05_capsule, center=<3 0.1531 -0.6048>, rotation=<rot 0 0.0002 -14.2036>}]
 loincloth_fl_master [{shapeType=Capsule, length=1, radius=1, shapeName=loincloth_fl_master_capsule}]
 loincloth_fr_01 [{shapeType=Capsule, length=4, radius=3, shapeName=loincloth_fr_01_capsule, center=<3 0 0>}]
 loincloth_fr_02 [{shapeType=Capsule, length=3, radius=2.5, shapeName=loincloth_fr_02_capsule, center=<3 0 0>}]
 loincloth_fr_03 [{shapeType=Capsule, length=4, radius=2.5, shapeName=loincloth_fr_03_capsule, center=<3 0 0>}]
 loincloth_fr_04 [{shapeType=Capsule, length=6, radius=2.5, shapeName=loincloth_fr_04_capsule, center=<3 0 0>}]
 loincloth_fr_05 [{shapeType=Capsule, length=5.75, radius=2.75, shapeName=loincloth_fr_05_capsule, center=<3 0 0>}]
 loincloth_fr_master [{shapeType=Capsule, length=1, radius=1, shapeName=loincloth_fr_master_capsule}]
 pelvis [{
   shapeType=Capsule
   length=15.1534
   radius=6.5145
   shapeName=pelvis_capsule
   center=<0.4119 -1.5176 -0.0401>
   rotation=<rot -0.8022 -90 -0.0752>
  }]
 skirt_back_a_01 [{shapeType=Capsule, length=6, radius=3, shapeName=skirt_back_a_01_capsule, center=<3 0 0>}]
 skirt_back_a_02 [{shapeType=Capsule, length=6, radius=3, shapeName=skirt_back_a_02_capsule, center=<3 0 0>}]
 skirt_back_a_03 [{shapeType=Capsule, length=6, radius=3, shapeName=skirt_back_a_03_capsule, center=<3 0 0>}]
 skirt_back_a_04 [{shapeType=Capsule, length=6, radius=3, shapeName=skirt_back_a_04_capsule, center=<3 0 0>}]
 skirt_back_a_base [{shapeType=Capsule, length=1, radius=1, shapeName=skirt_back_a_base_capsule}]
 skirt_back_b_01 [{shapeType=Capsule, length=6, radius=3, shapeName=skirt_back_b_01_capsule, center=<3 0 0>}]
 skirt_back_b_02 [{shapeType=Capsule, length=6, radius=3, shapeName=skirt_back_b_02_capsule, center=<3 0 0>}]
 skirt_back_b_03 [{shapeType=Capsule, length=6, radius=3, shapeName=skirt_back_b_03_capsule, center=<3 0 0>}]
 skirt_back_b_04 [{shapeType=Capsule, length=6, radius=3, shapeName=skirt_back_b_04_capsule, center=<3 0 0>}]
 skirt_back_b_base [{shapeType=Capsule, length=1, radius=1, shapeName=skirt_back_b_base_capsule}]
 skirt_bl_01 [{shapeType=Capsule, length=6, radius=3, shapeName=skirt_bl_01_capsule, center=<3 0 0>, rotation=<rot 0 0 90>}]
 skirt_bl_02 [{shapeType=Capsule, length=15.75, radius=3, shapeName=skirt_bl_02_capsule, center=<3 0 0>, rotation=<rot 0 0 90>}]
 skirt_bl_03 [{shapeType=Capsule, length=15.75, radius=3, shapeName=skirt_bl_03_capsule, center=<3 0 0>, rotation=<rot 0 0 90>}]
 skirt_bl_04 [{shapeType=Capsule, length=15.75, radius=3, shapeName=skirt_bl_04_capsule, center=<3 0 0>, rotation=<rot 0 0 90>}]
 skirt_bl_base [{shapeType=Capsule, length=1, radius=1, shapeName=skirt_bl_base_capsule, rotation=<rot 0 0 90>}]
 skirt_br_01 [{shapeType=Capsule, length=13, radius=3, shapeName=skirt_br_01_capsule, center=<3 0 0>, rotation=<rot 0 0 90>}]
 skirt_br_02 [{shapeType=Capsule, length=13, radius=3, shapeName=skirt_br_02_capsule, center=<3 0 0>, rotation=<rot 0 0 90>}]
 skirt_br_03 [{shapeType=Capsule, length=13, radius=3, shapeName=skirt_br_03_capsule, center=<3 0 0>, rotation=<rot 0 0 90>}]
 skirt_br_04 [{shapeType=Capsule, length=13, radius=3, shapeName=skirt_br_04_capsule, center=<3 0 0>, rotation=<rot 0 0 90>}]
 skirt_br_base [{shapeType=Capsule, length=1, radius=1, shapeName=skirt_br_base_capsule, rotation=<rot 0 0 90>}]
 skirt_connector_01_l [{shapeType=Capsule, length=1, radius=2, shapeName=skirt_connector_01_l_capsule, center=<0.5 0 0>, rotation=<rot 90 0 0>}]
 skirt_connector_01_r [{shapeType=Capsule, length=4, radius=2, shapeName=skirt_connector_01_r_capsule, center=<3 0 0>, rotation=<rot 90 0 0>}]
 skirt_fl_01 [{
   shapeType=Box
   extentX=10
   extentY=14
   extentZ=2
   shapeName=skirt_fl_01_box
   center=<-6.2482 0.0221 0.6411>
   rotation=<rot 3.5272 0.6495 9.26>
  }]
 skirt_fl_02 [{
   shapeType=Box
   extentX=9
   extentY=13.5
   extentZ=2.5
   shapeName=skirt_fl_02_box
   center=<-4.2064 0.0609 0.3899>
   rotation=<rot -1.2125 -0.2505 8.8843>
  }]
 skirt_fl_03 [{shapeType=Capsule, length=10.25, radius=3, shapeName=skirt_fl_03_capsule, center=<-3 0 0>, rotation=<rot 0 0 90>}]
 skirt_fl_04 [{shapeType=Capsule, length=10.25, radius=3, shapeName=skirt_fl_04_capsule, center=<-3 0 0>, rotation=<rot 0 0 90>}]
 skirt_fl_base [{
   shapeType=Capsule
   length=2.8586
   radius=1.3863
   shapeName=skirt_fl_base_capsule
   center=<-2.0134 0.5022 -0.0455>
   rotation=<rot 0 151.2575 -85.7636>
  }]
 skirt_fr_01 [{shapeType=Box, extentX=10, extentY=14, extentZ=2.5, shapeName=skirt_fr_01_box, center=<6.7124 0 0>, rotation=<rot -0.0001 0.0001 10.0001>}]
 skirt_fr_02 [{shapeType=Box, extentX=9, extentY=13.5, extentZ=3, shapeName=skirt_fr_02_box, center=<4.2268 0 0>, rotation=<rot -0.0002 0.0003 20.0001>}]
 skirt_fr_03 [{shapeType=Box, extentX=9, extentY=13.5, extentZ=3, shapeName=skirt_fr_02_box, center=<4.2268 0 0>}]
 skirt_fr_04 [{shapeType=Box, extentX=9, extentY=13.5, extentZ=3, shapeName=skirt_fr_02_box, center=<4.2268 0 0>}]
 skirt_fr_base [{shapeType=Capsule, length=1, radius=1, shapeName=skirt_fr_base_capsule}]
 thigh_l [
  {shapeType=Capsule, length=25.25, radius=7, shapeName=thigh_l_capsule, center=<-15.003 -0.0996 0>, rotation=<rot 90 180 -178.2811>}
  {
   shapeType=Box
   extentX=40.5
   extentY=12.5
   extentZ=13.5
   shapeName=thigh_l_box
   center=<-28.269 -8.8974 -1.7109>
   rotation=<rot 1.4505 16.0001 0.1112>
  }
 ]
 thigh_r [
  {shapeType=Capsule, length=25.5, radius=7, shapeName=thigh_r_capsule, center=<14.997 -0.0996 0>, rotation=<rot 90 179.8881 -178.393>}
  {
   shapeType=Box
   extentX=31.5
   extentY=14.5
   extentZ=13.5
   shapeName=thigh_r_box
   center=<33.1219 10.2703 2.1231>
   rotation=<rot -0.7134 16.1379 -10.5567>
  }
 ]
 thigh_twist_01_l [
  {shapeType=Sphere, radius=5, shapeName=thigh_twist_01_l_sphere, center=<14.5683 5.8734 0.0958>}
  {shapeType=Capsule, length=9.5025, radius=8.7687, shapeName=thigh_twist_01_l_capsule, rotation=<rot 90 0 0>}
 ]
 thigh_twist_01_r [{shapeType=Capsule, length=9.5025, radius=8.7687, shapeName=thigh_twist_01_r_capsule, rotation=<rot 90 0 0>}]
}
constraints [
 {bone1Name=loincloth_fr_01, bone2Name=loincloth_fr_master, swing1LimitDegrees=12.5, swing2LimitDegrees=7.5, twistLimitDegrees=5}
 {bone1Name=loincloth_fr_02, bone2Name=loincloth_fr_01, swing1LimitDegrees=20, swing2LimitDegrees=7.5, twistLimitDegrees=5}
 {bone1Name=loincloth_fr_03, bone2Name=loincloth_fr_02, swing1LimitDegrees=15, swing2LimitDegrees=7.5, twistLimitDegrees=5}
 {bone1Name=loincloth_fr_04, bone2Name=loincloth_fr_03, swing1LimitDegrees=15, swing2LimitDegrees=2, twistLimitDegrees=5}
 {bone1Name=loincloth_fr_05, bone2Name=loincloth_fr_04, swing1LimitDegrees=15, swing2LimitDegrees=2, twistLimitDegrees=5}
 {bone1Name=loincloth_fl_01, bone2Name=loincloth_fl_master, swing1LimitDegrees=12.5, swing2LimitDegrees=7.5, twistLimitDegrees=5}
 {bone1Name=loincloth_fl_02, bone2Name=loincloth_fl_01, swing1LimitDegrees=20, swing2LimitDegrees=7.5, twistLimitDegrees=5}
 {bone1Name=loincloth_fl_03, bone2Name=loincloth_fl_02, swing1LimitDegrees=15, swing2LimitDegrees=7.5, twistLimitDegrees=5}
 {bone1Name=loincloth_fl_04, bone2Name=loincloth_fl_03, swing1LimitDegrees=15, swing2LimitDegrees=2, twistLimitDegrees=5}
 {bone1Name=loincloth_fl_05, bone2Name=loincloth_fl_04, swing1LimitDegrees=15, swing2LimitDegrees=2, twistLimitDegrees=5}
 {bone1Name=skirt_bl_01, bone2Name=skirt_bl_base, swing1LimitDegrees=5, swing2LimitDegrees=15, twistLimitDegrees=5}
 {bone1Name=skirt_bl_02, bone2Name=skirt_bl_01, swing1LimitDegrees=5, swing2LimitDegrees=15, twistLimitDegrees=5}
 {bone1Name=skirt_bl_03, bone2Name=skirt_bl_02, swing1LimitDegrees=5, swing2LimitDegrees=15, twistLimitDegrees=5}
 {bone1Name=skirt_bl_04, bone2Name=skirt_bl_03, swing1LimitDegrees=5, swing2LimitDegrees=15, twistLimitDegrees=5}
 {bone1Name=skirt_br_01, bone2Name=skirt_br_base, swing1LimitDegrees=5, swing2LimitDegrees=15, twistLimitDegrees=5}
 {bone1Name=skirt_br_02, bone2Name=skirt_br_01, swing1LimitDegrees=5, swing2LimitDegrees=15, twistLimitDegrees=5}
 {bone1Name=skirt_br_03, bone2Name=skirt_br_02, swing1LimitDegrees=5, swing2LimitDegrees=15, twistLimitDegrees=5}
 {bone1Name=skirt_br_04, bone2Name=skirt_br_03, swing1LimitDegrees=5, swing2LimitDegrees=15, twistLimitDegrees=5}
 {bone1Name=skirt_back_a_01, bone2Name=skirt_back_a_base, swing1LimitDegrees=30, swing2LimitDegrees=30, twistLimitDegrees=5}
 {bone1Name=skirt_back_a_02, bone2Name=skirt_back_a_01, swing1LimitDegrees=30, swing2LimitDegrees=30, twistLimitDegrees=5}
 {bone1Name=skirt_back_a_03, bone2Name=skirt_back_a_02, swing1LimitDegrees=30, swing2LimitDegrees=10, twistLimitDegrees=5}
 {bone1Name=skirt_back_a_04, bone2Name=skirt_back_a_03, swing1LimitDegrees=30, swing2LimitDegrees=10, twistLimitDegrees=5}
 {bone1Name=skirt_back_b_01, bone2Name=skirt_back_b_base, swing1LimitDegrees=30, swing2LimitDegrees=30, twistLimitDegrees=5}
 {bone1Name=skirt_back_b_02, bone2Name=skirt_back_b_01, swing1LimitDegrees=30, swing2LimitDegrees=30, twistLimitDegrees=5}
 {bone1Name=skirt_back_b_03, bone2Name=skirt_back_b_02, swing1LimitDegrees=30, swing2LimitDegrees=10, twistLimitDegrees=5}
 {bone1Name=skirt_back_b_04, bone2Name=skirt_back_b_03, swing1LimitDegrees=30, swing2LimitDegrees=10, twistLimitDegrees=5}
 {
  bone1Name=loincloth_fr_master
  bone2Name=pelvis
  swing1LimitDegrees=45
  swing1Motion=Locked
  swing2LimitDegrees=45
  swing2Motion=Locked
  twistLimitDegrees=45
  twistMotion=Locked
 }
 {
  bone1Name=loincloth_fl_master
  bone2Name=pelvis
  swing1LimitDegrees=45
  swing1Motion=Locked
  swing2LimitDegrees=45
  swing2Motion=Locked
  twistLimitDegrees=45
  twistMotion=Locked
 }
 {
  bone1Name=skirt_bl_base
  bone2Name=pelvis
  swing1LimitDegrees=45
  swing1Motion=Free
  swing2LimitDegrees=45
  swing2Motion=Free
  twistLimitDegrees=45
  twistMotion=Free
 }
 {
  bone1Name=skirt_back_a_base
  bone2Name=pelvis
  swing1LimitDegrees=45
  swing1Motion=Free
  swing2LimitDegrees=45
  swing2Motion=Free
  twistLimitDegrees=45
  twistMotion=Free
 }
 {
  bone1Name=skirt_back_b_base
  bone2Name=pelvis
  swing1LimitDegrees=45
  swing1Motion=Free
  swing2LimitDegrees=45
  swing2Motion=Free
  twistLimitDegrees=45
  twistMotion=Free
 }
 {
  bone1Name=skirt_br_base
  bone2Name=pelvis
  swing1LimitDegrees=45
  swing1Motion=Free
  swing2LimitDegrees=45
  swing2Motion=Free
  twistLimitDegrees=45
  twistMotion=Free
 }
 {
  bone1Name=thigh_l
  bone2Name=pelvis
  swing1LimitDegrees=45
  swing1Motion=Free
  swing2LimitDegrees=45
  swing2Motion=Free
  twistLimitDegrees=45
  twistMotion=Free
 }
 {
  bone1Name=thigh_r
  bone2Name=pelvis
  swing1LimitDegrees=45
  swing1Motion=Free
  swing2LimitDegrees=45
  swing2Motion=Free
  twistLimitDegrees=45
  twistMotion=Free
 }
 {
  bone1Name=thigh_twist_01_r
  bone2Name=thigh_r
  swing1LimitDegrees=45
  swing1Motion=Free
  swing2LimitDegrees=45
  swing2Motion=Free
  twistLimitDegrees=45
  twistMotion=Free
 }
 {
  bone1Name=thigh_twist_01_l
  bone2Name=thigh_l
  swing1LimitDegrees=45
  swing1Motion=Free
  swing2LimitDegrees=45
  swing2Motion=Free
  twistLimitDegrees=45
  twistMotion=Free
 }
 {
  bone1Name=skirt_fr_base
  bone2Name=pelvis
  swing1LimitDegrees=45
  swing1Motion=Free
  swing2LimitDegrees=45
  swing2Motion=Free
  twistLimitDegrees=45
  twistMotion=Free
 }
 {bone1Name=skirt_fr_01, bone2Name=skirt_fr_base, swing1LimitDegrees=20, swing2LimitDegrees=20, twistLimitDegrees=45}
 {bone1Name=skirt_fr_02, bone2Name=skirt_fr_01, swing1LimitDegrees=5, swing2LimitDegrees=15, twistLimitDegrees=5}
 {bone1Name=skirt_fr_03, bone2Name=skirt_fr_02, swing1LimitDegrees=5, swing2LimitDegrees=15, twistLimitDegrees=5}
 {bone1Name=skirt_fr_04, bone2Name=skirt_fr_03, swing1LimitDegrees=5, swing2LimitDegrees=15, twistLimitDegrees=5}
 {
  bone1Name=skirt_fl_base
  bone2Name=pelvis
  swing1LimitDegrees=45
  swing1Motion=Free
  swing2LimitDegrees=45
  swing2Motion=Free
  twistLimitDegrees=45
  twistMotion=Free
 }
 {bone1Name=skirt_fl_01, bone2Name=skirt_fl_base, swing1LimitDegrees=45, swing2LimitDegrees=30, twistLimitDegrees=45}
 {bone1Name=skirt_fl_02, bone2Name=skirt_fl_01, swing1LimitDegrees=5, swing2LimitDegrees=15, twistLimitDegrees=5}
 {bone1Name=skirt_fl_03, bone2Name=skirt_fl_02, swing1LimitDegrees=5, swing2LimitDegrees=15, twistLimitDegrees=5}
 {bone1Name=skirt_fl_04, bone2Name=skirt_fl_03, swing1LimitDegrees=5, swing2LimitDegrees=15, twistLimitDegrees=5}
 {
  bone1Name=loincloth_fl_05
  bone2Name=loincloth_fr_05
  swing1LimitDegrees=45
  swing1Motion=Free
  swing2LimitDegrees=45
  swing2Motion=Free
  twistLimitDegrees=45
  twistMotion=Free
 }
 {
  bone1Name=calf_r
  bone2Name=thigh_r
  swing1LimitDegrees=45
  swing1Motion=Free
  swing2LimitDegrees=45
  swing2Motion=Free
  twistLimitDegrees=45
  twistMotion=Free
 }
 {
  bone1Name=calf_l
  bone2Name=thigh_l
  swing1LimitDegrees=45
  swing1Motion=Free
  swing2LimitDegrees=45
  swing2Motion=Free
  twistLimitDegrees=45
  twistMotion=Free
 }
 {
  bone1Name=calf_twist_02_r
  bone2Name=calf_r
  swing1LimitDegrees=45
  swing1Motion=Free
  swing2LimitDegrees=45
  swing2Motion=Free
  twistLimitDegrees=45
  twistMotion=Free
 }
 {
  bone1Name=calf_twist_02_l
  bone2Name=calf_l
  swing1LimitDegrees=45
  swing1Motion=Free
  swing2LimitDegrees=45
  swing2Motion=Free
  twistLimitDegrees=45
  twistMotion=Free
 }
 {bone1Name=skirt_connector_01_l, bone2Name=thigh_l, swing1LimitDegrees=45, swing2LimitDegrees=5, twistLimitDegrees=15}
 {bone1Name=skirt_connector_01_r, bone2Name=thigh_r, swing1LimitDegrees=45, swing2LimitDegrees=5, twistLimitDegrees=15}
 {
  bone1Name=skirt_br_04
  bone2Name=skirt_fr_04
  swing1LimitDegrees=45
  swing1Motion=Free
  swing2LimitDegrees=45
  swing2Motion=Free
  twistLimitDegrees=45
  twistMotion=Free
 }
 {
  bone1Name=skirt_back_b_04
  bone2Name=skirt_back_a_04
  swing1LimitDegrees=45
  swing1Motion=Free
  swing2LimitDegrees=45
  swing2Motion=Free
  twistLimitDegrees=45
  twistMotion=Free
 }
 {
  bone1Name=skirt_bl_04
  bone2Name=skirt_fl_04
  swing1LimitDegrees=45
  swing1Motion=Free
  swing2LimitDegrees=45
  swing2Motion=Free
  twistLimitDegrees=45
  twistMotion=Free
 }
]
