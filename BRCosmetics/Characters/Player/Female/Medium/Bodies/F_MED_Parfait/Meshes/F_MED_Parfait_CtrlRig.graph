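# Control Rig for the F_MED_Parfait player body that applies corrective offsets to skirt, loincloth and forearm plate bones.
# Forearm plates get constant offsets, then Crouching, Emoting and default branches pose the cloth, with Moving refining crouch and Jumping refining default.
# Weights come mostly from pelvis-relative thigh Euler angles remapped to ranges, or spherical pose readers on the thighs.
# AlphaInterp smooths some weights, and several debug draws remain enabled.
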ControlRigBlueprint /BRCosmetics/Characters/Player/Female/Medium/Bodies/F_MED_Parfait/Meshes/F_MED_Parfait_CtrlRig
graphs [
 RigVMFunctionLibrary {links [], nodes []}
 RigVMModel {nodes [
   RigUnit_BeginExecution = &RigUnit_BeginExecution()
   n1 = /Script/ControlRig.RigUnit_SequenceExecution(ExecuteContext=@RigUnit_BeginExecution.ExecutePin) {is=struct RigVMUnitNode}
   &RigUnit_ModifyTransforms(
    ItemToModify=((Item=(Type=Bone,Name="deform_lowerarm_plates_l"),Transform=(Rotation=(X=0,Y=0,Z=0,W=1),Translation=(X=-3,Y=0,Z=0.5),Scale3D=(X=1,Y=1,Z=1))))
    Weight=1
    ExecutePin=@n1.A
   )
   > &RigUnit_ModifyTransforms(
    ItemToModify=((Item=(Type=Bone,Name="deform_lowerarm_plates_r"),Transform=(Rotation=(X=0,Y=0,Z=0,W=1),Translation=(X=3,Y=0,Z=-0.5),Scale3D=(X=1,Y=1,Z=1))))
    Weight=1
   )
   Branch_1_RigVMUnitNode = &RigVMFunction_ControlFlowBranch(ExecuteContext=@n1.B, Condition=RigVMVariableNode() {variable=Crouching}.Value)
   Branch_1_1_RigVMUnitNode = &RigVMFunction_ControlFlowBranch(ExecuteContext=@Branch_1_RigVMUnitNode.False, Condition=RigVMVariableNode() {variable=Emoting}.Value)
   n2 = &RigUnit_SequenceExecution(ExecuteContext=@Branch_1_1_RigVMUnitNode.False)
   RigUnit_VisualDebugQuatItemSpace = /Script/ControlRig.RigUnit_VisualDebugQuatItemSpace(
    DebugDrawSettings=(DepthPriority=SDPG_Foreground,Lifetime=-1)
    Scale=10
    Space=(Type=Bone)
    bEnabled=True
    Value=&RigUnit_GetRelativeTransformForItem(Child=(Type=Bone,Name="thigh_l"), Parent=(Type=Bone,Name="pelvis")).RelativeTransform.Rotation
   ) {is=struct RigVMUnitNode}
   &RigUnit_ModifyTransforms(
    ItemToModify="((Item=(Type=Bone,Name=\"loincloth_fl_master\"),Transform=(Rotation=(X=0,Y=0,Z=0,W=1),Translation=(X=0,Y=0,Z=0),Scale3D=(X=1,Y=1,Z=1))),(Item=(Type=Bone,Name=\"skirt_connector_01_l\"),Transform=(Rotation=(X=0,Y=0,Z=0,W=1),Translation=(X=0,Y=0,Z=0),Scale3D=(X=1 ...(truncated)"
    Weight=&RigVMFunction_MathFloatRemap(
     SourceMaximum=0
     SourceMinimum=-90
     TargetMaximum=0
     TargetMinimum=1
     Value=&RigVMFunction_MathQuaternionToEuler(RotationOrder=ZYX, Value=@RigUnit_VisualDebugQuatItemSpace.Value).Result.Z
    ).Result
    ItemToModify.0.Transform.Rotation=&RigVMFunction_MathQuaternionFromEuler(Euler=(X=0,Y=0,Z=30)).Result
    ExecutePin=@n2.A
   )
   n3 = &RigUnit_GetRelativeTransformForItem(Child=(Type=Bone,Name="thigh_r"), Parent=(Type=Bone,Name="pelvis"))
   n4 = &RigVMFunction_MathFloatRemap(
    SourceMaximum=90
    SourceMinimum=0
    Value=&RigVMFunction_MathQuaternionToEuler(
     RotationOrder=XZY
     Value=&RigUnit_VisualDebugQuatItemSpace(
      DebugDrawSettings=(DepthPriority=SDPG_Foreground,Lifetime=-1)
      Scale=10
      Space=(Type=Bone)
      bEnabled=True
      Value=@n3.RelativeTransform.Rotation
     ).Value
    ).Result.Z
   )
   > &RigUnit_ModifyTransforms(
    ItemToModify="((Item=(Type=Bone,Name=\"loincloth_fr_master\"),Transform=(Rotation=(X=0,Y=0,Z=0,W=1),Translation=(X=0,Y=0,Z=0),Scale3D=(X=1,Y=1,Z=1))),(Item=(Type=Bone,Name=\"skirt_connector_01_r\"),Transform=(Rotation=(X=0,Y=0,Z=0,W=1),Translation=(X=0,Y=0,Z=0),Scale3D=(X=1 ...(truncated)"
    Weight=@n4.Result
    ItemToModify.0.Transform.Rotation=&RigVMFunction_MathQuaternionFromEuler(Euler=(X=0,Y=0,Z=30)).Result
   )
   > &RigUnit_ModifyTransforms(
    ItemToModify=((Item=(Type=Bone,Name="skirt_fr_base"),Transform=(Rotation=(X=0,Y=0,Z=0,W=1),Translation=(X=0,Y=0,Z=0),Scale3D=(X=1,Y=1,Z=1))))
    Weight=@n4.Result
    ItemToModify.0.Transform.Rotation=&RigVMFunction_MathQuaternionFromEuler(Euler=(X=0,Y=45,Z=0)).Result
   )
   Branch_RigVMUnitNode = &RigVMFunction_ControlFlowBranch(ExecuteContext=@n2.B, Condition=RigVMVariableNode() {variable=Jumping}.Value)
   n5 = &RigUnit_SphericalPoseReader(
    ActiveRegionScaleFactors=(PositiveWidth=0.75,NegativeWidth=0.75,PositiveHeight=0.75,NegativeHeight=0.75)
    ActiveRegionSize=0.25
    Debug=(bDrawDebug=True,DebugScale=25,DebugSegments=20,DebugThickness=0.25)
    DriverAxis=(X=1,Y=0,Z=0)
    DriverItem=(Type=Bone,Name="thigh_r")
    FalloffRegionScaleFactors=(PositiveWidth=1,NegativeWidth=1,PositiveHeight=1,NegativeHeight=1)
    FalloffSize=0.2
    OptionalParentItem=(Type=Bone)
    RotationOffset=(X=-45,Y=-30,Z=0)
    ExecutePin=@Branch_RigVMUnitNode.True
   )
   RigUnit_AlphaInterp = &RigVMFunction_AlphaInterp(
    InRange=(Max=1)
    InterpSpeedDecreasing=10
    InterpSpeedIncreasing=10
    OutRange=(Max=1)
    Scale=1
    Bias=~
    ClampMin=~
    bInterpResult=~
    bMapRange=~
    Value=@n5.OutputParam
   )
   > &RigUnit_ModifyTransforms(
    ItemToModify=((Item=(Type=Bone,Name="skirt_fr_base"),Transform=(Rotation=(X=0,Y=0,Z=0,W=1),Translation=(X=0,Y=-6,Z=-3),Scale3D=(X=1,Y=1,Z=1))))
    ItemToModify.0.Transform.Rotation=&RigVMFunction_MathQuaternionFromEuler(Euler=(X=-60,Y=15,Z=0), RotationOrder=XYZ).Result
    Weight=@RigUnit_AlphaInterp.Result
   )
   > n6 = &RigUnit_SphericalPoseReader(
    ActiveRegionScaleFactors=(PositiveWidth=0.5,NegativeWidth=0.5,PositiveHeight=0.5,NegativeHeight=0.5)
    ActiveRegionSize=0.25
    Debug=(bDrawDebug=True,DebugScale=25,DebugSegments=20,DebugThickness=0.25)
    DriverAxis=(X=-1,Y=0,Z=0)
    DriverItem=(Type=Bone,Name="thigh_l")
    FalloffRegionScaleFactors=(PositiveWidth=1,NegativeWidth=1,PositiveHeight=1,NegativeHeight=1)
    FalloffSize=0.25
    OptionalParentItem=(Type=Bone)
    RotationOffset=(X=0,Y=120,Z=-45)
   )
   > &RigUnit_ModifyTransforms(
    ItemToModify=((Item=(Type=Bone,Name="skirt_fl_base"),Transform=(Rotation=(X=0,Y=0,Z=0,W=1),Translation=(X=0,Y=1,Z=2),Scale3D=(X=1,Y=1,Z=1))))
    ItemToModify.0.Transform.Rotation=&RigVMFunction_MathQuaternionFromEuler(Euler=(X=-15,Y=30,Z=-15)).Result
    Weight=@n6.OutputParam
   )
   n7 = &RigUnit_SphericalPoseReader(
    ActiveRegionScaleFactors=(PositiveWidth=1,NegativeWidth=1,PositiveHeight=1,NegativeHeight=1)
    ActiveRegionSize=0.1
    Debug=(bDrawDebug=false,DebugScale=25,DebugSegments=20,DebugThickness=0.25)
    DriverAxis=(X=-1,Y=0,Z=0)
    DriverItem=(Type=Bone,Name="thigh_l")
    FalloffRegionScaleFactors=(PositiveWidth=1,NegativeWidth=1,PositiveHeight=1,NegativeHeight=1)
    FalloffSize=0.2
    OptionalParentItem=(Type=Bone)
    RotationOffset=(X=0,Y=100,Z=-20)
    ExecutePin=@n2.C
   )
   > n8 = &RigUnit_SphericalPoseReader(
    ActiveRegionScaleFactors=(PositiveWidth=1,NegativeWidth=1,PositiveHeight=1,NegativeHeight=1)
    ActiveRegionSize=0.25
    Debug=(bDrawDebug=false,DebugScale=25,DebugSegments=20,DebugThickness=0.25)
    DriverAxis=(X=1,Y=0,Z=0)
    DriverItem=(Type=Bone,Name="thigh_r")
    FalloffRegionScaleFactors=(PositiveWidth=1,NegativeWidth=1,PositiveHeight=1,NegativeHeight=1)
    FalloffSize=0.2
    OptionalParentItem=(Type=Bone)
    RotationOffset=(X=0,Y=-80,Z=0)
   )
   > &RigUnit_ModifyTransforms(
    ItemToModify=((Item=(Type=Bone),Transform=(Rotation=(X=0,Y=0,Z=0,W=1),Translation=(X=0,Y=0,Z=0),Scale3D=(X=1,Y=1,Z=1))),(Item=(Type=Bone,Name="skirt_connector_01_l"),Transform=(Rotation=(X=0,Y=0,Z=0,W=1),Translation=(X=0,Y=0,Z=0),Scale3D=(X=1,Y=1,Z=1))))
    ItemToModify.1.Transform.Rotation=&RigVMFunction_MathQuaternionFromEuler(Euler=(X=0,Y=0,Z=-45)).Result
    Weight=@n7.OutputParam
   )
   > &RigUnit_ModifyTransforms(
    ItemToModify=((Item=(Type=Bone),Transform=(Rotation=(X=0,Y=0,Z=0,W=1),Translation=(X=0,Y=0,Z=0),Scale3D=(X=1,Y=1,Z=1))),(Item=(Type=Bone,Name="skirt_connector_01_r"),Transform=(Rotation=(X=0,Y=0,Z=0,W=1),Translation=(X=0,Y=0,Z=0),Scale3D=(X=1,Y=1,Z=1))))
    ItemToModify.1.Transform.Rotation=&RigVMFunction_MathQuaternionFromEuler(Euler=(X=0,Y=0,Z=-20)).Result
    Weight=@n8.OutputParam
   )
   n9 = &RigUnit_GetRelativeTransformForItem(Child=(Type=Bone,Name="thigh_l"), Parent=(Type=Bone,Name="pelvis"))
   n10 = &RigVMFunction_MathFloatRemap(
    SourceMaximum=0
    SourceMinimum=-90
    TargetMaximum=0
    TargetMinimum=1
    Value=&RigVMFunction_MathQuaternionToEuler(
     RotationOrder=ZYX
     Value=&RigUnit_VisualDebugQuatItemSpace(
      DebugDrawSettings=(DepthPriority=SDPG_Foreground,Lifetime=-1)
      Scale=10
      Space=(Type=Bone)
      bEnabled=True
      Value=@n9.RelativeTransform.Rotation
     ).Value
    ).Result.Z
   )
   &RigUnit_ModifyTransforms(
    ItemToModify="((Item=(Type=Bone,Name=\"loincloth_fl_master\"),Transform=(Rotation=(X=0,Y=0,Z=0,W=1),Translation=(X=0,Y=0,Z=0),Scale3D=(X=1,Y=1,Z=1))),(Item=(Type=Bone,Name=\"skirt_connector_01_l\"),Transform=(Rotation=(X=0,Y=0,Z=0,W=1),Translation=(X=0,Y=0,Z=0),Scale3D=(X=1 ...(truncated)"
    ExecutePin=@Branch_1_1_RigVMUnitNode.True
    Weight=@n10.Result
    ItemToModify.0.Transform.Rotation=&RigVMFunction_MathQuaternionFromEuler(Euler=(X=0,Y=0,Z=30)).Result
   )
   n11 = &RigUnit_GetRelativeTransformForItem(Child=(Type=Bone,Name="thigh_r"), Parent=(Type=Bone,Name="pelvis"))
   n12 = &RigVMFunction_MathFloatRemap(
    SourceMaximum=90
    SourceMinimum=0
    Value=&RigVMFunction_MathQuaternionToEuler(
     RotationOrder=XZY
     Value=&RigUnit_VisualDebugQuatItemSpace(
      DebugDrawSettings=(DepthPriority=SDPG_Foreground,Lifetime=-1)
      Scale=10
      Space=(Type=Bone)
      bEnabled=True
      Value=@n11.RelativeTransform.Rotation
     ).Value
    ).Result.Z
   )
   > &RigUnit_ModifyTransforms(
    ItemToModify="((Item=(Type=Bone,Name=\"loincloth_fr_master\"),Transform=(Rotation=(X=0,Y=0,Z=0,W=1),Translation=(X=0,Y=0,Z=0),Scale3D=(X=1,Y=1,Z=1))),(Item=(Type=Bone,Name=\"skirt_connector_01_r\"),Transform=(Rotation=(X=0,Y=0,Z=0,W=1),Translation=(X=0,Y=0,Z=0),Scale3D=(X=1 ...(truncated)"
    Weight=@n12.Result
    ItemToModify.0.Transform.Rotation=&RigVMFunction_MathQuaternionFromEuler(Euler=(X=0,Y=0,Z=30)).Result
   )
   > &RigUnit_ModifyTransforms(
    ItemToModify=((Item=(Type=Bone,Name="skirt_fr_base"),Transform=(Rotation=(X=0,Y=0,Z=0,W=1),Translation=(X=0,Y=-2,Z=-5),Scale3D=(X=1,Y=1,Z=1))))
    Weight=@n12.Result
    ItemToModify.0.Transform.Rotation=&RigVMFunction_MathQuaternionFromEuler(Euler=(X=30,Y=60,Z=-15)).Result
   )
   > &RigUnit_ModifyTransforms(
    ItemToModify=((Item=(Type=Bone,Name="skirt_fl_base"),Transform=(Rotation=(X=0,Y=0,Z=0,W=1),Translation=(X=0,Y=2,Z=5),Scale3D=(X=1,Y=1,Z=1))))
    Weight=@n10.Result
    ItemToModify.0.Transform.Rotation=&RigVMFunction_MathQuaternionFromEuler(Euler=(X=-30,Y=60,Z=-15)).Result
   )
   > n13 = &RigUnit_SphericalPoseReader(
    ActiveRegionScaleFactors=(PositiveWidth=1,NegativeWidth=1,PositiveHeight=1,NegativeHeight=1)
    ActiveRegionSize=0.1
    Debug=(bDrawDebug=false,DebugScale=25,DebugSegments=20,DebugThickness=0.25)
    DriverAxis=(X=-1,Y=0,Z=0)
    DriverItem=(Type=Bone,Name="thigh_l")
    FalloffRegionScaleFactors=(PositiveWidth=1,NegativeWidth=1,PositiveHeight=1,NegativeHeight=1)
    FalloffSize=0.2
    OptionalParentItem=(Type=Bone)
    RotationOffset=(X=0,Y=100,Z=-20)
   )
   > n14 = &RigUnit_SphericalPoseReader(
    ActiveRegionScaleFactors=(PositiveWidth=1,NegativeWidth=1,PositiveHeight=1,NegativeHeight=1)
    ActiveRegionSize=0.25
    Debug=(bDrawDebug=false,DebugScale=25,DebugSegments=20,DebugThickness=0.25)
    DriverAxis=(X=1,Y=0,Z=0)
    DriverItem=(Type=Bone,Name="thigh_r")
    FalloffRegionScaleFactors=(PositiveWidth=1,NegativeWidth=1,PositiveHeight=1,NegativeHeight=1)
    FalloffSize=0.2
    OptionalParentItem=(Type=Bone)
    RotationOffset=(X=0,Y=-80,Z=0)
   )
   > &RigUnit_ModifyTransforms(
    ItemToModify=((Item=(Type=Bone),Transform=(Rotation=(X=0,Y=0,Z=0,W=1),Translation=(X=0,Y=0,Z=0),Scale3D=(X=1,Y=1,Z=1))),(Item=(Type=Bone,Name="skirt_connector_01_l"),Transform=(Rotation=(X=0,Y=0,Z=0,W=1),Translation=(X=0,Y=0,Z=0),Scale3D=(X=1,Y=1,Z=1))))
    ItemToModify.1.Transform.Rotation=&RigVMFunction_MathQuaternionFromEuler(Euler=(X=0,Y=0,Z=-45)).Result
    Weight=@n13.OutputParam
   )
   > &RigUnit_ModifyTransforms(
    ItemToModify=((Item=(Type=Bone),Transform=(Rotation=(X=0,Y=0,Z=0,W=1),Translation=(X=0,Y=0,Z=0),Scale3D=(X=1,Y=1,Z=1))),(Item=(Type=Bone,Name="skirt_connector_01_r"),Transform=(Rotation=(X=0,Y=0,Z=0,W=1),Translation=(X=0,Y=0,Z=0),Scale3D=(X=1,Y=1,Z=1))))
    ItemToModify.1.Transform.Rotation=&RigVMFunction_MathQuaternionFromEuler(Euler=(X=0,Y=0,Z=-20)).Result
    Weight=@n14.OutputParam
   )
   n15 = &RigUnit_SequenceExecution(ExecuteContext=@Branch_1_RigVMUnitNode.True)
   n16 = &RigUnit_SphericalPoseReader(
    ActiveRegionScaleFactors=(PositiveWidth=0.5,NegativeWidth=0.5,PositiveHeight=0.5,NegativeHeight=0.5)
    ActiveRegionSize=0.33
    Debug=(bDrawDebug=True,DebugScale=25,DebugSegments=20,DebugThickness=0.25)
    DriverAxis=(X=1,Y=0,Z=0)
    DriverItem=(Type=Bone,Name="thigh_r")
    FalloffRegionScaleFactors=(PositiveWidth=1,NegativeWidth=1,PositiveHeight=1,NegativeHeight=1)
    FalloffSize=0.25
    OptionalParentItem=(Type=Bone)
    RotationOffset=(X=-45,Y=-15,Z=0)
    ExecutePin=@n15.A
   )
   > &RigUnit_ModifyTransforms(
    ItemToModify=((Item=(Type=Bone,Name="skirt_fr_base"),Transform=(Rotation=(X=0,Y=0,Z=0,W=1),Translation=(X=0,Y=0,Z=-3),Scale3D=(X=1,Y=1,Z=1))))
    ItemToModify.0.Transform.Rotation=&RigVMFunction_MathQuaternionFromEuler(Euler=(X=-30,Y=60,Z=-30), RotationOrder=ZXY).Result
    Weight=@n16.OutputParam
   )
   n17 = &RigUnit_GetRelativeTransformForItem(Child=(Type=Bone,Name="thigh_r"), Parent=(Type=Bone,Name="pelvis"))
   n18 = &RigVMFunction_MathFloatRemap(
    SourceMaximum=90
    SourceMinimum=180
    Value=&RigVMFunction_MathQuaternionToEuler(
     RotationOrder=ZXY
     Value=&RigUnit_VisualDebugQuatItemSpace(
      DebugDrawSettings=(DepthPriority=SDPG_Foreground,Lifetime=-1)
      Scale=10
      Space=(Type=Bone)
      bEnabled=True
      Value=@n17.RelativeTransform.Rotation
     ).Value
    ).Result.Z
   )
   > &RigUnit_ModifyTransforms(
    ItemToModify=((Item=(Type=Bone,Name="skirt_fr_base"),Transform=(Rotation=(X=0,Y=0,Z=0,W=1),Translation=(X=-5,Y=0,Z=0),Scale3D=(X=1,Y=1,Z=1))))
    Weight=@n18.Result
   )
   > &RigUnit_ModifyTransforms(
    ItemToModify=((Item=(Type=Bone,Name="skirt_connector_01_r"),Transform=(Rotation=(X=0,Y=0,Z=0,W=1),Translation=(X=-3,Y=0,Z=0),Scale3D=(X=1,Y=1,Z=1))))
    Weight=@n18.Result
    ItemToModify.0.Transform.Rotation=&RigVMFunction_MathQuaternionFromEuler(Euler=(X=30,Y=0,Z=-100)).Result
   )
   > &RigUnit_ModifyTransforms(
    ItemToModify=((Item=(Type=Bone,Name="skirt_connector_01_r"),Transform=(Rotation=(X=0,Y=0,Z=0,W=1),Translation=(X=0,Y=0,Z=0),Scale3D=(X=1,Y=1,Z=1))))
    Weight=1
    ItemToModify.0.Transform.Rotation=&RigVMFunction_MathQuaternionFromEuler(Euler=(X=0,Y=-30,Z=0)).Result
   )
   > &RigUnit_ModifyTransforms(
    ItemToModify=((Item=(Type=Bone,Name="skirt_fl_base"),Transform=(Rotation=(X=0,Y=0,Z=0,W=1),Translation=(X=2,Y=2,Z=0),Scale3D=(X=1,Y=1,Z=1))))
    Weight=1
    ItemToModify.0.Transform.Rotation=&RigVMFunction_MathQuaternionFromEuler(Euler=(X=-20,Y=0,Z=0), RotationOrder=XYZ).Result
   )
   n19 = &RigUnit_VisualDebugQuatItemSpace(
    DebugDrawSettings=(DepthPriority=SDPG_Foreground,Lifetime=-1)
    Scale=10
    Space=(Type=Bone)
    bEnabled=True
    Value=&RigUnit_GetRelativeTransformForItem(Child=(Type=Bone,Name="thigh_r"), Parent=(Type=Bone,Name="pelvis")).RelativeTransform.Rotation
   )
   > n20 = &RigUnit_ModifyTransforms(
    ItemToModify=((Item=(Type=Bone,Name="loincloth_fr_master"),Transform=(Rotation=(X=0,Y=0,Z=0,W=1),Translation=(X=0,Y=0,Z=0),Scale3D=(X=1,Y=1,Z=1))))
    Weight=&RigVMFunction_MathFloatRemap(SourceMaximum=90, SourceMinimum=0, Value=&RigVMFunction_MathQuaternionToEuler(RotationOrder=XZY, Value=@n19.Value).Result.Z).Result
    ItemToModify.0.Transform.Rotation=&RigVMFunction_MathQuaternionFromEuler(Euler=(X=0,Y=0,Z=90)).Result
   )
   Branch_3_RigVMUnitNode = &RigVMFunction_ControlFlowBranch(ExecuteContext=@n20.ExecutePin, Condition=RigVMVariableNode() {variable=Moving}.Value)
   RigUnit_AlphaInterp_1_3 = &RigVMFunction_AlphaInterp(
    InRange=(Max=1)
    InterpSpeedDecreasing=10
    InterpSpeedIncreasing=10
    OutRange=(Max=1)
    Scale=1
    Value=1
    Bias=~
    ClampMin=~
    bInterpResult=~
    bMapRange=~
   )
   &RigUnit_ModifyTransforms(
    ItemToModify=((Item=(Type=Bone,Name="loincloth_fl_master"),Transform=(Rotation=(X=0,Y=0,Z=0,W=1),Translation=(X=0,Y=0,Z=-4),Scale3D=(X=1,Y=1,Z=1))))
    ExecutePin=@Branch_3_RigVMUnitNode.False
    ItemToModify.0.Transform.Rotation=&RigVMFunction_MathQuaternionFromEuler(Euler=(X=0,Y=0,Z=30)).Result
    Weight=@RigUnit_AlphaInterp_1_3.Result
   )
   n21 = &RigUnit_VisualDebugQuatItemSpace(
    DebugDrawSettings=(DepthPriority=SDPG_Foreground,Lifetime=-1)
    Scale=10
    Space=(Type=Bone)
    bEnabled=True
    Value=&RigUnit_GetRelativeTransformForItem(Child=(Type=Bone,Name="thigh_l"), Parent=(Type=Bone,Name="pelvis")).RelativeTransform.Rotation
   )
   RigUnit_AlphaInterp_1 = &RigVMFunction_AlphaInterp(
    InRange=(Max=1)
    InterpSpeedDecreasing=10
    InterpSpeedIncreasing=10
    OutRange=(Max=1)
    Scale=1
    Bias=~
    ClampMin=~
    bInterpResult=~
    bMapRange=~
    Value=&RigVMFunction_MathFloatRemap(
     SourceMaximum=0
     SourceMinimum=-90
     TargetMaximum=0
     TargetMinimum=1
     Value=&RigVMFunction_MathQuaternionToEuler(RotationOrder=XZY, Value=@n21.Value).Result.Z
    ).Result
   )
   &RigUnit_ModifyTransforms(
    ItemToModify=((Item=(Type=Bone,Name="loincloth_fl_master"),Transform=(Rotation=(X=0,Y=0,Z=0,W=1),Translation=(X=0,Y=0,Z=0),Scale3D=(X=1,Y=1,Z=1))))
    ExecutePin=@Branch_3_RigVMUnitNode.True
    ItemToModify.0.Transform.Rotation=&RigVMFunction_MathQuaternionFromEuler(Euler=(X=-15,Y=0,Z=90)).Result
    Weight=@RigUnit_AlphaInterp_1.Result
   )
  ]}
]
hierarchy [
 root {loc=<0 0 0>, rot=<0 0 0 1>}
 attach {parent=root, loc=<0 0 0>, rot=<0 0 0 1>}
 pelvis {parent=root, loc=<0 -0.5301 98.0571>, rot=<0.0127 0.707 -0.0127 -0.707>}
 spine_01 {parent=pelvis, loc=<6.1208 -0.9328 0>, rot=<0 0 0.0576 -0.9983>}
 spine_02 {parent=spine_01, loc=<6.6022 -0.0153 0>, rot=<0 0 0.0022 -1>}
 spine_03 {parent=spine_02, loc=<6.4919 -0.0105 0>, rot=<0 0 0.0015 -1>}
 spine_04 {parent=spine_03, loc=<6.9596 -0.007 0>, rot=<0 0 0.0009 -1>}
 spine_05 {parent=spine_04, loc=<6.9981 -0.0017 0>, rot=<0 0 -0.0042 -1>}
 clavicle_l {parent=spine_05, loc=<5.4831 -0.1053 -3.2946>, rot=<0.0938 -0.7388 0.0307 -0.6666>}
 upperarm_l {parent=clavicle_l, loc=<9.9486 0 0>, rot=<0.0371 -0.2992 -0.0705 -0.9509>}
 lowerarm_l {parent=upperarm_l, loc=<26.4793 0 0>, rot=<0 0 -0.2382 -0.9712>}
 hand_l {parent=lowerarm_l, loc=<22.7458 0 0>, rot=<-0.6747 -0.0182 0.0605 -0.7354>}
 weapon_l {parent=hand_l, loc=<-0.1974 -3.4111 0.3811>, rot=<0 0 0 1>}
 thumb_01_l {parent=hand_l, loc=<1.3118 -0.3191 -2.459>, rot=<0.6546 -0.4042 0.0664 -0.6354>}
 thumb_02_l {parent=thumb_01_l, loc=<5.9322 -0.0253 0.1416>, rot=<0.0121 0.0046 0.2964 -0.955>}
 thumb_03_l {parent=thumb_02_l, loc=<2.2223 0.3682 0>, rot=<0 0 0.0172 -0.9999>}
 index_metacarpal_l {parent=hand_l, loc=<1.6535 0.0183 -2.96>, rot=<0.0158 -0.0124 -0.0173 -0.9996>}
 index_01_l {parent=index_metacarpal_l, loc=<8.3371 0.1718 -0.2461>, rot=<0.0363 0.0499 0.142 -0.9879>}
 index_02_l {parent=index_01_l, loc=<3.4919 -0.3004 0>, rot=<0 0 0.2948 -0.9556>}
 index_03_l {parent=index_02_l, loc=<2.5578 -0.123 0>, rot=<0 0 0.1968 -0.9804>}
 middle_metacarpal_l {parent=hand_l, loc=<2.1856 -0.1725 -0.5984>, rot=<0.006 0.0168 -0.0329 -0.9993>}
 middle_01_l {parent=middle_metacarpal_l, loc=<7.6374 0.346 -0.1508>, rot=<-0.0085 0.0328 0.1813 -0.9829>}
 middle_02_l {parent=middle_01_l, loc=<3.9108 0 0>, rot=<0 0 0.2808 -0.9598>}
 middle_03_l {parent=middle_02_l, loc=<2.6103 0.0517 0>, rot=<0 0 0.1446 -0.9895>}
 ring_metacarpal_l {parent=hand_l, loc=<2.2074 -0.3625 1.3638>, rot=<-0.0284 0.0722 -0.0458 -0.9959>}
 ring_01_l {parent=ring_metacarpal_l, loc=<7.2483 0.0329 -0.1426>, rot=<-0.0456 -0.0277 0.2951 -0.954>}
 ring_02_l {parent=ring_01_l, loc=<3.7208 0.4647 0>, rot=<0 0 0.1583 -0.9874>}
 ring_03_l {parent=ring_02_l, loc=<2.4152 0.114 0>, rot=<0 0 0.1504 -0.9886>}
 pinky_metacarpal_l {parent=hand_l, loc=<1.7616 -0.9753 3.3814>, rot=<-0.0397 0.0814 -0.069 -0.9935>}
 pinky_01_l {parent=pinky_metacarpal_l, loc=<7.1933 -0.4205 -0.2601>, rot=<-0.0487 -0.0063 0.2556 -0.9655>}
 pinky_02_l {parent=pinky_01_l, loc=<2.8274 0.0893 0>, rot=<0 0 0.2474 -0.9689>}
 pinky_03_l {parent=pinky_02_l, loc=<2.2046 0.0307 0>, rot=<0 0 0.125 -0.9922>}
 hand_attach_l {parent=hand_l, loc=<8.5 -3 0>, rot=<0.7071 0.7071 0 0>}
 lowerarm_twist_02_l {parent=lowerarm_l, loc=<9.6199 0 0>, rot=<0 0 0 1>}
 deform_lowerarm_plates_l {parent=lowerarm_twist_02_l, loc=<0 0 0>, rot=<0 0 0 1>}
 lowerarm_twist_01_l {parent=lowerarm_l, loc=<17.8549 0 0>, rot=<0 0 0 1>}
 deform_elbow_in_base_l {parent=lowerarm_l, loc=<0.1385 3.6052 -0.2182>, rot=<0.0059 0.0117 -0.0071 -0.9999>}
 deform_elbow_in_l {parent=deform_elbow_in_base_l, loc=<0 0 0>, rot=<0 0 0 1>}
 deform_elbow_out_base_l {parent=lowerarm_l, loc=<0.0565 -3.9993 0.1259>, rot=<0.0059 0.0117 -0.0071 -0.9999>}
 deform_elbow_out_l {parent=deform_elbow_out_base_l, loc=<0 0 0>, rot=<0 0 0 1>}
 upperarm_twist_01_l {parent=upperarm_l, loc=<0.0041 0 0>, rot=<0 0 0 1>}
 deform_upperarm_bk_base_l {parent=upperarm_twist_01_l, loc=<0.9473 -4.46 0.3113>, rot=<0.0004 -0.0034 -0.0158 -0.9999>}
 deform_upperarm_bk_l {parent=deform_upperarm_bk_base_l, loc=<0 0 0>, rot=<0 0 0 1>}
 deform_upperarm_fr_base_l {parent=upperarm_twist_01_l, loc=<1.4228 5.0263 0.3212>, rot=<0.0004 -0.0034 -0.0158 -0.9999>}
 deform_upperarm_fr_l {parent=deform_upperarm_fr_base_l, loc=<0 0 0>, rot=<0 0 0 1>}
 deform_upperarm_up_base_l {parent=upperarm_twist_01_l, loc=<-1.6653 0.3195 3.9629>, rot=<0.0004 -0.0034 -0.0158 -0.9999>}
 deform_upperarm_up_l {parent=deform_upperarm_up_base_l, loc=<0 0 0>, rot=<0 0 0 1>}
 upperarm_twist_02_l {parent=upperarm_l, loc=<13.7941 0 0>, rot=<0 0 0 1>}
 clavicle_cape_l {parent=clavicle_l, loc=<0 0 0>, rot=<-0.004 0.0209 -0.061 -0.9979>}
 cape_01_l {parent=clavicle_cape_l, loc=<7.3041 -4.7499 5.7122>, rot=<0.3413 -0.5667 0.2993 -0.6876>}
 cape_02_l {parent=cape_01_l, loc=<6.892 0 0>, rot=<0 0 -0.3403 -0.9403>}
 cape_03_l {parent=cape_02_l, loc=<11.8886 0 0>, rot=<0 0 -0.1042 -0.9946>}
 cape_04_l {parent=cape_03_l, loc=<23.3505 0 0>, rot=<0 0 0.0042 -1>}
 cape_05_l {parent=cape_04_l, loc=<20.0893 0 0>, rot=<0 0 -0.008 -1>}
 cape_06_l {parent=cape_05_l, loc=<18.5056 0 0>, rot=<0 0 -0.0041 -1>}
 cape_07_l {parent=cape_06_l, loc=<13.5867 0 0>, rot=<0 0 0.0015 -1>}
 cape_08_l {parent=cape_07_l, loc=<15.1001 0 0>, rot=<0 0 0 1>}
 neck_01 {parent=spine_05, loc=<11.3138 0.1529 0>, rot=<0 0 -0.2003 -0.9797>}
 neck_02 {parent=neck_01, loc=<4.5627 0 0>, rot=<0 0 0.0504 -0.9987>}
 head {parent=neck_02, loc=<5.3914 0 0>, rot=<0 0 0.115 -0.9934>}
 clavicle_r {parent=spine_05, loc=<5.4831 -0.1053 3.2946>, rot=<0.7388 0.0938 -0.6666 -0.0307>}
 upperarm_r {parent=clavicle_r, loc=<-9.9486 0 0>, rot=<0.0371 -0.2992 -0.0705 -0.9509>}
 lowerarm_r {parent=upperarm_r, loc=<-26.4793 0 0>, rot=<0 0 -0.2382 -0.9712>}
 hand_r {parent=lowerarm_r, loc=<-22.7458 0 0>, rot=<-0.6747 -0.0182 0.0605 -0.7354>}
 weapon_r {parent=hand_r, loc=<0.1974 3.4111 -0.3811>, rot=<0 0 0 1>}
 thumb_01_r {parent=hand_r, loc=<-1.3118 0.3191 2.459>, rot=<0.6546 -0.4042 0.0664 -0.6354>}
 thumb_02_r {parent=thumb_01_r, loc=<-5.9322 0.0253 -0.1416>, rot=<0.0121 0.0046 0.2964 -0.955>}
 thumb_03_r {parent=thumb_02_r, loc=<-2.2223 -0.3682 0>, rot=<0 0 0.0172 -0.9999>}
 index_metacarpal_r {parent=hand_r, loc=<-1.6535 -0.0183 2.96>, rot=<0.0158 -0.0124 -0.0173 -0.9996>}
 index_01_r {parent=index_metacarpal_r, loc=<-8.3371 -0.1718 0.2461>, rot=<0.0363 0.0499 0.142 -0.9879>}
 index_02_r {parent=index_01_r, loc=<-3.4919 0.3004 0>, rot=<0 0 0.2948 -0.9556>}
 index_03_r {parent=index_02_r, loc=<-2.5578 0.123 0>, rot=<0 0 0.1968 -0.9804>}
 middle_metacarpal_r {parent=hand_r, loc=<-2.1856 0.1725 0.5984>, rot=<0.006 0.0168 -0.0329 -0.9993>}
 middle_01_r {parent=middle_metacarpal_r, loc=<-7.6374 -0.346 0.1508>, rot=<-0.0085 0.0328 0.1813 -0.9829>}
 middle_02_r {parent=middle_01_r, loc=<-3.9108 0 0>, rot=<0 0 0.2808 -0.9598>}
 middle_03_r {parent=middle_02_r, loc=<-2.6103 -0.0517 0>, rot=<0 0 0.1446 -0.9895>}
 ring_metacarpal_r {parent=hand_r, loc=<-2.2074 0.3625 -1.3638>, rot=<-0.0284 0.0722 -0.0458 -0.9959>}
 ring_01_r {parent=ring_metacarpal_r, loc=<-7.2483 -0.0329 0.1426>, rot=<-0.0456 -0.0277 0.2951 -0.954>}
 ring_02_r {parent=ring_01_r, loc=<-3.7208 -0.4647 0>, rot=<0 0 0.1583 -0.9874>}
 ring_03_r {parent=ring_02_r, loc=<-2.4152 -0.114 0>, rot=<0 0 0.1504 -0.9886>}
 pinky_metacarpal_r {parent=hand_r, loc=<-1.7616 0.9753 -3.3814>, rot=<-0.0397 0.0814 -0.069 -0.9935>}
 pinky_01_r {parent=pinky_metacarpal_r, loc=<-7.1933 0.4205 0.2601>, rot=<-0.0487 -0.0063 0.2556 -0.9655>}
 pinky_02_r {parent=pinky_01_r, loc=<-2.8274 -0.0893 0>, rot=<0 0 0.2474 -0.9689>}
 pinky_03_r {parent=pinky_02_r, loc=<-2.2046 -0.0307 0>, rot=<0 0 0.125 -0.9922>}
 hand_attach_r {parent=hand_r, loc=<-8.5001 2.9999 0>, rot=<0 0 -0.7071 -0.7071>}
 lowerarm_twist_02_r {parent=lowerarm_r, loc=<-9.6199 0 0>, rot=<0 0 0 -1>}
 deform_lowerarm_plates_r {parent=lowerarm_twist_02_r, loc=<0 0 0>, rot=<0 0 0 1>}
 lowerarm_twist_01_r {parent=lowerarm_r, loc=<-17.8549 0 0>, rot=<0 0 0 -1>}
 deform_elbow_in_base_r {parent=lowerarm_r, loc=<-0.1385 -3.6052 0.2182>, rot=<0.0059 0.0117 -0.0071 -0.9999>}
 deform_elbow_in_r {parent=deform_elbow_in_base_r, loc=<0 0 0>, rot=<0 0 0 1>}
 deform_elbow_out_base_r {parent=lowerarm_r, loc=<-0.0565 3.9993 -0.126>, rot=<0.0059 0.0117 -0.0071 -0.9999>}
 deform_elbow_out_r {parent=deform_elbow_out_base_r, loc=<0 0 0>, rot=<0 0 0 1>}
 upperarm_twist_01_r {parent=upperarm_r, loc=<-0.0041 0 0>, rot=<0 0 0 -1>}
 deform_upperarm_bk_base_r {parent=upperarm_twist_01_r, loc=<-0.9473 4.46 -0.3113>, rot=<0.0004 -0.0034 -0.0158 -0.9999>}
 deform_upperarm_bk_r {parent=deform_upperarm_bk_base_r, loc=<0 0 0>, rot=<0 0 0 1>}
 deform_upperarm_fr_base_r {parent=upperarm_twist_01_r, loc=<-1.4228 -5.0263 -0.3212>, rot=<0.0004 -0.0034 -0.0158 -0.9999>}
 deform_upperarm_fr_r {parent=deform_upperarm_fr_base_r, loc=<0 0 0>, rot=<0 0 0 1>}
 deform_upperarm_up_base_r {parent=upperarm_twist_01_r, loc=<1.6653 -0.3195 -3.9628>, rot=<0.0004 -0.0034 -0.0158 -0.9999>}
 deform_upperarm_up_r {parent=deform_upperarm_up_base_r, loc=<0 0 0>, rot=<0 0 0 1>}
 upperarm_twist_02_r {parent=upperarm_r, loc=<-13.7941 0 0>, rot=<0 0 0 -1>}
 clavicle_cape_r {parent=clavicle_r, loc=<0 0 0>, rot=<-0.004 0.0209 -0.061 -0.9979>}
 cape_01_r {parent=clavicle_cape_r, loc=<-7.3041 4.7499 -5.712>, rot=<0.5667 0.3413 0.6876 0.2993>}
 cape_02_r {parent=cape_01_r, loc=<6.8922 -0.0002 0>, rot=<0 0 -0.3403 -0.9403>}
 cape_03_r {parent=cape_02_r, loc=<11.8886 0 0>, rot=<0 0 -0.1042 -0.9946>}
 cape_04_r {parent=cape_03_r, loc=<23.3501 0.0001 0>, rot=<0 0 0.0042 -1>}
 cape_05_r {parent=cape_04_r, loc=<20.0898 0 0.0001>, rot=<0 0 -0.008 -1>}
 cape_06_r {parent=cape_05_r, loc=<18.5056 -0.0001 -0.0001>, rot=<0 0 -0.0041 -1>}
 cape_07_r {parent=cape_06_r, loc=<13.5867 0 0.0001>, rot=<0 0 0.0015 -1>}
 cape_08_r {parent=cape_07_r, loc=<15.1001 0.0001 -0.0001>, rot=<0 0 0 -1>}
 cape_root {parent=spine_05, loc=<9.0735 -9.9874 0>, rot=<-0.0087 -0.7071 0.0087 -0.7071>}
 cape_mid_01_l {parent=cape_root, loc=<5 -0.2606 -0.4394>, rot=<0.1329 -0.6858 0.1296 -0.7037>}
 cape_mid_02_l {parent=cape_mid_01_l, loc=<12.8326 0 0>, rot=<0 0 -0.0975 -0.9952>}
 cape_mid_03_l {parent=cape_mid_02_l, loc=<23.0451 0 0>, rot=<0 0 -0.0173 -0.9998>}
 cape_mid_04_l {parent=cape_mid_03_l, loc=<20.3244 0 0>, rot=<0 0 -0.0139 -0.9999>}
 cape_mid_05_l {parent=cape_mid_04_l, loc=<18.7251 0 0>, rot=<0 0 0.0005 -1>}
 cape_mid_06_l {parent=cape_mid_05_l, loc=<13.6769 0 0>, rot=<0 0 0.0016 -1>}
 cape_mid_07_l {parent=cape_mid_06_l, loc=<14.7332 0 0>, rot=<0 0 0 1>}
 cape_mid_01_r {parent=cape_root, loc=<-5 -0.2606 -0.4389>, rot=<0.1296 -0.7037 0.1329 -0.6858>}
 cape_mid_02_r {parent=cape_mid_01_r, loc=<12.8333 -0.0003 0>, rot=<0 0 -0.0975 -0.9952>}
 cape_mid_03_r {parent=cape_mid_02_r, loc=<23.0446 0 0>, rot=<0 0 -0.0173 -0.9998>}
 cape_mid_04_r {parent=cape_mid_03_r, loc=<20.3247 0 0>, rot=<0 0 -0.0139 -0.9999>}
 cape_mid_05_r {parent=cape_mid_04_r, loc=<18.7251 0 0>, rot=<0 0 0.0005 -1>}
 cape_mid_06_r {parent=cape_mid_05_r, loc=<13.6769 -0.0001 0>, rot=<0 0 0.0016 -1>}
 cape_mid_07_r {parent=cape_mid_06_r, loc=<14.7332 0 0>, rot=<0 0 0 1>}
 deform_pec_aim_base_l {parent=spine_04, loc=<3.1942 9.3086 -1.2237>, rot=<0.0614 -0.4031 0.3295 -0.8516>}
 deform_pec_aim_l {parent=deform_pec_aim_base_l, loc=<0 0 0>, rot=<0 0 0 1>}
 deform_pec_base_l {parent=deform_pec_aim_l, loc=<9.3665 0 0>, rot=<0 0 0 1>}
 deform_pec_l {parent=deform_pec_base_l, loc=<0 0 0>, rot=<0 0 0 1>}
 deform_pec_aim_base_r {parent=spine_04, loc=<3.1938 9.3085 1.2237>, rot=<-0.4031 -0.0614 0.8516 0.3295>}
 deform_pec_aim_r {parent=deform_pec_aim_base_r, loc=<0 0 0>, rot=<0 0 0 1>}
 deform_pec_base_r {parent=deform_pec_aim_r, loc=<-9.3667 -0.0003 -0.0004>, rot=<0 0 0 1>}
 deform_pec_r {parent=deform_pec_base_r, loc=<0 0 0>, rot=<0 0 0 1>}
 thigh_l {parent=pelvis, loc=<-1.0855 -0.2109 -8.7053>, rot=<0.0009 0.0345 0.0155 -0.9993>}
 calf_l {parent=thigh_l, loc=<-42.5268 0 0>, rot=<0 0 -0.0657 -0.9978>}
 foot_l {parent=calf_l, loc=<-46.8037 0 0>, rot=<0.0422 -0.04 0.0613 -0.9964>}
 ball_l {parent=foot_l, loc=<-3.5658 10.8066 -0.1544>, rot=<0.0001 -0.0001 -0.7063 -0.7079>}
 calf_twist_02_l {parent=calf_l, loc=<-18.8064 -0.0967 0.0397>, rot=<0.0237 0.0095 -0.0035 -0.9997>}
 calf_twist_01_l {parent=calf_l, loc=<-35.813 -0.0967 0.0397>, rot=<0.0237 0.0095 -0.0035 -0.9997>}
 thigh_twist_01_l {parent=thigh_l, loc=<-21.629 -0.0919 -0.0002>, rot=<0.0013 0 0 -1>}
 thigh_r {parent=pelvis, loc=<-1.0855 -0.2109 8.7053>, rot=<0.0345 -0.0009 0.9993 0.0155>}
 calf_r {parent=thigh_r, loc=<42.5268 0 0>, rot=<0 0 -0.0657 -0.9978>}
 foot_r {parent=calf_r, loc=<46.8037 0 0>, rot=<0.0422 -0.04 0.0613 -0.9964>}
 ball_r {parent=foot_r, loc=<3.5658 -10.8066 0.1544>, rot=<0.0001 -0.0001 -0.7063 -0.7079>}
 calf_twist_02_r {parent=calf_r, loc=<18.8064 0.0967 -0.0397>, rot=<0.0237 0.0095 -0.0035 -0.9997>}
 calf_twist_01_r {parent=calf_r, loc=<35.813 0.0967 -0.0397>, rot=<0.0237 0.0095 -0.0035 -0.9997>}
 thigh_twist_01_r {parent=thigh_r, loc=<21.629 0.0919 0.0002>, rot=<0.0013 0 0 -1>}
 loincloth_fr_master {parent=pelvis, loc=<6.5494 7.479 4.3068>, rot=<-0.0466 -0.9986 -0.0002 0.0253>}
 loincloth_fr_01 {parent=loincloth_fr_master, loc=<0 0 0>, rot=<0 0 0 1>}
 loincloth_fr_02 {parent=loincloth_fr_01, loc=<7.1094 0.0014 -0.0032>, rot=<-0.0002 -0.0034 -0.0015 -1>}
 loincloth_fr_03 {parent=loincloth_fr_02, loc=<5.3642 0.0073 -0.0024>, rot=<0.0002 0.004 0.0123 -0.9999>}
 loincloth_fr_04 {parent=loincloth_fr_03, loc=<6.4831 -0.0052 0.0013>, rot=<0.0001 0.0022 0.009 -1>}
 loincloth_fr_05 {parent=loincloth_fr_04, loc=<6.9538 0 0>, rot=<0 0 0 1>}
 loincloth_fr_06 {parent=loincloth_fr_05, loc=<6.924 0 0>, rot=<0 0 0 1>}
 loincloth_fl_master {parent=pelvis, loc=<6.5494 7.479 -4.3068>, rot=<-0.0466 -0.9986 0.0002 -0.0253>}
 loincloth_fl_01 {parent=loincloth_fl_master, loc=<0 0 0>, rot=<0 0 0 1>}
 loincloth_fl_02 {parent=loincloth_fl_01, loc=<7.1094 0.0014 0.0032>, rot=<0.0002 0.0034 -0.0015 -1>}
 loincloth_fl_03 {parent=loincloth_fl_02, loc=<5.3641 0.0073 0.0024>, rot=<-0.0002 -0.004 0.0123 -0.9999>}
 loincloth_fl_04 {parent=loincloth_fl_03, loc=<6.4831 -0.0052 -0.0013>, rot=<-0.0001 -0.0022 0.009 -1>}
 loincloth_fl_05 {parent=loincloth_fl_04, loc=<6.9538 0 0>, rot=<0 0 0 1>}
 loincloth_fl_06 {parent=loincloth_fl_05, loc=<6.924 0 0>, rot=<0 0 0 1>}
 skirt_fl_base {parent=pelvis, loc=<5.3951 4.7668 -10.8218>, rot=<0.9569 0.055 0.1635 -0.2337>}
 skirt_fl_01 {parent=skirt_fl_base, loc=<-3.0058 0 0>, rot=<-0.0536 0 0 -0.9986>}
 skirt_fl_02 {parent=skirt_fl_01, loc=<-11.7826 0 0.0001>, rot=<0.0397 0.0586 -0.0165 -0.9974>}
 skirt_fl_03 {parent=skirt_fl_02, loc=<-9.5195 0 0>, rot=<0.0536 0.044 0.0173 -0.9974>}
 skirt_fl_04 {parent=skirt_fl_03, loc=<-7.0252 0 0>, rot=<-0.0129 0.071 -0.0011 -0.9974>}
 skirt_fl_05 {parent=skirt_fl_04, loc=<-7.0686 0 0>, rot=<0.01 0 0 -0.9999>}
 skirt_bl_base {parent=pelvis, loc=<0.6545 -7.2504 -13.7078>, rot=<-0.0087 -0.9185 0.3859 -0.0859>}
 skirt_bl_01 {parent=skirt_bl_base, loc=<0 0 0>, rot=<0 0 0 1>}
 skirt_bl_02 {parent=skirt_bl_01, loc=<9.6167 0 0>, rot=<0.001 -0.0593 -0.0162 -0.9981>}
 skirt_bl_03 {parent=skirt_bl_02, loc=<9.7339 0 0>, rot=<-0.0005 -0.0164 0.0073 -0.9998>}
 skirt_bl_04 {parent=skirt_bl_03, loc=<7.0574 0 0>, rot=<-0.0009 -0.0232 0.0027 -0.9997>}
 skirt_bl_05 {parent=skirt_bl_04, loc=<6.6879 0 0>, rot=<0 0 0 1>}
 skirt_back_a_base {parent=pelvis, loc=<2.6864 -11.6905 -3.1214>, rot=<0.1447 -0.9891 0.005 -0.0271>}
 skirt_back_a_01 {parent=skirt_back_a_base, loc=<4.3989 0 0>, rot=<0.0038 -0.0212 -0.1496 -0.9885>}
 skirt_back_a_02 {parent=skirt_back_a_01, loc=<9.7519 0 0>, rot=<0.0001 0.0017 -0.0621 -0.9981>}
 skirt_back_a_03 {parent=skirt_back_a_02, loc=<9.1371 0 0>, rot=<-0.0002 -0.0023 0.0239 -0.9997>}
 skirt_back_a_04 {parent=skirt_back_a_03, loc=<10.0436 0 0>, rot=<0 -0.0001 0.0351 -0.9994>}
 skirt_back_a_05 {parent=skirt_back_a_04, loc=<5.6385 0 0>, rot=<0 0 0 1>}
 skirt_back_b_base {parent=pelvis, loc=<2.8318 -11.6469 3.1135>, rot=<0.1447 -0.9891 -0.005 0.0271>}
 skirt_back_b_01 {parent=skirt_back_b_base, loc=<4.5509 0 0>, rot=<-0.0038 0.0212 -0.1496 -0.9885>}
 skirt_back_b_02 {parent=skirt_back_b_01, loc=<9.7519 0 0>, rot=<-0.0001 -0.0017 -0.0621 -0.9981>}
 skirt_back_b_03 {parent=skirt_back_b_02, loc=<9.1371 0 0>, rot=<0.0002 0.0023 0.0239 -0.9997>}
 skirt_back_b_04 {parent=skirt_back_b_03, loc=<10.0436 0 0>, rot=<0 0.0001 0.0351 -0.9994>}
 skirt_back_b_05 {parent=skirt_back_b_04, loc=<5.6385 0 0>, rot=<0 0 0 1>}
 skirt_fr_base {parent=pelvis, loc=<5.395 4.7668 10.8218>, rot=<0.055 -0.9569 0.2337 0.1635>}
 skirt_fr_01 {parent=skirt_fr_base, loc=<2.807 -0.6531 -0.1555>, rot=<-0.0536 0 0 -0.9986>}
 skirt_fr_02 {parent=skirt_fr_01, loc=<11.9813 0 0>, rot=<0.0397 0.0586 -0.0165 -0.9974>}
 skirt_fr_03 {parent=skirt_fr_02, loc=<9.5195 0 0>, rot=<0.0536 0.044 0.0173 -0.9974>}
 skirt_fr_04 {parent=skirt_fr_03, loc=<7.0253 0 0>, rot=<-0.0129 0.071 -0.0011 -0.9974>}
 skirt_fr_05 {parent=skirt_fr_04, loc=<7.0686 0 0>, rot=<0.01 0 0 -0.9999>}
 skirt_br_base {parent=pelvis, loc=<0.6545 -7.2504 13.7078>, rot=<0.0035 -0.9636 -0.2528 0.0863>}
 skirt_br_01 {parent=skirt_br_base, loc=<0 0 0>, rot=<0 0 0 1>}
 skirt_br_02 {parent=skirt_br_01, loc=<9.6166 0 0>, rot=<-0.001 0.0524 -0.032 -0.9981>}
 skirt_br_03 {parent=skirt_br_02, loc=<9.7339 0 0>, rot=<0.0005 0.0178 0.0025 -0.9998>}
 skirt_br_04 {parent=skirt_br_03, loc=<7.0575 0 0>, rot=<0.0009 0.023 -0.0039 -0.9997>}
 skirt_br_05 {parent=skirt_br_04, loc=<6.6878 0 0>, rot=<0.9901 0 0 0.1407>}
 skirt_connector_01_r {parent=pelvis, loc=<4.7022 6.8768 8.5453>, rot=<0.1297 -0.2648 0.9553 0.0206>}
 skirt_connector_02_r {parent=skirt_connector_01_r, loc=<5.1694 0 0>, rot=<0 0 0 1>}
 skirt_connector_03_r {parent=skirt_connector_02_r, loc=<3.4 0 0>, rot=<0 0 0 1>}
 skirt_connector_01_l {parent=pelvis, loc=<4.7204 6.8774 -8.5404>, rot=<-0.1296 0.2692 0.9541 0.0212>}
 skirt_connector_02_l {parent=skirt_connector_01_l, loc=<5.2679 0 0>, rot=<0 0 0 1>}
 skirt_connector_03_l {parent=skirt_connector_02_l, loc=<3.4 0 0>, rot=<0 0 0 1>}
 deform_lower_disk {parent=pelvis, loc=<2.4664 10.3786 0>, rot=<-0.0127 -0.707 0.0127 -0.707>}
 ik_foot_root {parent=root, loc=<0 0 0>, rot=<0 0 0 1>}
 ik_foot_l {parent=ik_foot_root, loc=<14.8262 -7.3545 8.296>, rot=<0.0357 0.7045 0.0284 -0.7082>}
 ik_foot_r {parent=ik_foot_root, loc=<-14.8262 -7.3545 8.296>, rot=<0.7082 0.0284 -0.7045 0.0357>}
 ik_hand_root {parent=root, loc=<0 0 0>, rot=<0 0 0 1>}
 ik_hand_gun {parent=ik_hand_root, loc=<-48.1743 3.3545 104.4381>, rot=<0.6276 0.1309 0.4183 -0.6434>}
 ik_hand_r {parent=ik_hand_gun, loc=<0 0 0>, rot=<0 0 0 1>}
 ik_hand_l {parent=ik_hand_gun, loc=<59.3271 67.6997 34.3524>, rot=<-0.6158 -0.7027 -0.3565 0>}
]
variables [
 Crouching {type=bool, default=False}
 Emoting {type=bool, default=False}
 Jumping {type=bool, default=False}
 Moving {type=bool, default=False}
]
