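# Physics asset that gives the skirt and front loincloth panels of the F_MED_Parfait female character body their collision and joint setup.
# Capsule and box chains use limited swing and twist, with root bodies free to rotate on the pelvis.
# Pelvis and thigh capsules supply leg and hip shapes, and two-box connector chains attach to each thigh.
# Rotation-free constraints link the hem ends of front-to-back side panels and both back chains.
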PhysicsAsset /BRCosmetics/Characters/Player/Female/Medium/Bodies/F_MED_Parfait/Meshes/F_MED_Parfait_Skirt_Physics
bodies {
 loincloth_fl_01 [{shapeType=Capsule, length=10, radius=2.5, shapeName=loincloth_fl_01_capsule, center=<3 0.0848 -2.8645>}]
 loincloth_fl_02 [{shapeType=Capsule, length=12, radius=2.5, shapeName=loincloth_fr_02_capsule, center=<3 0.6632 -2.8645>}]
 loincloth_fl_03 [{shapeType=Capsule, length=12, radius=3, shapeName=loincloth_fr_02_capsule, center=<3 0.0745 -2.2746>}]
 loincloth_fl_04 [
  {shapeType=Capsule, length=12, radius=2.5, shapeName=loincloth_fr_02_capsule, center=<3 0.0007 -1.0594>}
  {shapeType=Capsule, length=4, radius=3, shapeName=loincloth_fl_04_capsule, center=<4.4853 -0.672 3.7011>, rotation=<rot 90 0 0>}
  {shapeType=Capsule, length=4, radius=3, shapeName=loincloth_fl_04_capsule, center=<4.5798 0 -2.5>, rotation=<rot 90 0 0>}
 ]
 loincloth_fl_05 [
  {shapeType=Capsule, length=12, radius=2.5, shapeName=loincloth_fr_02_capsule, center=<2.5 -0.1614 -1.9388>}
  {shapeType=Capsule, length=4, radius=3, shapeName=loincloth_fl_05_capsule, center=<7.7579 -1.1134 3.6796>, rotation=<rot 90 0 0>}
  {shapeType=Capsule, length=4, radius=3, shapeName=loincloth_fl_05_capsule, center=<6.9331 0 -2.8659>, rotation=<rot 90 0 0>}
 ]
 loincloth_fl_master [{shapeType=Capsule, length=1, radius=1, shapeName=loincloth_fl_master_capsule}]
 loincloth_fr_01 [{shapeType=Capsule, length=10, radius=2.5, shapeName=loincloth_fr_01_capsule, center=<3 0.0848 2.3723>}]
 loincloth_fr_02 [{shapeType=Capsule, length=12, radius=2.5, shapeName=loincloth_fr_02_capsule, center=<3 0.6632 2.3723>}]
 loincloth_fr_03 [{shapeType=Capsule, length=12, radius=3, shapeName=loincloth_fr_03_capsule, center=<3 0.1213 2.0321>}]
 loincloth_fr_04 [
  {shapeType=Capsule, length=12, radius=2.5, shapeName=loincloth_fr_03_capsule, center=<3 0.042 0.977>}
  {shapeType=Capsule, length=4, radius=3, shapeName=loincloth_fr_04_capsule, center=<4.4853 -0.672 -3.7011>, rotation=<rot 90 0 0>}
  {shapeType=Capsule, length=4, radius=3, shapeName=loincloth_fr_04_capsule, center=<4.5798 0 2.5>, rotation=<rot 90 0 0>}
 ]
 loincloth_fr_05 [
  {shapeType=Capsule, length=12, radius=2.5, shapeName=loincloth_fr_03_capsule, center=<2.5 -0.145 1.295>}
  {shapeType=Capsule, length=4, radius=3, shapeName=loincloth_fr_05_capsule, center=<7.7579 -1.1134 -3.7011>, rotation=<rot 90 0 0>}
  {shapeType=Capsule, length=4, radius=3, shapeName=loincloth_fr_05_capsule, center=<6.9331 0 2.5552>, rotation=<rot 90 0 0>}
 ]
 loincloth_fr_master [{shapeType=Capsule, length=1, radius=1, shapeName=loincloth_fr_master_capsule}]
 pelvis [{
   shapeType=Capsule
   length=15.1534
   radius=6.5145
   shapeName=pelvis_capsule
   center=<0.4119 -1.5176 -0.0401>
   rotation=<rot -0.8022 -90 -0.0752>
  }]
 skirt_back_a_01 [
  {shapeType=Capsule, length=6, radius=3, shapeName=skirt_back_a_01_capsule, center=<3 0 0>}
  {shapeType=Capsule, length=3, radius=2.75, shapeName=skirt_back_a_01_capsule, center=<6 0.5 4>, rotation=<rot 90 0 0>}
 ]
 skirt_back_a_02 [{shapeType=Capsule, length=6, radius=3, shapeName=skirt_back_a_02_capsule, center=<3 0 0>}]
 skirt_back_a_03 [{shapeType=Capsule, length=6, radius=3, shapeName=skirt_back_a_03_capsule, center=<3 0 0>}]
 skirt_back_a_04 [{shapeType=Capsule, length=7.5, radius=3, shapeName=skirt_back_a_04_capsule, center=<3 0 0.5>}]
 skirt_back_a_base [{shapeType=Capsule, length=1, radius=1, shapeName=skirt_back_a_base_capsule}]
 skirt_back_b_01 [
  {shapeType=Capsule, length=6, radius=3, shapeName=skirt_back_b_01_capsule, center=<3 0 0>}
  {shapeType=Capsule, length=3, radius=2.75, shapeName=skirt_back_b_01_capsule, center=<6 0.5 -4>, rotation=<rot 90 0 0>}
 ]
 skirt_back_b_02 [{shapeType=Capsule, length=6, radius=3, shapeName=skirt_back_b_02_capsule, center=<3 0 0>}]
 skirt_back_b_03 [{shapeType=Capsule, length=6, radius=3, shapeName=skirt_back_b_03_capsule, center=<3 0 0>}]
 skirt_back_b_04 [{shapeType=Capsule, length=6, radius=3, shapeName=skirt_back_b_04_capsule, center=<3 0 0>}]
 skirt_back_b_base [{shapeType=Capsule, length=1, radius=1, shapeName=skirt_back_b_base_capsule}]
 skirt_bl_01 [{shapeType=Capsule, length=6, radius=3, shapeName=skirt_bl_01_capsule, center=<3 0 0>, rotation=<rot 0 0 90>}]
 skirt_bl_02 [{shapeType=Capsule, length=16.25, radius=3.75, shapeName=skirt_bl_02_capsule, center=<3 0 0>, rotation=<rot 0 0 90>}]
 skirt_bl_03 [{shapeType=Capsule, length=17.5, radius=4, shapeName=skirt_bl_03_capsule, center=<3 0 0>, rotation=<rot 0 0 90>}]
 skirt_bl_04 [{shapeType=Capsule, length=16.25, radius=3.75, shapeName=skirt_bl_04_capsule, center=<3 0 0>, rotation=<rot 0 0 90>}]
 skirt_bl_base [{shapeType=Capsule, length=1, radius=1, shapeName=skirt_bl_base_capsule, rotation=<rot 0 0 90>}]
 skirt_br_01 [{shapeType=Capsule, length=13, radius=3, shapeName=skirt_br_01_capsule, center=<3 0 0>, rotation=<rot 0 0 90>}]
 skirt_br_02 [{shapeType=Capsule, length=13, radius=3, shapeName=skirt_br_02_capsule, center=<3 0 0>, rotation=<rot 0 0 90>}]
 skirt_br_03 [{shapeType=Capsule, length=16.5, radius=3, shapeName=skirt_br_03_capsule, center=<3 0 0>, rotation=<rot 0 0 90>}]
 skirt_br_04 [{shapeType=Capsule, length=16.5, radius=3, shapeName=skirt_br_04_capsule, center=<3 0 0>, rotation=<rot 0 0 90>}]
 skirt_br_base [{shapeType=Capsule, length=1, radius=1, shapeName=skirt_br_base_capsule, rotation=<rot 0 0 90>}]
 skirt_connector_01_l [{
   shapeType=Box
   extentX=4.5
   extentY=6
   extentZ=6.5
   shapeName=skirt_connector_02_l_box
   center=<3.9695 -0.0567 0.319>
   rotation=<rot 75.7607 88.7114 94.7338>
  }]
 skirt_connector_01_r [{shapeType=Box, extentX=4.5, extentY=6, extentZ=6.5, shapeName=skirt_connector_02_r_box, center=<3.971 -0.143 0>, rotation=<rot 90 0 7.2501>}]
 skirt_connector_02_l [{
   shapeType=Box
   extentX=4.5
   extentY=6
   extentZ=6.5
   shapeName=skirt_connector_02_l_box
   center=<5.0527 -0.5531 0.4351>
   rotation=<rot 75.7641 88.5385 92.3724>
  }]
 skirt_connector_02_r [{
   shapeType=Box
   extentX=4.5
   extentY=6
   extentZ=6.5
   shapeName=skirt_connector_02_r_box
   center=<5.0381 -0.9319 0>
   rotation=<rot 90 180 -172.7499>
  }]
 skirt_fl_01 [{
   shapeType=Box
   extentX=10
   extentY=14
   extentZ=2.5
   shapeName=skirt_fr_01_box
   center=<-6.2425 -0.0933 0.1537>
   rotation=<rot 1.941 0.3737 17.7229>
  }]
 skirt_fl_02 [{shapeType=Box, extentX=9, extentY=13.5, extentZ=2.5, shapeName=skirt_fl_02_box, center=<-4.2268 0 0>, rotation=<rot -4.1302 -0.7077 12.923>}]
 skirt_fl_03 [{shapeType=Capsule, length=16.5, radius=3.75, shapeName=skirt_bl_03_capsule, center=<-3 0 0>, rotation=<rot -0.0001 0.0004 112.1364>}]
 skirt_fl_04 [{shapeType=Capsule, length=16.5, radius=3.75, shapeName=skirt_bl_04_capsule, center=<-4 0 0>, rotation=<rot -0.0001 0.0003 112.1364>}]
 skirt_fl_base [{
   shapeType=Capsule
   length=2.8586
   radius=1.3863
   shapeName=skirt_fl_base_capsule
   center=<-2.0134 0.5022 -0.0455>
   rotation=<rot 0 151.2575 -85.7636>
  }]
 skirt_fr_01 [{
   shapeType=Box
   extentX=10
   extentY=14
   extentZ=2.5
   shapeName=skirt_fr_01_box
   center=<6.25 0.0479 -0.2923>
   rotation=<rot -0.0004 0.0007 13.3968>
  }]
 skirt_fr_02 [{shapeType=Box, extentX=9, extentY=13.5, extentZ=2.5, shapeName=skirt_fr_02_box, center=<4.2268 0 0>, rotation=<rot -0.0003 0.0005 9.7175>}]
 skirt_fr_03 [{shapeType=Capsule, length=16.5, radius=3, shapeName=skirt_br_03_capsule, center=<3 0 0>, rotation=<rot -0.0001 0.0002 103.2581>}]
 skirt_fr_04 [{shapeType=Capsule, length=16.5, radius=3, shapeName=skirt_br_04_capsule, center=<3 0 0>, rotation=<rot -0.0001 0.0002 103.2581>}]
 skirt_fr_base [{shapeType=Capsule, length=1, radius=1, shapeName=skirt_fr_base_capsule}]
 thigh_l [
  {shapeType=Capsule, length=24, radius=7, shapeName=thigh_l_capsule, center=<-15.6097 0 0>, rotation=<rot 90 0 0>}
  {shapeType=Capsule, length=15, radius=7.25, shapeName=thigh_l_capsule, center=<-15.3804 -4.3437 5.2057>, rotation=<rot 90 180 176.0622>}
 ]
 thigh_r [
  {shapeType=Capsule, length=24, radius=7, shapeName=thigh_r_capsule, center=<15.6097 0 0>, rotation=<rot 90 0 0>}
  {shapeType=Capsule, length=13.5, radius=7.25, shapeName=thigh_r_capsule, center=<15.3804 4.3437 -5.2057>, rotation=<rot 90 180 176.0622>}
 ]
 thigh_twist_01_l [
  {shapeType=Capsule, length=12.5025, radius=7.7687, shapeName=thigh_twist_01_l_capsule, center=<0.8916 0 0>, rotation=<rot 90 0 0>}
  {shapeType=Capsule, length=12.5025, radius=7.7687, shapeName=thigh_twist_01_l_capsule, center=<-3 0.1688 0>, rotation=<rot 90 0 0>}
 ]
 thigh_twist_01_r [
  {shapeType=Capsule, length=12.5025, radius=7.7687, shapeName=thigh_twist_01_r_capsule, center=<-0.8916 0 0>, rotation=<rot 90 0 0>}
  {shapeType=Capsule, length=12.5025, radius=7.7687, shapeName=thigh_twist_01_r_capsule, center=<3 -0.2085 0>, rotation=<rot 90 0 0>}
 ]
}
constraints [
 {bone1Name=loincloth_fr_01, bone2Name=loincloth_fr_master, swing1LimitDegrees=20, swing2LimitDegrees=10, twistLimitDegrees=15}
 {bone1Name=loincloth_fr_02, bone2Name=loincloth_fr_01, swing1LimitDegrees=20, swing2LimitDegrees=15, twistLimitDegrees=10}
 {bone1Name=loincloth_fr_03, bone2Name=loincloth_fr_02, swing1LimitDegrees=25, swing2LimitDegrees=15, twistLimitDegrees=10}
 {bone1Name=loincloth_fr_04, bone2Name=loincloth_fr_03, swing1LimitDegrees=22.5, swing2LimitDegrees=10, twistLimitDegrees=8}
 {bone1Name=loincloth_fr_05, bone2Name=loincloth_fr_04, swing1LimitDegrees=22.5, swing2LimitDegrees=8, twistLimitDegrees=8}
 {bone1Name=loincloth_fl_01, bone2Name=loincloth_fl_master, swing1LimitDegrees=20, swing2LimitDegrees=10, twistLimitDegrees=15}
 {bone1Name=loincloth_fl_02, bone2Name=loincloth_fl_01, swing1LimitDegrees=20, swing2LimitDegrees=15, twistLimitDegrees=10}
 {bone1Name=loincloth_fl_03, bone2Name=loincloth_fl_02, swing1LimitDegrees=25, swing2LimitDegrees=15, twistLimitDegrees=10}
 {bone1Name=loincloth_fl_04, bone2Name=loincloth_fl_03, swing1LimitDegrees=22.5, swing2LimitDegrees=10, twistLimitDegrees=8}
 {bone1Name=loincloth_fl_05, bone2Name=loincloth_fl_04, swing1LimitDegrees=22.5, swing2LimitDegrees=8, twistLimitDegrees=8}
 {bone1Name=skirt_bl_01, bone2Name=skirt_bl_base, swing1LimitDegrees=5, swing1Motion=Locked, swing2LimitDegrees=30, twistLimitDegrees=5}
 {bone1Name=skirt_bl_02, bone2Name=skirt_bl_01, swing1LimitDegrees=5, swing2LimitDegrees=15, twistLimitDegrees=5}
 {bone1Name=skirt_bl_03, bone2Name=skirt_bl_02, swing1LimitDegrees=5, swing2LimitDegrees=15, twistLimitDegrees=5}
 {bone1Name=skirt_bl_04, bone2Name=skirt_bl_03, swing1LimitDegrees=5, swing2LimitDegrees=15, twistLimitDegrees=5}
 {bone1Name=skirt_br_01, bone2Name=skirt_br_base, swing1LimitDegrees=5, swing2LimitDegrees=30, twistLimitDegrees=5}
 {bone1Name=skirt_br_02, bone2Name=skirt_br_01, swing1LimitDegrees=5, swing2LimitDegrees=30, twistLimitDegrees=5}
 {bone1Name=skirt_br_03, bone2Name=skirt_br_02, swing1LimitDegrees=5, swing2LimitDegrees=30, twistLimitDegrees=5}
 {bone1Name=skirt_br_04, bone2Name=skirt_br_03, swing1LimitDegrees=5, swing2LimitDegrees=30, twistLimitDegrees=5}
 {bone1Name=skirt_back_a_01, bone2Name=skirt_back_a_base, swing1LimitDegrees=10, swing2LimitDegrees=10, twistLimitDegrees=5}
 {bone1Name=skirt_back_a_02, bone2Name=skirt_back_a_01, swing1LimitDegrees=30, swing2LimitDegrees=30, twistLimitDegrees=5}
 {bone1Name=skirt_back_a_03, bone2Name=skirt_back_a_02, swing1LimitDegrees=30, swing2LimitDegrees=10, twistLimitDegrees=5}
 {bone1Name=skirt_back_a_04, bone2Name=skirt_back_a_03, swing1LimitDegrees=30, swing2LimitDegrees=10, twistLimitDegrees=5}
 {bone1Name=skirt_back_b_01, bone2Name=skirt_back_b_base, swing1LimitDegrees=10, swing2LimitDegrees=10, twistLimitDegrees=5}
 {bone1Name=skirt_back_b_02, bone2Name=skirt_back_b_01, swing1LimitDegrees=30, swing2LimitDegrees=30, twistLimitDegrees=5}
 {bone1Name=skirt_back_b_03, bone2Name=skirt_back_b_02, swing1LimitDegrees=30, swing2LimitDegrees=10, twistLimitDegrees=5}
 {bone1Name=skirt_back_b_04, bone2Name=skirt_back_b_03, swing1LimitDegrees=30, swing2LimitDegrees=10, twistLimitDegrees=5}
 {
  bone1Name=loincloth_fr_master
  bone2Name=pelvis
  swing1LimitDegrees=45
  swing1Motion=Free
  swing2LimitDegrees=45
  swing2Motion=Free
  twistLimitDegrees=45
  twistMotion=Free
 }
 {
  bone1Name=loincloth_fl_master
  bone2Name=pelvis
  swing1LimitDegrees=45
  swing1Motion=Free
  swing2LimitDegrees=45
  swing2Motion=Free
  twistLimitDegrees=45
  twistMotion=Free
 }
 {
  bone1Name=skirt_bl_base
  bone2Name=pelvis
  swing1LimitDegrees=45
  swing1Motion=Free
  swing2LimitDegrees=45
  swing2Motion=Free
  twistLimitDegrees=45
  twistMotion=Free
 }
 {
  bone1Name=skirt_back_a_base
  bone2Name=pelvis
  swing1LimitDegrees=45
  swing1Motion=Free
  swing2LimitDegrees=45
  swing2Motion=Free
  twistLimitDegrees=45
  twistMotion=Free
 }
 {
  bone1Name=skirt_back_b_base
  bone2Name=pelvis
  swing1LimitDegrees=45
  swing1Motion=Free
  swing2LimitDegrees=45
  swing2Motion=Free
  twistLimitDegrees=45
  twistMotion=Free
 }
 {
  bone1Name=skirt_br_base
  bone2Name=pelvis
  swing1LimitDegrees=45
  swing1Motion=Free
  swing2LimitDegrees=45
  swing2Motion=Free
  twistLimitDegrees=45
  twistMotion=Free
 }
 {
  bone1Name=thigh_l
  bone2Name=pelvis
  swing1LimitDegrees=45
  swing1Motion=Free
  swing2LimitDegrees=45
  swing2Motion=Free
  twistLimitDegrees=45
  twistMotion=Free
 }
 {
  bone1Name=thigh_r
  bone2Name=pelvis
  swing1LimitDegrees=45
  swing1Motion=Free
  swing2LimitDegrees=45
  swing2Motion=Free
  twistLimitDegrees=45
  twistMotion=Free
 }
 {
  bone1Name=thigh_twist_01_r
  bone2Name=thigh_r
  swing1LimitDegrees=45
  swing1Motion=Free
  swing2LimitDegrees=45
  swing2Motion=Free
  twistLimitDegrees=45
  twistMotion=Free
 }
 {
  bone1Name=thigh_twist_01_l
  bone2Name=thigh_l
  swing1LimitDegrees=45
  swing1Motion=Free
  swing2LimitDegrees=45
  swing2Motion=Free
  twistLimitDegrees=45
  twistMotion=Free
 }
 {
  bone1Name=skirt_fr_base
  bone2Name=pelvis
  swing1LimitDegrees=45
  swing1Motion=Free
  swing2LimitDegrees=45
  swing2Motion=Free
  twistLimitDegrees=45
  twistMotion=Free
 }
 {bone1Name=skirt_fr_01, bone2Name=skirt_fr_base, swing1LimitDegrees=20, swing2LimitDegrees=30, twistLimitDegrees=45}
 {bone1Name=skirt_fr_02, bone2Name=skirt_fr_01, swing1LimitDegrees=15, swing2LimitDegrees=15, twistLimitDegrees=15}
 {bone1Name=skirt_fr_03, bone2Name=skirt_fr_02, swing1LimitDegrees=5, swing2LimitDegrees=15, twistLimitDegrees=5}
 {bone1Name=skirt_fr_04, bone2Name=skirt_fr_03, swing1LimitDegrees=5, swing2LimitDegrees=15, twistLimitDegrees=5}
 {
  bone1Name=skirt_fl_base
  bone2Name=pelvis
  swing1LimitDegrees=45
  swing1Motion=Free
  swing2LimitDegrees=45
  swing2Motion=Free
  twistLimitDegrees=45
  twistMotion=Free
 }
 {bone1Name=skirt_fl_01, bone2Name=skirt_fl_base, swing1LimitDegrees=60, swing2LimitDegrees=30, twistLimitDegrees=90}
 {bone1Name=skirt_fl_02, bone2Name=skirt_fl_01, swing1LimitDegrees=15, swing2LimitDegrees=15, twistLimitDegrees=15}
 {bone1Name=skirt_fl_03, bone2Name=skirt_fl_02, swing1LimitDegrees=5, swing2LimitDegrees=15, twistLimitDegrees=5}
 {bone1Name=skirt_fl_04, bone2Name=skirt_fl_03, swing1LimitDegrees=5, swing2LimitDegrees=15, twistLimitDegrees=5}
 {bone1Name=skirt_connector_02_r, bone2Name=skirt_connector_01_r, swing1LimitDegrees=35, swing2LimitDegrees=15, twistLimitDegrees=15}
 {bone1Name=skirt_connector_02_l, bone2Name=skirt_connector_01_l, swing1LimitDegrees=35, swing2LimitDegrees=15, twistLimitDegrees=15}
 {
  bone1Name=skirt_br_04
  bone2Name=skirt_fr_04
  swing1LimitDegrees=45
  swing1Motion=Free
  swing2LimitDegrees=45
  swing2Motion=Free
  twistLimitDegrees=45
  twistMotion=Free
 }
 {
  bone1Name=skirt_bl_04
  bone2Name=skirt_fl_04
  swing1LimitDegrees=45
  swing1Motion=Free
  swing2LimitDegrees=45
  swing2Motion=Free
  twistLimitDegrees=45
  twistMotion=Free
 }
 {
  bone1Name=skirt_back_b_04
  bone2Name=skirt_back_a_04
  swing1LimitDegrees=45
  swing1Motion=Free
  swing2LimitDegrees=45
  swing2Motion=Free
  twistLimitDegrees=45
  twistMotion=Free
 }
 {bone1Name=skirt_connector_01_l, bone2Name=thigh_l, swing1LimitDegrees=45, swing2LimitDegrees=5, twistLimitDegrees=15}
 {bone1Name=skirt_connector_01_r, bone2Name=thigh_r, swing1LimitDegrees=45, swing2LimitDegrees=5, twistLimitDegrees=15}
]
