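# Physics asset for the skydive setup of the F_MED_Parfait character body, covering its skirt and loincloth around pelvis and thigh bodies.
# Two loincloth chains and six skirt chains use mostly capsules, with boxes for front skirt panels and side connectors.
# Chain bases rotate freely relative to the pelvis, while segments below mostly have tight swing and twist limits.
# A free-rotation constraint links the loincloth tips, and side connectors lock all but one swing axis.
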PhysicsAsset /BRCosmetics/Characters/Player/Female/Medium/Bodies/F_MED_Parfait/Meshes/F_MED_Parfait_Skydive_Physics
bodies {
 loincloth_fl_01 [{shapeType=Capsule, length=6, radius=3, shapeName=loincloth_fl_01_capsule, center=<3 0 0>}]
 loincloth_fl_02 [{shapeType=Capsule, length=8.5, radius=2.5, shapeName=loincloth_fl_02_capsule, center=<3 0 0>}]
 loincloth_fl_03 [{shapeType=Capsule, length=8.5, radius=2.5, shapeName=loincloth_fl_03_capsule, center=<3 0 0>}]
 loincloth_fl_04 [{shapeType=Capsule, length=11.5, radius=2.5, shapeName=loincloth_fl_04_capsule, center=<3 0 0>}]
 loincloth_fl_05 [{shapeType=Capsule, length=11, radius=3, shapeName=loincloth_fl_05_capsule, center=<3 0 0>}]
 loincloth_fl_master [{shapeType=Capsule, length=1, radius=1, shapeName=loincloth_fl_master_capsule}]
 loincloth_fr_01 [{shapeType=Capsule, length=6, radius=3, shapeName=loincloth_fr_01_capsule, center=<3 0 0>}]
 loincloth_fr_02 [{shapeType=Capsule, length=8, radius=3, shapeName=loincloth_fr_02_capsule, center=<3 0 0>, rotation=<rot 0 0 180>}]
 loincloth_fr_03 [{shapeType=Capsule, length=8, radius=3, shapeName=loincloth_fr_03_capsule, center=<3 0 0>, rotation=<rot 0 0 180>}]
 loincloth_fr_04 [{shapeType=Capsule, length=11, radius=3, shapeName=loincloth_fr_04_capsule, center=<3 0 0>, rotation=<rot 0 0 180>}]
 loincloth_fr_05 [{shapeType=Capsule, length=11, radius=3, shapeName=loincloth_fr_05_capsule, center=<3 0 0>, rotation=<rot 0 0 180>}]
 loincloth_fr_master [{shapeType=Capsule, length=1, radius=1, shapeName=loincloth_fr_master_capsule}]
 pelvis [{
   shapeType=Capsule
   length=15.1534
   radius=6.5145
   shapeName=pelvis_capsule
   center=<0.4119 -1.5176 -0.0401>
   rotation=<rot -0.8022 -90 -0.0752>
  }]
 skirt_back_a_01 [{shapeType=Capsule, length=6, radius=3, shapeName=skirt_back_a_01_capsule, center=<3 0 0>}]
 skirt_back_a_02 [{shapeType=Capsule, length=6, radius=3, shapeName=skirt_back_a_02_capsule, center=<3 0 0>}]
 skirt_back_a_03 [{shapeType=Capsule, length=6, radius=3, shapeName=skirt_back_a_03_capsule, center=<3 0 0>}]
 skirt_back_a_04 [{shapeType=Capsule, length=6, radius=3, shapeName=skirt_back_a_04_capsule, center=<3 0 0>}]
 skirt_back_a_base [{shapeType=Capsule, length=1, radius=1, shapeName=skirt_back_a_base_capsule}]
 skirt_back_b_01 [{shapeType=Capsule, length=11.5, radius=3, shapeName=skirt_back_b_01_capsule, center=<3 0 0>}]
 skirt_back_b_02 [{shapeType=Capsule, length=8, radius=6.5, shapeName=skirt_back_b_02_capsule, center=<3 0 0>}]
 skirt_back_b_03 [{shapeType=Capsule, length=8, radius=6.5, shapeName=skirt_back_b_03_capsule, center=<3 0 0>}]
 skirt_back_b_04 [{shapeType=Capsule, length=8, radius=6.5, shapeName=skirt_back_b_04_capsule, center=<3 0 0>}]
 skirt_back_b_base [{shapeType=Capsule, length=1, radius=1, shapeName=skirt_back_b_base_capsule}]
 skirt_bl_01 [{shapeType=Capsule, length=6, radius=3, shapeName=skirt_bl_01_capsule, center=<3 0 0>, rotation=<rot 0 0 90>}]
 skirt_bl_02 [{shapeType=Capsule, length=15.75, radius=3, shapeName=skirt_bl_02_capsule, center=<3 0 0>, rotation=<rot 0 0 90>}]
 skirt_bl_03 [{shapeType=Capsule, length=15.75, radius=3, shapeName=skirt_bl_03_capsule, center=<3 0 0>, rotation=<rot 0 0 90>}]
 skirt_bl_04 [{shapeType=Capsule, length=15.75, radius=3, shapeName=skirt_bl_04_capsule, center=<3 0 0>, rotation=<rot 0 0 90>}]
 skirt_bl_base [{shapeType=Capsule, length=1, radius=1, shapeName=skirt_bl_base_capsule, rotation=<rot 0 0 90>}]
 skirt_br_01 [{shapeType=Capsule, length=12.25, radius=3.75, shapeName=skirt_br_01_capsule, center=<3 0 0>, rotation=<rot 0 0 90>}]
 skirt_br_02 [{shapeType=Capsule, length=12.25, radius=3.75, shapeName=skirt_br_02_capsule, center=<3 0 0>, rotation=<rot 0 0 90>}]
 skirt_br_03 [{shapeType=Capsule, length=15.75, radius=3.75, shapeName=skirt_br_03_capsule, center=<3 0 0>, rotation=<rot 0 0 90>}]
 skirt_br_04 [{shapeType=Capsule, length=15.75, radius=3.75, shapeName=skirt_br_04_capsule, center=<3 0 0>, rotation=<rot 0 0 90>}]
 skirt_br_base [{shapeType=Capsule, length=1, radius=1, shapeName=skirt_br_base_capsule, rotation=<rot 0 0 90>}]
 skirt_connector_01_l [{shapeType=Box, extentX=7.5, extentY=1, extentZ=7.5, shapeName=skirt_connector_02_l_box}]
 skirt_connector_01_r [{shapeType=Box, extentX=7.5, extentY=1, extentZ=7.5, shapeName=skirt_connector_02_r_box}]
 skirt_connector_02_l [{shapeType=Box, extentX=7.5, extentY=1, extentZ=7.5, shapeName=skirt_connector_02_l_box}]
 skirt_connector_02_r [{shapeType=Box, extentX=7.5, extentY=1, extentZ=7.5, shapeName=skirt_connector_02_r_box}]
 skirt_fl_01 [{shapeType=Box, extentX=12.5, extentY=14, extentZ=3, shapeName=skirt_fr_01_box, center=<-6.7 0 0>}]
 skirt_fl_02 [{shapeType=Box, extentX=10, extentY=14.5, extentZ=3, shapeName=skirt_fl_02_box, center=<-4.2 0 0>}]
 skirt_fl_03 [{shapeType=Box, extentX=11.5, extentY=13.5, extentZ=3.5, shapeName=skirt_fr_03_box, center=<-4.2268 0 0>}]
 skirt_fl_04 [{shapeType=Box, extentX=11.5, extentY=13.5, extentZ=3.5, shapeName=skirt_fr_04_box, center=<-4.2268 0 0>}]
 skirt_fl_base [{
   shapeType=Capsule
   length=2.8586
   radius=1.3863
   shapeName=skirt_fl_base_capsule
   center=<-2.0134 0.5022 -0.0455>
   rotation=<rot 0 151.2575 -85.7636>
  }]
 skirt_fr_01 [{shapeType=Box, extentX=10, extentY=14, extentZ=2, shapeName=skirt_fr_01_box, center=<6.7124 0 0>}]
 skirt_fr_02 [{shapeType=Box, extentX=9, extentY=13.5, extentZ=2.5, shapeName=skirt_fr_02_box, center=<4.2268 0 0>}]
 skirt_fr_03 [{shapeType=Box, extentX=9.5, extentY=16, extentZ=2.5, shapeName=skirt_fr_03_box, center=<4.2268 0 0>}]
 skirt_fr_04 [{shapeType=Box, extentX=9.5, extentY=16, extentZ=2.5, shapeName=skirt_fr_04_box, center=<4.2268 0 0>}]
 skirt_fr_base [{shapeType=Capsule, length=1, radius=1, shapeName=skirt_fr_base_capsule}]
 thigh_l [{shapeType=Capsule, length=25.25, radius=7.25, shapeName=thigh_l_capsule, center=<-15 0 0>, rotation=<rot 90 0 0>}]
 thigh_r [{shapeType=Capsule, length=25.5, radius=7, shapeName=thigh_r_capsule, center=<15 0 0>, rotation=<rot 90 0 0>}]
 thigh_twist_01_l [{shapeType=Capsule, length=10.7525, radius=7.5187, shapeName=thigh_twist_01_l_capsule, rotation=<rot 90 0 0>}]
 thigh_twist_01_r [{shapeType=Capsule, length=10.7525, radius=7.5187, shapeName=thigh_twist_01_r_capsule, rotation=<rot 90 0 0>}]
}
constraints [
 {bone1Name=loincloth_fr_01, bone2Name=loincloth_fr_master, swing1LimitDegrees=15, swing2LimitDegrees=0, twistLimitDegrees=5}
 {bone1Name=loincloth_fr_02, bone2Name=loincloth_fr_01, swing1LimitDegrees=30, swing2LimitDegrees=15, twistLimitDegrees=5}
 {bone1Name=loincloth_fr_03, bone2Name=loincloth_fr_02, swing1LimitDegrees=30, swing2LimitDegrees=15, twistLimitDegrees=5}
 {bone1Name=loincloth_fr_04, bone2Name=loincloth_fr_03, swing1LimitDegrees=30, swing2LimitDegrees=2, twistLimitDegrees=5}
 {bone1Name=loincloth_fr_05, bone2Name=loincloth_fr_04, swing1LimitDegrees=30, swing2LimitDegrees=2, twistLimitDegrees=5}
 {bone1Name=loincloth_fl_01, bone2Name=loincloth_fl_master, swing1LimitDegrees=15, swing2LimitDegrees=0, twistLimitDegrees=5}
 {bone1Name=loincloth_fl_02, bone2Name=loincloth_fl_01, swing1LimitDegrees=30, swing2LimitDegrees=15, twistLimitDegrees=5}
 {bone1Name=loincloth_fl_03, bone2Name=loincloth_fl_02, swing1LimitDegrees=30, swing2LimitDegrees=15, twistLimitDegrees=5}
 {bone1Name=loincloth_fl_04, bone2Name=loincloth_fl_03, swing1LimitDegrees=30, swing2LimitDegrees=2, twistLimitDegrees=5}
 {bone1Name=loincloth_fl_05, bone2Name=loincloth_fl_04, swing1LimitDegrees=30, swing2LimitDegrees=2, twistLimitDegrees=5}
 {bone1Name=skirt_bl_01, bone2Name=skirt_bl_base, swing1LimitDegrees=5, swing2LimitDegrees=15, twistLimitDegrees=5}
 {bone1Name=skirt_bl_02, bone2Name=skirt_bl_01, swing1LimitDegrees=5, swing2LimitDegrees=15, twistLimitDegrees=5}
 {bone1Name=skirt_bl_03, bone2Name=skirt_bl_02, swing1LimitDegrees=5, swing2LimitDegrees=15, twistLimitDegrees=5}
 {bone1Name=skirt_bl_04, bone2Name=skirt_bl_03, swing1LimitDegrees=5, swing2LimitDegrees=15, twistLimitDegrees=5}
 {bone1Name=skirt_br_01, bone2Name=skirt_br_base, swing1LimitDegrees=5, swing2LimitDegrees=15, twistLimitDegrees=5}
 {bone1Name=skirt_br_02, bone2Name=skirt_br_01, swing1LimitDegrees=5, swing2LimitDegrees=15, twistLimitDegrees=5}
 {bone1Name=skirt_br_03, bone2Name=skirt_br_02, swing1LimitDegrees=5, swing2LimitDegrees=15, twistLimitDegrees=5}
 {bone1Name=skirt_br_04, bone2Name=skirt_br_03, swing1LimitDegrees=5, swing2LimitDegrees=15, twistLimitDegrees=5}
 {bone1Name=skirt_back_a_01, bone2Name=skirt_back_a_base, swing1LimitDegrees=30, swing2LimitDegrees=30, twistLimitDegrees=5}
 {bone1Name=skirt_back_a_02, bone2Name=skirt_back_a_01, swing1LimitDegrees=15, swing2LimitDegrees=10, twistLimitDegrees=5}
 {bone1Name=skirt_back_a_03, bone2Name=skirt_back_a_02, swing1LimitDegrees=15, swing2LimitDegrees=10, twistLimitDegrees=5}
 {bone1Name=skirt_back_a_04, bone2Name=skirt_back_a_03, swing1LimitDegrees=15, swing2LimitDegrees=10, twistLimitDegrees=5}
 {bone1Name=skirt_back_b_01, bone2Name=skirt_back_b_base, swing1LimitDegrees=30, swing2LimitDegrees=30, twistLimitDegrees=5}
 {bone1Name=skirt_back_b_02, bone2Name=skirt_back_b_01, swing1LimitDegrees=15, swing2LimitDegrees=10, twistLimitDegrees=5}
 {bone1Name=skirt_back_b_03, bone2Name=skirt_back_b_02, swing1LimitDegrees=15, swing2LimitDegrees=10, twistLimitDegrees=5}
 {bone1Name=skirt_back_b_04, bone2Name=skirt_back_b_03, swing1LimitDegrees=15, swing2LimitDegrees=10, twistLimitDegrees=5}
 {
  bone1Name=loincloth_fr_master
  bone2Name=pelvis
  swing1LimitDegrees=45
  swing1Motion=Free
  swing2LimitDegrees=45
  swing2Motion=Free
  twistLimitDegrees=45
  twistMotion=Free
 }
 {
  bone1Name=loincloth_fl_master
  bone2Name=pelvis
  swing1LimitDegrees=45
  swing1Motion=Free
  swing2LimitDegrees=45
  swing2Motion=Free
  twistLimitDegrees=45
  twistMotion=Free
 }
 {
  bone1Name=skirt_bl_base
  bone2Name=pelvis
  swing1LimitDegrees=45
  swing1Motion=Free
  swing2LimitDegrees=45
  swing2Motion=Free
  twistLimitDegrees=45
  twistMotion=Free
 }
 {
  bone1Name=skirt_back_a_base
  bone2Name=pelvis
  swing1LimitDegrees=45
  swing1Motion=Free
  swing2LimitDegrees=45
  swing2Motion=Free
  twistLimitDegrees=45
  twistMotion=Free
 }
 {
  bone1Name=skirt_back_b_base
  bone2Name=pelvis
  swing1LimitDegrees=45
  swing1Motion=Free
  swing2LimitDegrees=45
  swing2Motion=Free
  twistLimitDegrees=45
  twistMotion=Free
 }
 {
  bone1Name=skirt_br_base
  bone2Name=pelvis
  swing1LimitDegrees=45
  swing1Motion=Free
  swing2LimitDegrees=45
  swing2Motion=Free
  twistLimitDegrees=45
  twistMotion=Free
 }
 {
  bone1Name=thigh_l
  bone2Name=pelvis
  swing1LimitDegrees=45
  swing1Motion=Free
  swing2LimitDegrees=45
  swing2Motion=Free
  twistLimitDegrees=45
  twistMotion=Free
 }
 {
  bone1Name=thigh_r
  bone2Name=pelvis
  swing1LimitDegrees=45
  swing1Motion=Free
  swing2LimitDegrees=45
  swing2Motion=Free
  twistLimitDegrees=45
  twistMotion=Free
 }
 {
  bone1Name=loincloth_fr_05
  bone2Name=loincloth_fl_05
  swing1LimitDegrees=45
  swing1Motion=Free
  swing2LimitDegrees=45
  swing2Motion=Free
  twistLimitDegrees=45
  twistMotion=Free
 }
 {
  bone1Name=thigh_twist_01_r
  bone2Name=thigh_r
  swing1LimitDegrees=45
  swing1Motion=Free
  swing2LimitDegrees=45
  swing2Motion=Free
  twistLimitDegrees=45
  twistMotion=Free
 }
 {
  bone1Name=thigh_twist_01_l
  bone2Name=thigh_l
  swing1LimitDegrees=45
  swing1Motion=Free
  swing2LimitDegrees=45
  swing2Motion=Free
  twistLimitDegrees=45
  twistMotion=Free
 }
 {
  bone1Name=skirt_fr_base
  bone2Name=pelvis
  swing1LimitDegrees=45
  swing1Motion=Free
  swing2LimitDegrees=45
  swing2Motion=Free
  twistLimitDegrees=45
  twistMotion=Free
 }
 {bone1Name=skirt_fr_01, bone2Name=skirt_fr_base, swing1LimitDegrees=30, swing1Motion=Locked, swing2LimitDegrees=15, twistLimitDegrees=5}
 {bone1Name=skirt_fr_02, bone2Name=skirt_fr_01, swing1LimitDegrees=5, swing2LimitDegrees=15, twistLimitDegrees=5}
 {bone1Name=skirt_fr_03, bone2Name=skirt_fr_02, swing1LimitDegrees=15, swing2LimitDegrees=15, twistLimitDegrees=5}
 {bone1Name=skirt_fr_04, bone2Name=skirt_fr_03, swing1LimitDegrees=15, swing2LimitDegrees=15, twistLimitDegrees=5}
 {
  bone1Name=skirt_fl_base
  bone2Name=pelvis
  swing1LimitDegrees=45
  swing1Motion=Free
  swing2LimitDegrees=45
  swing2Motion=Free
  twistLimitDegrees=45
  twistMotion=Free
 }
 {bone1Name=skirt_fl_01, bone2Name=skirt_fl_base, swing1LimitDegrees=5, swing2LimitDegrees=15, twistLimitDegrees=5}
 {bone1Name=skirt_fl_02, bone2Name=skirt_fl_01, swing1LimitDegrees=5, swing2LimitDegrees=15, twistLimitDegrees=5}
 {bone1Name=skirt_fl_03, bone2Name=skirt_fl_02, swing1LimitDegrees=5, swing2LimitDegrees=15, twistLimitDegrees=5}
 {bone1Name=skirt_fl_04, bone2Name=skirt_fl_03, swing1LimitDegrees=5, swing2LimitDegrees=15, twistLimitDegrees=5}
 {
  bone1Name=skirt_connector_01_r
  bone2Name=pelvis
  swing1LimitDegrees=15
  swing2LimitDegrees=15
  swing2Motion=Locked
  twistLimitDegrees=45
  twistMotion=Locked
 }
 {
  bone1Name=skirt_connector_02_r
  bone2Name=skirt_connector_01_r
  swing1LimitDegrees=15
  swing2LimitDegrees=15
  swing2Motion=Locked
  twistLimitDegrees=45
  twistMotion=Locked
 }
 {
  bone1Name=skirt_connector_01_l
  bone2Name=pelvis
  swing1LimitDegrees=15
  swing2LimitDegrees=15
  swing2Motion=Locked
  twistLimitDegrees=45
  twistMotion=Locked
 }
 {
  bone1Name=skirt_connector_02_l
  bone2Name=skirt_connector_01_l
  swing1LimitDegrees=15
  swing2LimitDegrees=15
  swing2Motion=Locked
  twistLimitDegrees=45
  twistMotion=Locked
 }
]
