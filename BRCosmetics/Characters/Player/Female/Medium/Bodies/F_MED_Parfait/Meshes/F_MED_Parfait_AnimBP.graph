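# Post-process style AnimBP for the female Parfait cosmetic body, copying the parent mesh pose and adding skirt and cape secondary motion.
# Control Rigs and state-switched RigidBody skirt sims feed a cached base pose.
# Cape-removal, cloth and rigid-body flags select that base, RigidBody cape or Trail Controller cape output.
# The event graph drives gravity, cloth settings and movement-based cape corrections, suspending cloth for backblings, removed capes and certain vehicles.
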AnimBlueprint /BRCosmetics/Characters/Player/Female/Medium/Bodies/F_MED_Parfait/Meshes/F_MED_Parfait_AnimBP
target_skeleton=&/F_MED_Parfait_Skeleton
graphs [
 AnimGraph {nodes [
   Blend Poses by bool(BlendPose_0=Use cached pose 'dbnoCache'(), bActiveValue=Get bIsDBNO()) {blendType=Linear}
   n1 = Get parfait_built_in_emote() -> (parfait_built_in_emote:bool)
   "Transform (Modify) Bone - Bone: cape_02_r"!(ComponentPose, bAlphaBoolEnabled:bool=@n1, Translation="0,0,0", Rotation="0,-15,0", Scale="1,1,1") -> (Pose) {
    alphaInputType=Bool
    rotationMode=BMM_Additive
    rotationSpace=BCS_BoneSpace
    alphaBoolBlend {blendInTime=0.25, blendOption=Linear, blendOutTime=0.25}
    boneToModify {boneName=cape_02_r}
    rotation {yaw=-15}
   }
   n2 = Blend Poses by bool(BlendPose_0=Use cached pose 'Start_Cache'(), BlendPose_1=Use cached pose 'RBAN Cache'(), bActiveValue=Get bClothEnabled()) {blendType=Linear}
   n3 = Use cached pose 'trailCache'()
   n4 = Get cloth_disabled_rban_enabled() -> (cloth_disabled_rban_enabled:bool)
   Output Pose(Result=Blend Poses by bool(
     BlendPose_0=Use cached pose 'Start_Cache'()
     BlendPose_1=Blend Poses by bool(BlendPose_0=@n2, BlendPose_1=@n3, bActiveValue=@n4) {blendType=Linear}
     bActiveValue=Get cape_removed() -> (cape_removed:bool)
    ) {blendType=Linear}) {layerGroup=None}
   n5 = Control Rig!(
    Source=Copy Pose From Mesh() {bUseAttachedParent=true}
    Alpha
    Jumping:bool=Get jump_falling() -> (jump_falling:bool)
    Emoting:bool=Get bIsPlayingEmote()
    Crouching:bool=Get bIsCrouching()
    Moving:bool=Get bIsMoving2D()
   ) -> (Pose) {
    alphaBoolBlend {blendOption=Linear}
    controlRigAssetReference {blueprintRigClass=&/F_MED_Parfait_CtrlRig.F_MED_Parfait_CtrlRig_C}
   }
   n6 = RigidBody!(ComponentPose=Local To Component(LocalPose=@n5), Alpha, OverrideWorldGravity:Vector=Get GravityOverride(), ExternalForce) -> (Pose) {
    bOverrideWorldGravity=true
    overridePhysicsAsset=&/F_MED_Parfait_Skirt_Physics
    simulationSpace=BaseBoneSpace
    alphaBoolBlend {blendOption=Linear}
    baseBoneRef {boneName=pelvis}
    componentAppliedLinearAccClamp=<1500 1500 2000>
    componentLinearAccScale=<0.35 0.35 0.5>
    componentLinearVelScale=<0.35 0.35 0.5>
    overrideWorldGravity {z=-980}
   }
   n7 = RigidBody!(ComponentPose=@n6, bAlphaBoolEnabled:bool=Get skydive() -> (skydive:bool), OverrideWorldGravity=Get GravityOverride(), ExternalForce) -> (Pose) {
    alphaInputType=Bool
    bOverrideWorldGravity=true
    overridePhysicsAsset=&/F_MED_Parfait_Skydive_Physics
    simulationSpace=BaseBoneSpace
    alphaBoolBlend {blendOption=Linear}
    baseBoneRef {boneName=pelvis}
    componentAppliedLinearAccClamp=<1500 1500 1500>
    componentLinearAccScale=<0.25 0.25 0.3>
    componentLinearVelScale=<0.25 0.25 0.3>
    overrideWorldGravity {z=-980}
   }
   n8 = Component To Local(ComponentPose=@n7)
   n9 = Control Rig(
    Source=Copy Pose From Mesh() {bUseAttachedParent=true}
    Jumping=Get jump_falling()
    Emoting=Get bIsPlayingEmote()
    Crouching=Get bIsCrouching()
    Moving=Get bIsMoving2D()
   ) {
    alphaBoolBlend {blendOption=Linear}
    controlRigAssetReference {blueprintRigClass=&F_MED_Parfait_CtrlRig_C}
   }
   n10 = "Transform (Modify) Bone - Bone: skirt_fl_base"!(
    ComponentPose=Local To Component(LocalPose=@n9)
    bAlphaBoolEnabled:bool=True
    Translation="0,3.286,3.923"
    Rotation="7.415,-13.075,29.155"
    Scale="1,1,1"
   ) -> (Pose) {
    alphaInputType=Bool
    rotationMode=BMM_Additive
    rotationSpace=BCS_BoneSpace
    translationMode=BMM_Additive
    translationSpace=BCS_BoneSpace
    alphaBoolBlend {blendOption=Linear}
    boneToModify {boneName=skirt_fl_base}
    rotation=<rot 7.415 -13.075 29.155>
    translation {y=3.286, z=3.923}
   }
   n11 = RigidBody!(ComponentPose=@n10, Alpha:float, OverrideWorldGravity=Get GravityOverride(), ExternalForce) -> (Pose) {
    bOverrideWorldGravity=true
    overridePhysicsAsset=&/F_MED_Parfait_Crouch_Physics
    simulationSpace=BaseBoneSpace
    alphaBoolBlend {blendOption=Linear}
    baseBoneRef {boneName=pelvis}
    componentAppliedLinearAccClamp=<1500 1500 1500>
    componentLinearAccScale=<0.25 0.25 0.3>
    componentLinearVelScale=<0.25 0.25 0.3>
    overrideWorldGravity {z=-980}
   }
   n12 = Component To Local(ComponentPose=@n11)
   n13 = Get bIsCrouching()
   n14 = Control Rig(
    Source=Copy Pose From Mesh() {bUseAttachedParent=true}
    Alpha=0.7
    Jumping=Get jump_falling()
    Emoting=Get bIsPlayingEmote()
    Crouching=Get bIsCrouching()
    Moving=Get bIsMoving2D()
   ) {
    alphaBoolBlend {blendOption=Linear}
    controlRigAssetReference {blueprintRigClass=&F_MED_Parfait_CtrlRig_C}
   }
   n15 = RigidBody!(ComponentPose=Local To Component(LocalPose=@n14), bAlphaBoolEnabled:bool, OverrideWorldGravity=Get GravityOverride(), ExternalForce) -> (Pose) {
    alphaInputType=Bool
    bOverrideWorldGravity=true
    overridePhysicsAsset=&F_MED_Parfait_Skirt_Physics
    simulationSpace=BaseBoneSpace
    alphaBoolBlend {blendOption=Linear}
    baseBoneRef {boneName=pelvis}
    componentAppliedLinearAccClamp=<1500 1500 1500>
    componentLinearAccScale=<0.25 0.25 0.3>
    componentLinearVelScale=<0.25 0.25 0.3>
    overrideWorldGravity {z=-980}
   }
   n16 = Get bClothEnabled()
   n17 = "Transform (Modify) Bone - Bone: clavicle_cape_l"!(ComponentPose=@n15, bAlphaBoolEnabled:bool=@n16, Translation="0.511,-1.903,-0.000", Rotation="-0.000,15,-0.000", Scale="1,1,1") -> (Pose) {
    alphaInputType=Bool
    rotationMode=BMM_Additive
    rotationSpace=BCS_BoneSpace
    translationMode=BMM_Additive
    translationSpace=BCS_BoneSpace
    alphaBoolBlend {blendOption=Linear}
    boneToModify {boneName=clavicle_cape_l}
    rotation {yaw=15}
    translation=<0.511 -1.903>
   }
   n18 = "Transform (Modify) Bone - Bone: clavicle_cape_r"!(ComponentPose=@n17, bAlphaBoolEnabled:bool=@n16, Translation="-0.316,1.181,0.000", Rotation="0,15,0", Scale="1,1,1") -> (Pose) {
    alphaInputType=Bool
    rotationMode=BMM_Additive
    rotationSpace=BCS_BoneSpace
    translationMode=BMM_Additive
    translationSpace=BCS_BoneSpace
    alphaBoolBlend {blendOption=Linear}
    boneToModify {boneName=clavicle_cape_r}
    rotation {yaw=15}
    translation=<-0.316 1.181>
   }
   n19 = Control Rig!(
    Source=Blend Poses by bool(
     BlendPose_0=Component To Local(ComponentPose=@n18)
     BlendPose_1=Blend Poses by bool(BlendPose_0=@n12, BlendPose_1=@n8, BlendTime_1=2, bActiveValue=@n13) {blendType=Linear, transitionType=Inertialization}
     bActiveValue=Get bIsInFrontEnd()
    ) {blendType=Linear, transitionType=Inertialization}
    Alpha
   ) -> (Pose) {
    alphaBoolBlend {blendOption=Linear}
    controlRigAssetReference {blueprintRigClass=&/F_MED_DeformRig_Parfait_ControlRig.F_MED_DeformRig_Parfait_ControlRig_C}
   }
   n20 = Get bIsTacticalSprinting()
   n21 = "Transform (Modify) Bone - Bone: cape_02_l"!(ComponentPose=Local To Component(LocalPose=@n19), bAlphaBoolEnabled:bool=@n20, Translation="0,0,0", Rotation="0,-15,0", Scale="0,0,0") -> (Pose) {
    alphaInputType=Bool
    rotationMode=BMM_Additive
    rotationSpace=BCS_BoneSpace
    translationMode=BMM_Additive
    translationSpace=BCS_BoneSpace
    alphaBoolBlend {blendOption=Linear}
    boneToModify {boneName=cape_02_l}
    rotation {yaw=-15}
    scale=<0 0 0>
   }
   n22 = "Transform (Modify) Bone - Bone: cape_02_r"(ComponentPose=@n21, bAlphaBoolEnabled=@n20, Translation="0,0,0", Rotation="0,-10,0", Scale="0,0,0") {
    alphaInputType=Bool
    rotationMode=BMM_Additive
    rotationSpace=BCS_BoneSpace
    translationMode=BMM_Additive
    translationSpace=BCS_BoneSpace
    alphaBoolBlend {blendOption=Linear}
    boneToModify {boneName=cape_02_r}
    rotation {yaw=-10}
    scale=<0 0 0>
   }
   Save cached pose 'Start_Cache'(Pose=Component To Local(ComponentPose=@n22)) {cacheName=Start_Cache}
   n23 = Get BwdMovement() -> (BwdMovement:double)
   n24 = Trail controller - Bone: cape_mid_06_l(ComponentPose=Local To Component(LocalPose=Use cached pose 'Start_Cache'()), Alpha=@n23) {
    bUsePlanarLimit=true
    chainLength=5
    debugLifeTime=0.1
    maxDeltaTime=0.333
    alphaScaleBiasClamp {bInterpResult=true}
    rotationLimits [{limitMax=<180 180 180>}, {limitMax=<180 180 180>}, {limitMax=<180 180 180>}, {limitMax=<180 180 180>}, {limitMax=<180 180 180>}]
    rotationOffsets [<0 0 0>, <0 0 0>, <0 0 0>, <0 0 0>, <0 0 0>]
    trailBone {boneName=cape_mid_06_l}
    trailRelaxationSpeed {editorCurveData {keys [<at -0.0134 19.3841 RCIM_Linear>,<at 0.5079 12.8986 RCIM_Linear>,<at 0.9955 9.6044 RCIM_Linear>]}}
   }
   n25 = Trail controller - Bone: cape_mid_06_r(ComponentPose=@n24, Alpha=@n23) {
    bUsePlanarLimit=true
    chainLength=5
    debugLifeTime=0.1
    maxDeltaTime=0.0333
    alphaScaleBiasClamp {bInterpResult=true}
    rotationLimits [
     {limitMax=<180 180 180>}
     {limitMax=<180 180 180>}
     {limitMax=<180 90 180>, limitMin=<-180 -90 -180>}
     {limitMax=<180 180 180>}
     {limitMax=<180 180 180>}
    ]
    rotationOffsets [<0 0 0>, <0 0 0>, <0 0 0>, <0 0 0>, <0 0 0>]
    trailBone {boneName=cape_mid_06_r}
    trailRelaxationSpeed {editorCurveData {keys [<at 0 15 RCIM_Linear>,<at 0.5289 13.1838 RCIM_Linear>,<at 1 6 RCIM_Linear>]}}
   }
   n26 = Trail controller - Bone: cape_07_l(ComponentPose=@n25, Alpha=@n23) {
    bUsePlanarLimit=true
    chainLength=5
    debugLifeTime=0.1
    maxDeltaTime=0.0333
    alphaScaleBiasClamp {bInterpResult=true}
    rotationLimits [
     {limitMax=<180 180 180>}
     {limitMax=<180 180 180>}
     {limitMax=<180 90 180>, limitMin=<-180 -90 -180>}
     {limitMax=<180 180 180>}
     {limitMax=<180 180 180>}
    ]
    rotationOffsets [<0 0 0>, <0 0 0>, <0 0 0>, <0 0 0>, <0 0 0>]
    trailBone {boneName=cape_07_l}
    trailRelaxationSpeed {editorCurveData {keys [<at 0 15 RCIM_Linear>,<at 0.5289 13.1838 RCIM_Linear>,<at 1 6 RCIM_Linear>]}}
   }
   n27 = Trail controller - Bone: cape_07_r(ComponentPose=@n26, Alpha=@n23) {
    bUsePlanarLimit=true
    chainLength=5
    debugLifeTime=0.1
    maxDeltaTime=0.0333
    alphaScaleBiasClamp {bInterpResult=true}
    rotationLimits [
     {limitMax=<180 180 180>}
     {limitMax=<180 180 180>}
     {limitMax=<180 90 180>, limitMin=<-180 -90 -180>}
     {limitMax=<180 180 180>}
     {limitMax=<180 180 180>}
    ]
    rotationOffsets [<0 0 0>, <0 0 0>, <0 0 0>, <0 0 0>, <0 0 0>]
    trailBone {boneName=cape_07_r}
    trailRelaxationSpeed {editorCurveData {keys [<at 0 15 RCIM_Linear>,<at 0.5289 13.1838 RCIM_Linear>,<at 1 6 RCIM_Linear>]}}
   }
   n28 = Get clavicle_float_l() -> (clavicle_float_l:double)
   n29 = "Transform (Modify) Bone - Bone: clavicle_cape_l"!(ComponentPose=@n27, Alpha:float=@n28, Translation="0,0,0", Rotation="0,-15,15", Scale="1,1,1") -> (Pose) {
    rotationMode=BMM_Additive
    rotationSpace=BCS_BoneSpace
    translationMode=BMM_Additive
    translationSpace=BCS_BoneSpace
    alphaBoolBlend {blendOption=Linear}
    boneToModify {boneName=clavicle_cape_l}
    rotation {roll=15, yaw=-15}
   }
   n30 = "Transform (Modify) Bone - Bone: cape_mid_01_l"(ComponentPose=@n29, Alpha=@n28, Translation="0,0,0", Rotation="0,-15,0", Scale="1,1,1") {
    rotationMode=BMM_Additive
    rotationSpace=BCS_BoneSpace
    translationMode=BMM_Additive
    translationSpace=BCS_BoneSpace
    alphaBoolBlend {blendOption=Linear}
    boneToModify {boneName=cape_mid_01_l}
    rotation {yaw=-15}
   }
   n31 = Get clavicle_float_r() -> (clavicle_float_r:double)
   n32 = "Transform (Modify) Bone - Bone: clavicle_cape_r"!(ComponentPose=@n30, Alpha:float=@n31, Translation="0,0,0", Rotation="0,-15,15", Scale="1,1,1") -> (Pose) {
    rotationMode=BMM_Additive
    rotationSpace=BCS_BoneSpace
    translationMode=BMM_Additive
    translationSpace=BCS_BoneSpace
    alphaBoolBlend {blendOption=Linear}
    alphaScaleBiasClamp {bInterpResult=true}
    boneToModify {boneName=clavicle_cape_r}
    rotation {roll=15, yaw=-15}
   }
   n33 = "Transform (Modify) Bone - Bone: cape_mid_02_r"(ComponentPose=@n32, Alpha=@n31, Translation="0,0,0", Rotation="0,-10,0", Scale="1,1,1") {
    rotationMode=BMM_Additive
    rotationSpace=BCS_BoneSpace
    translationMode=BMM_Additive
    translationSpace=BCS_BoneSpace
    alphaBoolBlend {blendOption=Linear}
    boneToModify {boneName=cape_mid_02_r}
    rotation {yaw=-10}
   }
   n34 = "Transform (Modify) Bone - Bone: clavicle_cape_l"(
    ComponentPose=@n33
    Alpha=Get clavicle_glider_float_l() -> (clavicle_glider_float_l:double)
    Translation="0,0,0"
    Rotation="0,-15,0"
    Scale="1,1,1"
   ) {
    rotationMode=BMM_Additive
    rotationSpace=BCS_BoneSpace
    alphaBoolBlend {blendOption=Linear}
    boneToModify {boneName=clavicle_cape_l}
    rotation {yaw=-15}
   }
   n35 = Get TrailControllerRot() -> (TrailControllerRot:Rotator)
   n36 = "Transform (Modify) Bone - Bone: cape_mid_02_l"(ComponentPose=@n34, Alpha=1, Translation="0,0,0", Rotation=@n35, Scale="1,1,1") {
    rotationMode=BMM_Additive
    rotationSpace=BCS_BoneSpace
    alphaBoolBlend {blendOption=Linear}
    boneToModify {boneName=cape_mid_02_l}
    rotation {yaw=-15}
   }
   n37 = "Transform (Modify) Bone - Bone: cape_mid_02_r"(ComponentPose=@n36, Alpha=1, Translation="0,0,0", Rotation=@n35, Scale="1,1,1") {
    rotationMode=BMM_Additive
    rotationSpace=BCS_BoneSpace
    alphaBoolBlend {blendOption=Linear}
    boneToModify {boneName=cape_mid_02_r}
    rotation {yaw=-15}
   }
   n38 = "Transform (Modify) Bone - Bone: cape_02_l"!(ComponentPose=@n37, Alpha:float=1, Translation="0,0,0", Rotation=@n35, Scale="1,1,1") -> (Pose) {
    rotationMode=BMM_Additive
    rotationSpace=BCS_BoneSpace
    alphaBoolBlend {blendOption=Linear}
    boneToModify {boneName=cape_02_l}
    rotation {yaw=-15}
   }
   n39 = "Transform (Modify) Bone - Bone: cape_02_r"!(ComponentPose=@n38, Alpha:float=1, Translation="0,0,0", Rotation=@n35, Scale="1,1,1") -> (Pose) {
    rotationMode=BMM_Additive
    rotationSpace=BCS_BoneSpace
    alphaBoolBlend {blendOption=Linear}
    boneToModify {boneName=cape_02_r}
    rotation {yaw=-15}
   }
   n40 = "Transform (Modify) Bone - Bone: cape_mid_02_l"!(ComponentPose=@n39, bAlphaBoolEnabled:bool=@n1, Translation="0,0,0", Rotation="0,-45,0", Scale="1,1,1") -> (Pose) {
    alphaInputType=Bool
    rotationMode=BMM_Additive
    rotationSpace=BCS_BoneSpace
    alphaBoolBlend {blendInTime=0.25, blendOption=Linear, blendOutTime=0.25}
    boneToModify {boneName=cape_mid_02_l}
    rotation {yaw=-45}
   }
   n41 = "Transform (Modify) Bone - Bone: cape_02_l"!(ComponentPose=@n40, bAlphaBoolEnabled:bool=@n1, Translation="0,0,0", Rotation="0,-30,0", Scale="1,1,1") -> (Pose) {
    alphaInputType=Bool
    rotationMode=BMM_Additive
    rotationSpace=BCS_BoneSpace
    alphaBoolBlend {blendInTime=0.25, blendOption=Linear, blendOutTime=0.25}
    boneToModify {boneName=cape_02_l}
    rotation {yaw=-30}
   }
   n42 = "Transform (Modify) Bone - Bone: cape_mid_02_r"!(ComponentPose=@n41, bAlphaBoolEnabled:bool=@n1, Translation="0,0,0", Rotation="0,-15,0", Scale="1,1,1") -> (Pose) {
    alphaInputType=Bool
    rotationMode=BMM_Additive
    rotationSpace=BCS_BoneSpace
    alphaBoolBlend {blendInTime=0.25, blendOption=Linear, blendOutTime=0.25}
    boneToModify {boneName=cape_mid_02_r}
    rotation {yaw=-15}
   }
   Save cached pose 'trailCache'(Pose=Component To Local(ComponentPose=@n42)) {cacheName=trailCache}
   n43 = "Transform (Modify) Bone - Bone: cape_01_l"(ComponentPose=Local To Component(LocalPose=Use cached pose 'Start_Cache'()), Alpha=1, Translation="0,0,0", Rotation="5,0,0", Scale="1,1,1") {
    rotationMode=BMM_Additive
    rotationSpace=BCS_BoneSpace
    alphaBoolBlend {blendOption=Linear}
    boneToModify {boneName=cape_01_l}
    rotation {pitch=5}
   }
   n44 = "Transform (Modify) Bone - Bone: cape_03_l"(ComponentPose=@n43, Alpha=1, Translation="0,0,0", Rotation="-10,0,0", Scale="1,1,1") {
    rotationMode=BMM_Additive
    rotationSpace=BCS_BoneSpace
    alphaBoolBlend {blendOption=Linear}
    boneToModify {boneName=cape_03_l}
    rotation {pitch=-10}
   }
   n45 = "Transform (Modify) Bone - Bone: cape_06_l"(ComponentPose=@n44, Alpha=1, Translation="0,0,0", Rotation="-25,0,0", Scale="1,1,1") {
    rotationMode=BMM_Additive
    rotationSpace=BCS_BoneSpace
    alphaBoolBlend {blendOption=Linear}
    boneToModify {boneName=cape_06_l}
    rotation {pitch=-25}
   }
   n46 = Trail controller - Bone: cape_mid_06_l!(ComponentPose=@n45, Alpha=1, RelaxationSpeedScale:float=2, FakeVelocity:Vector="0,0,0") -> (Pose) {
    bActorSpaceFakeVel=true
    bShowTrailLocation=true
    bUsePlanarLimit=true
    chainLength=5
    debugLifeTime=0.1
    maxDeltaTime=0.0333
    relaxationSpeedScale=1.2
    alphaScaleBiasClamp {bInterpResult=true, interpSpeedDecreasing=25, interpSpeedIncreasing=25}
    baseJoint {boneName=cape_root}
    rotationLimits [{limitMax=<180 180 180>}, {limitMax=<180 180 180>}, {limitMax=<180 180 180>}, {limitMax=<180 180 180>}, {limitMax=<180 180 180>}]
    rotationOffsets [<0 0 0>, <0 0 0>, <0 0 0>, <0 0 0>, <0 0 0>]
    trailBone {boneName=cape_mid_06_l}
    trailRelaxationSpeed {editorCurveData {keys [<at 0 15 RCIM_Linear>,<at 1 6 RCIM_Linear>]}}
   }
   n47 = Trail controller - Bone: cape_mid_06_r!(ComponentPose=@n46, Alpha=1, RelaxationSpeedScale:float=2, FakeVelocity:Vector="0,0,0") -> (Pose) {
    bActorSpaceFakeVel=true
    bShowTrailLocation=true
    bUsePlanarLimit=true
    chainLength=5
    debugLifeTime=0.1
    maxDeltaTime=0.0333
    relaxationSpeedScale=1.2
    alphaScaleBiasClamp {bInterpResult=true, interpSpeedDecreasing=25, interpSpeedIncreasing=25}
    baseJoint {boneName=cape_root}
    rotationLimits [{limitMax=<180 180 180>}, {limitMax=<180 180 180>}, {limitMax=<180 180 180>}, {limitMax=<180 180 180>}, {limitMax=<180 180 180>}]
    rotationOffsets [<0 0 0>, <0 0 0>, <0 0 0>, <0 0 0>, <0 0 0>]
    trailBone {boneName=cape_mid_06_r}
    trailRelaxationSpeed {editorCurveData {keys [<at 0 15 RCIM_Linear>,<at 1 6 RCIM_Linear>]}}
   }
   n48 = Trail controller - Bone: cape_07_l!(ComponentPose=@n47, Alpha=1, RelaxationSpeedScale:float=2, FakeVelocity:Vector="0,0,0") -> (Pose) {
    bActorSpaceFakeVel=true
    bShowTrailLocation=true
    bUsePlanarLimit=true
    chainLength=6
    debugLifeTime=0.1
    maxDeltaTime=0.0333
    relaxationSpeedScale=1.2
    alphaScaleBiasClamp {bInterpResult=true, interpSpeedDecreasing=25, interpSpeedIncreasing=25}
    baseJoint {boneName=cape_root}
    rotationLimits [
     {limitMax=<180 180 180>}
     {limitMax=<180 180 180>}
     {limitMax=<180 180 180>}
     {limitMax=<180 180 180>}
     {limitMax=<180 180 180>}
     {limitMax=<180 180 180>}
    ]
    rotationOffsets [<0 0 0>, <0 0 0>, <0 0 0>, <0 0 0>, <0 0 0>, <0 0 0>]
    trailBone {boneName=cape_07_l}
    trailRelaxationSpeed {editorCurveData {keys [<at 0 15 RCIM_Linear>,<at 1 6 RCIM_Linear>]}}
   }
   n49 = Trail controller - Bone: cape_07_r!(ComponentPose=@n48, Alpha=1, RelaxationSpeedScale:float=2, FakeVelocity:Vector="0,0,0") -> (Pose) {
    bActorSpaceFakeVel=true
    bShowTrailLocation=true
    bUsePlanarLimit=true
    chainLength=6
    debugLifeTime=0.1
    maxDeltaTime=0.0333
    relaxationSpeedScale=1.2
    alphaScaleBiasClamp {bInterpResult=true, interpSpeedDecreasing=25, interpSpeedIncreasing=25}
    baseJoint {boneName=cape_root}
    rotationLimits [
     {limitMax=<180 180 180>}
     {limitMax=<180 180 180>}
     {limitMax=<180 180 180>}
     {limitMax=<180 180 180>}
     {limitMax=<180 180 180>}
     {limitMax=<180 180 180>}
    ]
    rotationOffsets [<0 0 0>, <0 0 0>, <0 0 0>, <0 0 0>, <0 0 0>, <0 0 0>]
    trailBone {boneName=cape_07_r}
    trailRelaxationSpeed {editorCurveData {keys [<at 0 15 RCIM_Linear>,<at 1 6 RCIM_Linear>]}}
   }
   Save cached pose 'dbnoCache'(Pose=Component To Local(ComponentPose=@n49)) {cacheName=dbnoCache}
   n50 = Get GravityOverride()
   n51 = RigidBody(
    ComponentPose=Local To Component(LocalPose=Use cached pose 'Start_Cache'())
    bAlphaBoolEnabled=Get cloth_disabled_backbling_equipped_not_driving() -> (cloth_disabled_backbling_equipped_not_driving:bool)
    OverrideWorldGravity=@n50
   ) {
    alphaInputType=Bool
    bClampLinearTranslationLimitToRefPose=true
    bOverrideWorldGravity=true
    overridePhysicsAsset=&/F_MED_Parfait_Cape_Backbling_Physics
    simulationSpace=BaseBoneSpace
    alphaBoolBlend {blendOption=Linear}
    baseBoneRef {boneName=root}
    componentAppliedLinearAccClamp=<3000 3000 3000>
    componentLinearAccScale=<0.2 0.2 0.3>
    componentLinearVelScale=<0.2 0.2 0.3>
    overrideWorldGravity {z=-980}
    simSpaceSettings {
     maxAngularAcceleration=50
     maxAngularVelocity=1
     maxLinearAcceleration=300
     maxLinearVelocity=150
     velocityScaleZ=0
     worldAlpha=1
     externalLinearDragV=<0.2 0.2>
    }
   }
   n52 = RigidBody(
    ComponentPose=@n51
    bAlphaBoolEnabled=Get cloth_disabled_backbling_not_equipped_not_driving() -> (cloth_disabled_backbling_not_equipped_not_driving:bool)
    OverrideWorldGravity=@n50
   ) {
    alphaInputType=Bool
    bClampLinearTranslationLimitToRefPose=true
    bOverrideWorldGravity=true
    overridePhysicsAsset=&/F_MED_Parfait_Cape_RBAN_Physics
    simulationSpace=BaseBoneSpace
    alphaBoolBlend {blendOption=Linear}
    baseBoneRef {boneName=root}
    componentAppliedLinearAccClamp=<3000 3000 3000>
    componentLinearAccScale=<0.2 0.2 0.3>
    componentLinearVelScale=<0.2 0.2 0.3>
    overrideWorldGravity {z=-980}
    simSpaceSettings {
     maxAngularAcceleration=50
     maxAngularVelocity=1
     maxLinearAcceleration=300
     maxLinearVelocity=150
     velocityScaleZ=0
     worldAlpha=1
     externalLinearDragV=<0.2 0.2>
    }
   }
   n53 = Component To Local(ComponentPose=@n52)
   n54 = RigidBody(ComponentPose=Local To Component(LocalPose=Use cached pose 'Start_Cache'()), OverrideWorldGravity=@n50) {
    alphaInputType=Bool
    bClampLinearTranslationLimitToRefPose=true
    bOverrideWorldGravity=true
    overridePhysicsAsset=&/F_MED_Parfait_Cape_Backbling_Skydive_Physics
    simulationSpace=BaseBoneSpace
    alphaBoolBlend {blendOption=Linear}
    baseBoneRef {boneName=root}
    componentAppliedLinearAccClamp=<3000 3000 3000>
    componentLinearAccScale=<0.2 0.2 0.3>
    componentLinearVelScale=<0.2 0.2 0.3>
    overrideWorldGravity {z=-980}
    simSpaceSettings {
     maxAngularAcceleration=50
     maxAngularVelocity=1
     maxLinearAcceleration=300
     maxLinearVelocity=150
     velocityScaleZ=0
     worldAlpha=1
     externalLinearDragV=<0.2 0.2>
    }
   }
   n55 = "Transform (Modify) Bone - Bone: cape_03_r"(ComponentPose=@n54, Alpha=1, Translation="0,0,0", Rotation="-15,-25.000,-0.000", Scale="1,1,1") {
    rotationMode=BMM_Additive
    rotationSpace=BCS_BoneSpace
    translationMode=BMM_Additive
    translationSpace=BCS_BoneSpace
    alphaBoolBlend {blendOption=Linear}
    boneToModify {boneName=cape_03_r}
    rotation {pitch=-15, yaw=-25}
   }
   n56 = "Transform (Modify) Bone - Bone: cape_mid_02_l"!(ComponentPose=@n55, Alpha:float=1, Translation="0,0,0", Rotation="15,-25,0", Scale="1,1,1") -> (Pose) {
    rotationMode=BMM_Additive
    rotationSpace=BCS_BoneSpace
    translationMode=BMM_Additive
    translationSpace=BCS_BoneSpace
    alphaBoolBlend {blendOption=Linear}
    boneToModify {boneName=cape_mid_02_l}
    rotation {pitch=15, yaw=-25}
   }
   n57 = "Transform (Modify) Bone - Bone: cape_mid_02_r"!(ComponentPose=@n56, Alpha:float=1, Translation="0,0,0", Rotation="-15,-25.000,-0.000", Scale="1,1,1") -> (Pose) {
    rotationMode=BMM_Additive
    rotationSpace=BCS_BoneSpace
    translationMode=BMM_Additive
    translationSpace=BCS_BoneSpace
    alphaBoolBlend {blendOption=Linear}
    boneToModify {boneName=cape_mid_02_r}
    rotation {pitch=-15, yaw=-25}
   }
   n58 = "Transform (Modify) Bone - Bone: cape_03_l"(ComponentPose=@n57, Alpha=1, Translation="0,0,0", Rotation="15,-25,0", Scale="1,1,1") {
    rotationMode=BMM_Additive
    rotationSpace=BCS_BoneSpace
    translationMode=BMM_Additive
    translationSpace=BCS_BoneSpace
    alphaBoolBlend {blendOption=Linear}
    boneToModify {boneName=cape_03_l}
    rotation {pitch=15, yaw=-25}
   }
   n59 = Component To Local(ComponentPose=@n58)
   n60 = Get bIsSkydiveFloating()
   n61 = "Transform (Modify) Bone - Bone: clavicle_cape_l"(
    ComponentPose=Local To Component(LocalPose=Blend Poses by bool(BlendPose_0=@n59, BlendPose_1=@n53, bActiveValue=@n60) {blendType=Linear})
    Alpha=Get clavicle_float_l()
    Translation="0,0,0"
    Rotation="-0.000,-25.000,-0.000"
    Scale="1,1,1"
   ) {
    rotationMode=BMM_Additive
    rotationSpace=BCS_BoneSpace
    translationMode=BMM_Additive
    translationSpace=BCS_BoneSpace
    alphaBoolBlend {blendOption=Linear}
    boneToModify {boneName=clavicle_cape_l}
    rotation=<rot 0 -25 0>
   }
   n62 = "Transform (Modify) Bone - Bone: clavicle_cape_r"(ComponentPose=@n61, Alpha=Get clavicle_float_r(), Translation="0,0,0", Rotation="0,-25,0", Scale="1,1,1") {
    rotationMode=BMM_Additive
    rotationSpace=BCS_BoneSpace
    translationMode=BMM_Additive
    translationSpace=BCS_BoneSpace
    alphaBoolBlend {blendOption=Linear}
    boneToModify {boneName=clavicle_cape_r}
    rotation {yaw=-25}
   }
   n63 = "Transform (Modify) Bone - Bone: clavicle_cape_l"(ComponentPose=@n62, Alpha=Get clavicle_glider_float_l(), Translation="0,0,0", Rotation="0,-15.074,0", Scale="1,1,1") {
    rotationMode=BMM_Additive
    rotationSpace=BCS_BoneSpace
    alphaBoolBlend {blendOption=Linear}
    boneToModify {boneName=clavicle_cape_l}
    rotation {yaw=-15.074}
   }
   n64 = Get In Car() -> (In Car:bool)
   n65 = "Transform (Modify) Bone - Bone: cape_02_l"(ComponentPose=@n63, bAlphaBoolEnabled=@n64, Translation="0,0,0", Rotation="0,0,0", Scale="0.25,0.25,0.25") {
    alphaInputType=Bool
    rotationMode=BMM_Additive
    rotationSpace=BCS_BoneSpace
    scaleMode=BMM_Replace
    translationMode=BMM_Additive
    translationSpace=BCS_BoneSpace
    alphaBoolBlend {blendOption=Linear}
    boneToModify {boneName=cape_02_l}
    scale=<0.25 0.25 0.25>
   }
   n66 = "Transform (Modify) Bone - Bone: cape_02_r"!(ComponentPose=@n65, bAlphaBoolEnabled:bool=@n64, Translation="0,0,0", Rotation="0,0,0", Scale="0.25,0.25,0.25") -> (Pose) {
    alphaInputType=Bool
    rotationMode=BMM_Additive
    rotationSpace=BCS_BoneSpace
    scaleMode=BMM_Replace
    translationMode=BMM_Additive
    translationSpace=BCS_BoneSpace
    alphaBoolBlend {blendOption=Linear}
    boneToModify {boneName=cape_02_r}
    scale=<0.25 0.25 0.25>
   }
   n67 = "Transform (Modify) Bone - Bone: cape_mid_02_l"!(ComponentPose=@n66, bAlphaBoolEnabled:bool=@n64, Translation="0,0,0", Rotation="0,0,0", Scale="0.25,0.25,0.25") -> (Pose) {
    alphaInputType=Bool
    rotationMode=BMM_Additive
    rotationSpace=BCS_BoneSpace
    scaleMode=BMM_Replace
    translationMode=BMM_Additive
    translationSpace=BCS_BoneSpace
    alphaBoolBlend {blendOption=Linear}
    boneToModify {boneName=cape_mid_02_l}
    scale=<0.25 0.25 0.25>
   }
   n68 = "Transform (Modify) Bone - Bone: cape_mid_02_r"!(ComponentPose=@n67, bAlphaBoolEnabled:bool=@n64, Translation="0,0,0", Rotation="0,0,0", Scale="0.25,0.25,0.25") -> (Pose) {
    alphaInputType=Bool
    rotationMode=BMM_Additive
    rotationSpace=BCS_BoneSpace
    scaleMode=BMM_Replace
    translationMode=BMM_Additive
    translationSpace=BCS_BoneSpace
    alphaBoolBlend {blendOption=Linear}
    boneToModify {boneName=cape_mid_02_r}
    scale=<0.25 0.25 0.25>
   }
   Save cached pose 'RBAN Cache'(Pose=Component To Local(ComponentPose=@n68)) {cacheName=RBAN Cache}
   Get bIsValetDriver() -> (bIsValetDriver:bool)
   Get skydive()
   Get bIsInVehicle() -> (bIsInVehicle:bool)
  ]}
 EventGraph {nodes [
   n1 = GetPartSkeletalMeshComponent()
   n2 = Cloth Controller Function DEPRECATED!(
    execute=@n5.then
    CharacterPartAnimInstance:Custom Character Part Anim Instance Object Reference=Self-Reference()
    ClothingAssetName:String=F_MED_Parfait_Clothing_1
    Cloth Paramaters:Cloth Params Struct Structure=Get Cloth Parameters() -> (Cloth Parameters:Cloth Params Struct Structure)
    Cloth Core Settings:Cloth Core Setting Struct Structure=Get Cloth Core Settings() -> (Cloth Core Settings:Cloth Core Setting Struct Structure)
    ClothUpAxis:Vector="0,0,1"
    OnGround_Moving_SpeedLimit:double=4
    OnGround_Moving_AngularVel_Max:double=0.05
    OnGround_Moving_FictitiousVel_Max:double=0.6
    LinVelScale_FwdBwd_Clamp:double=0
    LinVelScale_UpDwn_Clamp:double=0.7
    DeltaTime:double=Get DeltaTimeLocal() -> (DeltaTimeLocal:double)
    GameTimeInSeconds:double=GetGameTimeInSeconds()
    WindGustsParams:Wind Gust Params Struct Structure=Get Wind Gust Params Final() -> (Wind Gust Params Final:Wind Gust Params Struct Structure)
    WindGustLocalSettings:Wind Gust Local Setting Struct Structure=Get Gust Local Settings() -> (Gust Local Settings:Wind Gust Local Setting Struct Structure)
   ) -> (
    then
    CurrentStateOut:String
    Cloth Settings Out:Cloth Setting Struct Structure
    Cloth Core Settings Out:Cloth Core Setting Struct Structure
    WindGustSettings Out:Wind Gust Setting Struct Structure
    WindGustLocalSettings Out:Wind Gust Local Setting Struct Structure
    GustStrength - WindVelocityOut:Vector
    LinearVelVector-DELETE LATER:Vector
   )
   DrawVectorFromJointFunction!(
    execute
    SkelMeshComponent:Scene Component Object Reference=@n1
    JointName:Name=spine_05
    LineEnd:Vector=@n2.GustStrength - WindVelocityOut
    JointOffset:Vector="0, 0, 0"
    LineColor:Linear Color Structure=(R=1,G=0,B=0,A=1)
    Thickness:double=2
    Foreground:EDrawDebugSceneDepthPriorityGroup Enum=Foreground
   ) -> (then)
   > DrawVectorFromJointFunction(
    SkelMeshComponent=@n1
    JointName=spine_05
    LineEnd=Break Cloth Setting Struct!(ClothSettingStruct:"Cloth Setting Struct Structure (by ref)"=@n2.Cloth Settings Out) -> (
     LinearVelocityScale_7_#1:Vector=~
     AngularVelocityScale_6_#2:float=~
     FictitiousAngularScale_10_#3:float=~
     GravityOverride_15_#4:Vector=~
     DragCoefficient_23_#5:Vector 2D Structure
     LiftCoefficient_24_#6:Vector 2D Structure
     JointName_27_#7:Name
    ) {is=title K2Node_BreakStruct}.GravityOverride_15_#4
    JointOffset="0,40,40"
    LineColor=(R=0.0192,G=0,B=1,A=1)
    Thickness=2
    Foreground=Foreground
   )
   Event BlueprintUpdateAnimation()
   > n3 = Sequence!(execute) -> (then_0, then_1, then_2:Exec, then_3:Exec, then_4:Exec, then_5:Exec, then_6:Exec, then_7:Exec, then_8:Exec)
   > n4 = Gravity Override State Function DEPRECATED(CharacterPartAnimInstance=Self-Reference(), GameTimeInSeconds=GetGameTimeInSeconds(), GravityOverrideParams=Get Gravity Override Params())
   > n5 = Set GravityOverride(GravityOverride=@n4)
   n6 = PlayerMovement(execute=@n3.then_1, AnimInstance=Self-Reference(), XDotProductCoefficient=1, YDotProductCoefficient=1, ZDotProductCoefficient=1)
   n7 = MapRangeClamped(Value=@n6.YMovement, InRangeA=0, InRangeB=1)
   n8 = "Absolute (Float)"(A=@n6)
   > Set BwdMovement(BwdMovement:double=MapRangeClamped(Value=float + float(A=@n8, B=@n7) {is=title K2Node_CommutativeAssociativeBinaryOperator}, InRangeA=0, InRangeB=1.4)) -> (Output_Get:double)
   > n9 = Branch(Condition=Get bIsSwimSprinting() -> (bIsSwimSprinting:bool))
   > Set TrailControllerRot(TrailControllerRot:Rotator="0,-15, 0") -> (Output_Get:Rotator)
   n10 = Branch(execute=@n9.else, Condition=Get bIsSkyDiving())
   > n11 = Branch(Condition=Get bIsParachuteOpened())
   > Set TrailControllerRot(TrailControllerRot="0,-15, 0")
   > Set BwdMovement(BwdMovement=0.25)
   n12 = Branch(execute=@n11.else, Condition="float > float"(A=Get PawnVelocityZ(), B=50) {is=title K2Node_CallFunction})
   > Set BwdMovement(BwdMovement=float * float(A=Get BwdMovement(), B=0))
   > Set TrailControllerRot(TrailControllerRot="0,-25, 0")
   n13 = Branch(execute=@n12.else, Condition=Get bIsSkydiveDiveMode())
   > Set TrailControllerRot(TrailControllerRot="0,-15, 0")
   Set TrailControllerRot(execute=@n13.else, TrailControllerRot="0,-25, 0")
   Set TrailControllerRot(execute=@n10.else, TrailControllerRot="0,-12.5, 0")
   > Set BwdMovement(BwdMovement=float * float(A=Get BwdMovement(), B=1))
   n14 = Branch(execute=@n3.then_2, Condition=Get bIsInFrontEnd())
   n15 = CheckForBackbling()
   n16 = Set backbling_equipped(execute=@n14.else, backbling_equipped:bool=@n15) -> (Output_Get:bool)
   > Set cloth_disabled(cloth_disabled:bool=NOT Boolean(A=Get bClothEnabled())) -> (Output_Get:bool)
   n17 = CharacterItemDefinitionHasAnyMetaTag(Tags=Make Array!("[0]"=(TagName="Cosmetics.Variant.Property.Stage2"), "[1]":Gameplay Tag Structure=(TagName="Cosmetics.Variant.Property.Stage3")) -> (Array))
   > Set cape_removed(cape_removed:bool=@n17) -> (Output_Get:bool)
   > n18 = Branch(Condition=OR Boolean!(
     A=Get In Car()
     B=Get bIsGolfCartVehicleBackPassenger() -> (bIsGolfCartVehicleBackPassenger:bool)
     C:bool=Get bIsJackalVehicleDriver() -> (bIsJackalVehicleDriver:bool)
     D:bool=Get bIsInShoppingCart() -> (bIsInShoppingCart:bool)
     E:bool=Get bIsInCannon() -> (bIsInCannon:bool)
     F:bool=Get bIsMeatballPassenger() -> (bIsMeatballPassenger:bool)
     G:bool=@n15
     H:bool=@n16
     I:bool=@n17
    ) -> (ReturnValue))
   n19 = GetPartSkeletalMeshComponent()
   > Set ClothBlendWeight!(execute, ClothBlendWeight:float=0, self:Skeletal Mesh Component Object Reference=@n19) -> (then, Output_Get:float)
   > Do Once(Reset=@n21.then)
   > n20 = SuspendClothingSimulation(self:Skeletal Mesh Component Object Reference=@n19)
   Do Once(execute=@n22.then, Reset=@n20.then)
   > n21 = ResumeClothingSimulation(self:Skeletal Mesh Component Object Reference=@n19)
   n22 = Set ClothBlendWeight(execute=@n18.else, ClothBlendWeight=1, self=@n19)
   Set jump_falling(execute=@n3.then_3, jump_falling:bool=OR Boolean!(A=Get bIsJumping(), B=Get bIsFalling()) -> (ReturnValue)) -> (Output_Get:bool)
   > Set skydive(skydive:bool=OR Boolean!(A=Get bIsSkyDiving(), B=Get bIsSkydiveFloating(), C:bool=Get bIsSkydiveDiveMode()) -> (ReturnValue)) -> (Output_Get:bool)
   n23 = Branch(execute=@n3.then_4, Condition=Get bIsInVehicle())
   > n24 = Branch(Condition=OR Boolean!(A=Get bIsValetDriver(), B=Get bIsValetPassenger() -> (bIsValetPassenger:bool)) -> (ReturnValue))
   > n25 = Branch(Condition=Get bVehicleSeatForcesCrouch() -> (bVehicleSeatForcesCrouch:bool))
   Set In Car(execute=[@n25.then, @n24.else, @n23.else], In Car:bool=false) -> (Output_Get:bool)
   Set In Car(execute=@n25.else, In Car=true)
   n26 = Get In Car()
   Set cloth_disabled_backbling_equipped_not_driving(
    execute=@n3.then_5
    cloth_disabled_backbling_equipped_not_driving:bool=AND Boolean!(A=Get backbling_equipped() -> (backbling_equipped:bool), B=Get cloth_disabled() -> (cloth_disabled:bool), C:bool=NOT Boolean(A=@n26)) -> (ReturnValue)
   ) -> (Output_Get:bool)
   n27 = Get In Car()
   n28 = Get backbling_equipped()
   > Set cloth_disabled_backbling_not_equipped_not_driving(cloth_disabled_backbling_not_equipped_not_driving:bool=AND Boolean(A=NOT Boolean(A=@n28), B=Get cloth_disabled(), C=NOT Boolean(A=@n27))) -> (Output_Get:bool)
   > Set cloth_disabled_rban_enabled(cloth_disabled_rban_enabled:bool=AND Boolean!(A=Get cloth_disabled(), B=Get bRigidBodyEnabled()) -> (ReturnValue)) -> (Output_Get:bool)
   > Set frontend_skydive(frontend_skydive:bool=AND Boolean(A=Get skydive(), B=Get bIsInFrontEnd())) -> (Output_Get:bool)
   n29 = GetAnimInstance(self=GetOwnerSkeletalMeshComponent())
   > Set parfait_built_in_emote(parfait_built_in_emote:bool=OR Boolean!(
     A=Montage_IsPlaying(self=@n29, Montage=/BRCosmetics/Animation/Game/MainPlayer/Emotes/Trifle/Emote_Trifle_CMF_Equip_M)
     B=Montage_IsPlaying(Montage=/BRCosmetics/Animation/Game/MainPlayer/Emotes/Trifle/Emote_Trifle_CMF_NoEquip_M)
     C:bool=Montage_IsPlaying(Montage=/BRCosmetics/Animation/Game/MainPlayer/Emotes/Trifle/Emote_Trifle_CMM_Equipt_M)
     D:bool=Montage_IsPlaying(Montage=/BRCosmetics/Animation/Game/MainPlayer/Emotes/Trifle/Emote_Trifle_CMM_NoEquipt_M)
    ) -> (ReturnValue)) -> (Output_Get:bool)
   Branch(execute=@n3.then_6, Condition=OR Boolean!(A=Get bIsInFrontEnd(), B=Get bIsPlayingEmote()) -> (ReturnValue))
   > Set BwdMovement(BwdMovement=1)
   n30 = Branch(execute=@n3.then_7, Condition=Get skydive())
   > Set clavicle_float_r(clavicle_float_r:double=0) -> (Output_Get:double)
   > Set clavicle_float_l(clavicle_float_l:double=0) -> (Output_Get:double)
   n31 = GetDeltaTransformFromRefPose(self=GetPartSkeletalMeshComponent(), BoneName=upperarm_l, BaseName=spine_05)
   n32 = MapRangeClamped(Value=BreakRotator(InRot=BreakTransform(InTransform=@n31).Rotation).Yaw, InRangeA=10, InRangeB=-30)
   > Set clavicle_glider_float_l(clavicle_glider_float_l:double=@n32) -> (Output_Get:double)
   n33 = BreakTransform(InTransform=GetDeltaTransformFromRefPose(self=GetPartSkeletalMeshComponent(), BoneName=upperarm_r, BaseName=spine_05))
   Set clavicle_float_r(execute=@n30.else, clavicle_float_r=MapRangeClamped(Value=BreakRotator(InRot=@n33.Rotation).Yaw, InRangeA=0, InRangeB=-40))
   > Set clavicle_float_l(clavicle_float_l=@n32)
   > Set clavicle_glider_float_l(clavicle_glider_float_l=0)
   Event OnCharacterCustomizationLoaded()
   > n34 = Sequence!(execute) -> (then_0, then_1)
   > Set backbling_equipped(backbling_equipped=CheckForBackbling())
   > Set cloth_disabled(cloth_disabled=NOT Boolean(A=Get bClothEnabled()))
   > Set cape_removed(cape_removed=CharacterItemDefinitionHasAnyMetaTag(Tags=Make Array("[0]"=(TagName="Cosmetics.Variant.Property.Stage2"), "[1]"=(TagName="Cosmetics.Variant.Property.Stage3"))))
   n35 = GetPartSkeletalMeshComponent()
   > Set ClothBlendWeight(ClothBlendWeight=0, self=@n35)
   > Delay(Duration=0.1)
   > Set ClothBlendWeight(ClothBlendWeight=1, self=@n35)
   Branch(execute=@n34.then_1, Condition=Get skydive())
   > Delay(Duration=0.3)
   > ResetDynamics(InTeleportType=TeleportPhysics)
   > ResetDynamics(InTeleportType=ResetPhysics)
   Event OnBeginSkydiving()
   > Set TrailControllerRot(TrailControllerRot="0,-65,0")
   > Delay(Duration=0.3)
   > ResetDynamics(InTeleportType=TeleportPhysics)
   > ResetDynamics(InTeleportType=ResetPhysics)
   Event OnEndSkydiving()
   > Set TrailControllerRot(TrailControllerRot="0,-10,0")
   n36 = GetPartSkeletalMeshComponent()
   Set ClothBlendWeight(ClothBlendWeight=0, self=@n36)
   > Delay(Duration=0.1)
   > Set ClothBlendWeight(ClothBlendWeight=1, self=@n36)
   NOT Boolean(A=Get skydive())
   Get bVehicleSeatForcesCrouch()
   Get bIsValetDriver()
   Get bIsValetDriver()
   Get bIsValetDriver()
   Set Gust Local Settings(execute=@n2.then, Gust Local Settings:Wind Gust Local Setting Struct Structure=@n2.WindGustLocalSettings Out) -> (Output_Get:Wind Gust Local Setting Struct Structure)
  ]}
]
variables [
 backbling_equipped {type=boolean}
 BwdMovement {type=number}
 cape_removed {type=boolean}
 clavicle_float_l {type=number}
 clavicle_float_r {type=number}
 clavicle_glider_float_l {type=number}
 Cloth Core Settings {
  type=ClothCoreSettingStruct
  default {
   collisionThickness=1.5
   dampingCoefficient=0.1
   frictionCoefficient=0.2
   localDampingCoefficient=0.1
   numberIterations=2
   numberSubsteps=1
   selfCollisionThickness=0
   animDriveDamping=<0 1>
   animDriveStiffness=<0 1>
   areaStiffness=<1 1>
   bendingStiffness=<1 1>
   edgeStiffness=<1 1>
   tetherStiffness=<1 1>
  }
 }
 Cloth Parameters {
  type=ClothParamsStruct
  default {
   dBNO {
    angularVelocityScale=0.6
    fictitiousAngularScale=0.1
    jointName=pelvis
    dragCoefficient=<0.2 1>
    gravityOverride=<-980 400 0>
    liftCoefficient=<0 0.5>
    linearVelocityScale=<0.3 0.3 0.1>
   }
   dBNO_Carried {
    angularVelocityScale=0.6
    fictitiousAngularScale=0.1
    jointName=pelvis
    dragCoefficient=<0.2 1>
    gravityOverride=<-980 -200 0>
    liftCoefficient=<0 0.5>
    linearVelocityScale=<0.3 0.3 0.1>
   }
   dBNO_Impulse {
    angularVelocityScale=0
    fictitiousAngularScale=0
    jointName=pelvis
    dragCoefficient=<0 0.2>
    gravityOverride=<-3000 -2000 0>
    liftCoefficient=<0 0.1>
    linearVelocityScale=<0.3 0.1 0.3>
   }
   dBNO_RollBwd {
    angularVelocityScale=0
    fictitiousAngularScale=0
    jointName=pelvis
    dragCoefficient=<0 0.2>
    gravityOverride=<-2000 0 0>
    liftCoefficient=<0 0.1>
    linearVelocityScale=<0.3 0.1 0.3>
   }
   dBNO_RollFwd {
    angularVelocityScale=0
    fictitiousAngularScale=0
    jointName=pelvis
    dragCoefficient=<0 0.2>
    gravityOverride=<-980 0 0>
    liftCoefficient=<0 0.1>
    linearVelocityScale=<0.3 0.1 0.3>
   }
   dBNO_Tumbling {
    angularVelocityScale=0
    fictitiousAngularScale=0
    jointName=pelvis
    dragCoefficient=<0 0.2>
    gravityOverride=<-2500 -300 0>
    liftCoefficient=<0 0.1>
    linearVelocityScale=<0.3 0.1 0.3>
   }
   emoteOrMelee {
    angularVelocityScale=0.05
    fictitiousAngularScale=0.1
    jointName=root
    dragCoefficient=<2.5 2.5>
    gravityOverride=<0 0 -980>
    liftCoefficient=<0.2 0.2>
    linearVelocityScale=<0.4 0.4 0.4>
   }
   front_End {
    angularVelocityScale=0.05
    fictitiousAngularScale=0.6
    jointName=root
    dragCoefficient=<0.2 0.2>
    gravityOverride=<0 -200 -980>
    liftCoefficient=<0.1 0.1>
    linearVelocityScale=<0.1 0.1 0.1>
   }
   onGround_Crouch_Moving {
    angularVelocityScale=0.1
    fictitiousAngularScale=0.3
    jointName=root
    dragCoefficient=<2 2>
    gravityOverride=<0 0 -980>
    liftCoefficient=<0.5 0.5>
    linearVelocityScale=<0.6 0.6 0.5>
   }
   onGround_Falling {
    angularVelocityScale=0.6
    fictitiousAngularScale=0
    jointName=root
    dragCoefficient=<0 0.2>
    gravityOverride=<0 -300 -980>
    liftCoefficient=<0 0.1>
    linearVelocityScale=<1 1 0.3>
   }
   onGround_Moving {
    angularVelocityScale=0.6
    fictitiousAngularScale=0.1
    jointName=root
    dragCoefficient=<0.2 0.2>
    gravityOverride=<0 -300 -980>
    liftCoefficient=<0.2 0.2>
    linearVelocityScale=<0.35 0.35 0.5>
   }
   onGround_Standing {
    angularVelocityScale=0.05
    fictitiousAngularScale=0.1
    jointName=root
    dragCoefficient=<0.5 0.5>
    gravityOverride=<0 0 -980>
    liftCoefficient=<0.5 0.5>
    linearVelocityScale=<0.3 0.3 0.3>
   }
   rail_Grinding {
    angularVelocityScale=0.2
    fictitiousAngularScale=0
    jointName=root
    dragCoefficient=<0.1 0.2>
    gravityOverride=<50 -500 -400>
    liftCoefficient=<0 0.5>
    linearVelocityScale=<0.1 0.1 0.1>
   }
   riding_Moving {
    angularVelocityScale=0.05
    fictitiousAngularScale=1
    jointName=root
    dragCoefficient=<0.5 0.5>
    gravityOverride=<0 0 -980>
    liftCoefficient=<0.5 0.5>
    linearVelocityScale=<0.4 0.6 0.3>
   }
   riding_Moving_Targeting {
    angularVelocityScale=0.05
    fictitiousAngularScale=0.7
    jointName=root
    dragCoefficient=<0.5 0.5>
    gravityOverride=<0 0 -980>
    liftCoefficient=<0.5 0.5>
    linearVelocityScale=<0.2 0.6 0.2>
   }
   riding_Standing {
    angularVelocityScale=0.05
    fictitiousAngularScale=1
    jointName=root
    dragCoefficient=<0.5 0.5>
    gravityOverride=<0 -450 -980>
    liftCoefficient=<0.5 0.5>
    linearVelocityScale=<0.6 0.6 0.3>
   }
   skydive_DivingDown {
    angularVelocityScale=0.05
    fictitiousAngularScale=0
    jointName=pelvis
    dragCoefficient=<1 1>
    gravityOverride=<-250 0 0>
    liftCoefficient=<0.5 0.5>
    linearVelocityScale=<0.05 0.05 0.05>
   }
   skydive_FloatingDown {
    angularVelocityScale=0.1
    fictitiousAngularScale=0
    jointName=root
    dragCoefficient=<2 2>
    gravityOverride=<0 -800 200>
    liftCoefficient=<2 2>
    linearVelocityScale=<0.2 0.2 0.1>
   }
   skydive_GoingUp {
    angularVelocityScale=0.1
    fictitiousAngularScale=0
    jointName=pelvis
    dragCoefficient=<0.1 0.1>
    gravityOverride=<-10000 -400 0>
    liftCoefficient=<0.05 0.05>
    linearVelocityScale=<0.01 0.1 0.1>
   }
   skydive_Parachuting {
    angularVelocityScale=0.6
    fictitiousAngularScale=0.1
    jointName=root
    dragCoefficient=<1.5 1.5>
    gravityOverride=<0 -600 -500>
    liftCoefficient=<1 1>
    linearVelocityScale=<0.1 0.1 0.2>
   }
   skydive_Surfing {
    angularVelocityScale=0.6
    fictitiousAngularScale=0.1
    jointName=root
    dragCoefficient=<1.5 1.5>
    gravityOverride=<0 -600 -500>
    liftCoefficient=<1 1>
    linearVelocityScale=<0.1 0.1 0.3>
   }
   swimming {
    angularVelocityScale=0.6
    fictitiousAngularScale=0
    jointName=root
    dragCoefficient=<2.5 2.5>
    gravityOverride=<0 0 0>
    liftCoefficient=<0.1 0.1>
    linearVelocityScale=<0.6 0.6 0.5>
   }
   tactical_Sprint {
    angularVelocityScale=0.6
    fictitiousAngularScale=0.1
    jointName=root
    dragCoefficient=<2 2>
    gravityOverride=<0 0 -980>
    liftCoefficient=<0.5 0.5>
    linearVelocityScale=<0.6 0.6 0.5>
   }
   ziplining {
    angularVelocityScale=0.6
    fictitiousAngularScale=0
    jointName=root
    dragCoefficient=<0 1.5>
    gravityOverride=<0 0 -980>
    liftCoefficient=<0 0.3>
    linearVelocityScale=<0.4 0.4 0.5>
   }
  }
 }
 cloth_disabled {type=boolean}
 cloth_disabled_backbling_equipped_not_driving {type=boolean}
 cloth_disabled_backbling_not_equipped_not_driving {type=boolean}
 cloth_disabled_rban_enabled {type=boolean}
 DeltaTimeLocal {type=number}
 frontend_not_emote {type=boolean}
 frontend_skydive {type=boolean}
 Gravity Override Params {type=GravityOverrideParamsStruct}
 GravityOverride {type=Vector, default=<0 0 -980>}
 Gust Local Settings {type=WindGustLocalSettingStruct, default {elapsedTime=0, finalGustStrengthLocal=0, gustDuration=0, gustStrength=0, timeBetweenGusts=0}}
 In Car {type=boolean}
 jump_falling {type=boolean}
 parfait_built_in_emote {type=boolean}
 skydive {type=boolean}
 TrailControllerRot {type=Rotator}
 Wind Gust Params Final {
  type=WindGustParamsStruct
  default {
   dBNO {
    maxDownInterpSpeed=5
    maxGustDuration=2
    maxGustStrength=7
    maxTimeBetweenGusts=7
    maxUpInterpSpeed=5
    minDownInterpSpeed=2
    minGustDuration=0.5
    minGustStrength=3
    minTimeBetweenGusts=0
    minUpInterpSpeed=2
    windTurbulenceScalarA=0.1
    windTurbulenceScalarB=1
    windAmplitude=<0 0 0>
    windFrequency=<0 0 0>
    windOffset=<0 0 0>
   }
   dBNO_Carried {
    maxDownInterpSpeed=5
    maxGustDuration=2
    maxGustStrength=7
    maxTimeBetweenGusts=7
    maxUpInterpSpeed=5
    minDownInterpSpeed=2
    minGustDuration=0.5
    minGustStrength=3
    minTimeBetweenGusts=0
    minUpInterpSpeed=2
    windTurbulenceScalarA=0.1
    windTurbulenceScalarB=1
    windAmplitude=<0 0 0>
    windFrequency=<0 0 0>
    windOffset=<0 0 0>
   }
   dBNO_Impulse {
    maxDownInterpSpeed=5
    maxGustDuration=2
    maxGustStrength=7
    maxTimeBetweenGusts=7
    maxUpInterpSpeed=5
    minDownInterpSpeed=2
    minGustDuration=0.5
    minGustStrength=3
    minTimeBetweenGusts=0
    minUpInterpSpeed=2
    windTurbulenceScalarA=0.1
    windTurbulenceScalarB=1
    windAmplitude=<0 0 0>
    windFrequency=<0 0 0>
    windOffset=<0 0 0>
   }
   dBNO_RollBwd {
    maxDownInterpSpeed=5
    maxGustDuration=2
    maxGustStrength=7
    maxTimeBetweenGusts=7
    maxUpInterpSpeed=5
    minDownInterpSpeed=2
    minGustDuration=0.5
    minGustStrength=3
    minTimeBetweenGusts=0
    minUpInterpSpeed=2
    windTurbulenceScalarA=0.1
    windTurbulenceScalarB=1
    windAmplitude=<0 0 0>
    windFrequency=<0 0 0>
    windOffset=<0 0 0>
   }
   dBNO_RollFwd {
    maxDownInterpSpeed=5
    maxGustDuration=2
    maxGustStrength=7
    maxTimeBetweenGusts=7
    maxUpInterpSpeed=5
    minDownInterpSpeed=2
    minGustDuration=0.5
    minGustStrength=3
    minTimeBetweenGusts=0
    minUpInterpSpeed=2
    windTurbulenceScalarA=0.1
    windTurbulenceScalarB=1
    windAmplitude=<0 0 0>
    windFrequency=<0 0 0>
    windOffset=<0 0 0>
   }
   dBNO_Tumbling {
    maxDownInterpSpeed=5
    maxGustDuration=2
    maxGustStrength=7
    maxTimeBetweenGusts=7
    maxUpInterpSpeed=5
    minDownInterpSpeed=2
    minGustDuration=0.5
    minGustStrength=3
    minTimeBetweenGusts=0
    minUpInterpSpeed=2
    windTurbulenceScalarA=0.1
    windTurbulenceScalarB=1
    windAmplitude=<0 0 0>
    windFrequency=<0 0 0>
    windOffset=<0 0 0>
   }
   emoteOrMelee {
    maxDownInterpSpeed=5
    maxGustDuration=2
    maxGustStrength=7
    maxTimeBetweenGusts=7
    maxUpInterpSpeed=5
    minDownInterpSpeed=2
    minGustDuration=0.5
    minGustStrength=3
    minTimeBetweenGusts=0
    minUpInterpSpeed=2
    windTurbulenceScalarA=0.1
    windTurbulenceScalarB=1
    windAmplitude=<0 0 0>
    windFrequency=<0 0 0>
    windOffset=<0 0 0>
   }
   front_End {
    maxDownInterpSpeed=5
    maxGustDuration=2
    maxGustStrength=7
    maxTimeBetweenGusts=7
    maxUpInterpSpeed=5
    minDownInterpSpeed=2
    minGustDuration=0.5
    minGustStrength=3
    minTimeBetweenGusts=0
    minUpInterpSpeed=2
    windTurbulenceScalarA=0.1
    windTurbulenceScalarB=1
    windAmplitude=<0 0 0>
    windFrequency=<0 0 0>
    windOffset=<0 0 0>
   }
   onGround_Crouch_Moving {
    maxDownInterpSpeed=5
    maxGustDuration=2
    maxGustStrength=7
    maxTimeBetweenGusts=7
    maxUpInterpSpeed=5
    minDownInterpSpeed=2
    minGustDuration=0.5
    minGustStrength=3
    minTimeBetweenGusts=0
    minUpInterpSpeed=2
    windTurbulenceScalarA=0.1
    windTurbulenceScalarB=1
    windAmplitude=<0 0 0>
    windFrequency=<0 0 0>
    windOffset=<0 0 0>
   }
   onGround_Falling {
    maxDownInterpSpeed=5
    maxGustDuration=2
    maxGustStrength=7
    maxTimeBetweenGusts=7
    maxUpInterpSpeed=5
    minDownInterpSpeed=2
    minGustDuration=0.5
    minGustStrength=3
    minTimeBetweenGusts=0
    minUpInterpSpeed=2
    windTurbulenceScalarA=0.1
    windTurbulenceScalarB=1
    windAmplitude=<0 0 0>
    windFrequency=<0 0 0>
    windOffset=<0 0 0>
   }
   onGround_Moving {
    maxDownInterpSpeed=5
    maxGustDuration=2
    maxGustStrength=7
    maxTimeBetweenGusts=7
    maxUpInterpSpeed=5
    minDownInterpSpeed=2
    minGustDuration=0.5
    minGustStrength=3
    minTimeBetweenGusts=0
    minUpInterpSpeed=2
    windTurbulenceScalarA=0.1
    windTurbulenceScalarB=1
    windAmplitude=<0 0 0>
    windFrequency=<0 0 0>
    windOffset=<0 0 0>
   }
   onGround_Standing {
    maxDownInterpSpeed=5
    maxGustDuration=2
    maxGustStrength=7
    maxTimeBetweenGusts=7
    maxUpInterpSpeed=5
    minDownInterpSpeed=2
    minGustDuration=0.5
    minGustStrength=3
    minTimeBetweenGusts=0
    minUpInterpSpeed=2
    windTurbulenceScalarA=0.1
    windTurbulenceScalarB=1
    windAmplitude=<0 0 0>
    windFrequency=<0 0 0>
    windOffset=<0 0 0>
   }
   rail_Grinding {
    maxDownInterpSpeed=5
    maxGustDuration=8
    maxGustStrength=7
    maxTimeBetweenGusts=8
    maxUpInterpSpeed=5
    minDownInterpSpeed=2
    minGustDuration=0.2
    minGustStrength=3
    minTimeBetweenGusts=1
    minUpInterpSpeed=2
    windTurbulenceScalarA=0.1
    windTurbulenceScalarB=0.8
    windAmplitude=<20 65 42>
    windFrequency=<5 6 8.6>
    windOffset=<50 -500 -400>
   }
   riding_Moving {
    maxDownInterpSpeed=5
    maxGustDuration=2
    maxGustStrength=7
    maxTimeBetweenGusts=7
    maxUpInterpSpeed=5
    minDownInterpSpeed=2
    minGustDuration=0.5
    minGustStrength=3
    minTimeBetweenGusts=0
    minUpInterpSpeed=2
    windTurbulenceScalarA=0.1
    windTurbulenceScalarB=1
    windAmplitude=<0 0 0>
    windFrequency=<0 0 0>
    windOffset=<0 0 0>
   }
   riding_Moving_Targeting {
    maxDownInterpSpeed=5
    maxGustDuration=2
    maxGustStrength=7
    maxTimeBetweenGusts=7
    maxUpInterpSpeed=5
    minDownInterpSpeed=2
    minGustDuration=0.5
    minGustStrength=3
    minTimeBetweenGusts=0
    minUpInterpSpeed=2
    windTurbulenceScalarA=0.1
    windTurbulenceScalarB=1
    windAmplitude=<0 0 0>
    windFrequency=<0 0 0>
    windOffset=<0 0 0>
   }
   riding_Standing {
    maxDownInterpSpeed=5
    maxGustDuration=2
    maxGustStrength=7
    maxTimeBetweenGusts=7
    maxUpInterpSpeed=5
    minDownInterpSpeed=2
    minGustDuration=0.5
    minGustStrength=3
    minTimeBetweenGusts=0
    minUpInterpSpeed=2
    windTurbulenceScalarA=0.1
    windTurbulenceScalarB=1
    windAmplitude=<0 0 0>
    windFrequency=<0 0 0>
    windOffset=<0 0 0>
   }
   skydive_DivingDown {
    maxDownInterpSpeed=5
    maxGustDuration=2
    maxGustStrength=7
    maxTimeBetweenGusts=7
    maxUpInterpSpeed=5
    minDownInterpSpeed=2
    minGustDuration=0.5
    minGustStrength=3
    minTimeBetweenGusts=0
    minUpInterpSpeed=2
    windTurbulenceScalarA=0.1
    windTurbulenceScalarB=1
    windAmplitude=<0 0 0>
    windFrequency=<0 0 0>
    windOffset=<0 0 0>
   }
   skydive_FloatingDown {
    maxDownInterpSpeed=5
    maxGustDuration=2
    maxGustStrength=7
    maxTimeBetweenGusts=7
    maxUpInterpSpeed=5
    minDownInterpSpeed=2
    minGustDuration=0.5
    minGustStrength=3
    minTimeBetweenGusts=0
    minUpInterpSpeed=2
    windTurbulenceScalarA=0.1
    windTurbulenceScalarB=1
    windAmplitude=<0 0 0>
    windFrequency=<0 0 0>
    windOffset=<0 0 0>
   }
   skydive_GoingUp {
    maxDownInterpSpeed=5
    maxGustDuration=2
    maxGustStrength=7
    maxTimeBetweenGusts=7
    maxUpInterpSpeed=5
    minDownInterpSpeed=2
    minGustDuration=0.5
    minGustStrength=3
    minTimeBetweenGusts=0
    minUpInterpSpeed=2
    windTurbulenceScalarA=0.1
    windTurbulenceScalarB=1
    windAmplitude=<0 0 0>
    windFrequency=<0 0 0>
    windOffset=<0 0 0>
   }
   skydive_Parachuting {
    maxDownInterpSpeed=5
    maxGustDuration=8
    maxGustStrength=1
    maxTimeBetweenGusts=8
    maxUpInterpSpeed=5
    minDownInterpSpeed=2
    minGustDuration=5
    minGustStrength=0
    minTimeBetweenGusts=5
    minUpInterpSpeed=2
    windTurbulenceScalarA=0.2
    windTurbulenceScalarB=1
    windAmplitude=<12 25 48>
    windFrequency=<8 6 2>
    windOffset=<0 -200 -75>
   }
   skydive_Surfing {
    maxDownInterpSpeed=5
    maxGustDuration=2
    maxGustStrength=7
    maxTimeBetweenGusts=7
    maxUpInterpSpeed=5
    minDownInterpSpeed=2
    minGustDuration=0.5
    minGustStrength=3
    minTimeBetweenGusts=0
    minUpInterpSpeed=2
    windTurbulenceScalarA=0.1
    windTurbulenceScalarB=1
    windAmplitude=<0 0 0>
    windFrequency=<0 0 0>
    windOffset=<0 0 0>
   }
   swimming {
    maxDownInterpSpeed=5
    maxGustDuration=2
    maxGustStrength=7
    maxTimeBetweenGusts=7
    maxUpInterpSpeed=5
    minDownInterpSpeed=2
    minGustDuration=0.5
    minGustStrength=3
    minTimeBetweenGusts=0
    minUpInterpSpeed=2
    windTurbulenceScalarA=0.1
    windTurbulenceScalarB=1
    windAmplitude=<0 0 0>
    windFrequency=<0 0 0>
    windOffset=<0 0 0>
   }
   tactical_Sprint {
    maxDownInterpSpeed=5
    maxGustDuration=2
    maxGustStrength=7
    maxTimeBetweenGusts=7
    maxUpInterpSpeed=5
    minDownInterpSpeed=2
    minGustDuration=0.5
    minGustStrength=3
    minTimeBetweenGusts=0
    minUpInterpSpeed=2
    windTurbulenceScalarA=0.1
    windTurbulenceScalarB=1
    windAmplitude=<0 0 0>
    windFrequency=<0 0 0>
    windOffset=<0 0 0>
   }
   ziplining {
    maxDownInterpSpeed=5
    maxGustDuration=2
    maxGustStrength=7
    maxTimeBetweenGusts=7
    maxUpInterpSpeed=5
    minDownInterpSpeed=2
    minGustDuration=0.5
    minGustStrength=3
    minTimeBetweenGusts=0
    minUpInterpSpeed=2
    windTurbulenceScalarA=0.1
    windTurbulenceScalarB=1
    windAmplitude=<0 0 0>
    windFrequency=<0 0 0>
    windOffset=<0 0 0>
   }
  }
 }
]
