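# Physics asset for the skydive face accessory part of the female medium Parfait character, covering front and back hair chains.
# Front strands chain from the head with tight angular limits; back strands allow wider swing and hang from base bodies constrained to spine_05 rather than the head.
# The back chains are asymmetric, with an extra left segment.
# Capsules on head, neck, upper spine and clavicles form the body, linked by free-rotation constraints.
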PhysicsAsset /BRCosmetics/Characters/Player/Female/Medium/Bodies/F_MED_Parfait/Meshes/Parts/F_MED_Parfait_Skydive_FaceAcc_Physics
bodies {
 clavicle_l [{shapeType=Capsule, length=7.6964, radius=6.3853, center=<4.0147 -0.8354 -1.7419>, rotation=<rot -77.5566 177.3915 7.7314>}]
 clavicle_r [{shapeType=Capsule, length=5.6964, radius=6.7957, center=<-4.0147 0.8354 1.7419>, rotation=<rot -77.5565 177.3915 -172.2685>}]
 hairBack_01_l [{shapeType=Capsule, length=6, radius=2.5, shapeName=hairBack_01_l_capsule, center=<3 0 0>, rotation=<rot 90 0 0>}]
 hairBack_01_r [{shapeType=Capsule, length=4, radius=2, shapeName=hairBack_01_r_capsule, center=<3 0 0>, rotation=<rot 90 0 0>}]
 hairBack_02_l [{shapeType=Capsule, length=6, radius=2.5, shapeName=hairBack_02_l_capsule, center=<3 0 0>, rotation=<rot 90 0 0>}]
 hairBack_02_r [{shapeType=Capsule, length=3.25, radius=3.25, shapeName=hairBack_02_r_capsule, center=<4.2327 0 0>, rotation=<rot 90 0 0>}]
 hairBack_03_l [{shapeType=Capsule, length=5.75, radius=2.75, shapeName=hairBack_03_l_capsule, center=<3 0 0>, rotation=<rot 90 0 0>}]
 hairBack_03_r [{shapeType=Capsule, length=3.75, radius=3.25, shapeName=hairBack_03_r_capsule, center=<4.2327 0 0>, rotation=<rot 90 0 0>}]
 hairBack_04_l [{shapeType=Capsule, length=6, radius=2.5, shapeName=hairBack_04_l_capsule, center=<3 0 0>, rotation=<rot 90 0 0>}]
 hairBack_l_base [{shapeType=Capsule, length=1, radius=1, shapeName=hairBack_l_base_capsule}]
 hairBack_r_base [{shapeType=Capsule, length=1, radius=1, shapeName=hairBack_r_base_capsule}]
 hairFront_01_l [{shapeType=Capsule, length=2, radius=1.2, center=<1.5856 0 0>, rotation=<rot 90 0 0>}]
 hairFront_01_r [{shapeType=Capsule, length=2, radius=1.2, center=<-0.9528 0 0>, rotation=<rot 90 0 0>}]
 hairFront_02_l [{shapeType=Capsule, length=2, radius=1.2, center=<1.5856 0 0>, rotation=<rot 90 0 0>}]
 hairFront_02_r [{shapeType=Capsule, length=2, radius=1.2, center=<-0.9528 0 0>, rotation=<rot 90 0 0>}]
 hairFront_03_l [{shapeType=Capsule, length=2, radius=1.2, center=<1.5856 0 0>, rotation=<rot 90 0 0>}]
 hairFront_03_r [{shapeType=Capsule, length=2, radius=1.2, center=<-0.9528 0 0>, rotation=<rot 90 0 0>}]
 head [{shapeType=Capsule, length=1, radius=1, shapeName=head_capsule}]
 neck_02 [{shapeType=Capsule, length=1, radius=1, shapeName=neck_02_capsule}]
 spine_03 [{shapeType=Capsule, length=21.005, radius=11.505, rotation=<rot 0 -90 0>}]
 spine_04 [
  {shapeType=Capsule, length=15.0796, radius=9.8916, center=<6.9796 -3.5266 0>}
  {shapeType=Capsule, length=10.0312, radius=9.2812, center=<-1.2326 1.308 0>, rotation=<rot 90 0 0>}
 ]
 spine_05 [{shapeType=Capsule, length=10.7359, radius=10.3291, center=<2.0205 0.1125 0>}]
}
constraints [
 {bone1Name=hairFront_01_l, bone2Name=head, swing1LimitDegrees=5, swing2LimitDegrees=5, twistLimitDegrees=5}
 {bone1Name=hairFront_02_l, bone2Name=hairFront_01_l, swing1LimitDegrees=5, swing2LimitDegrees=5, twistLimitDegrees=5}
 {bone1Name=hairFront_03_l, bone2Name=hairFront_02_l, swing1LimitDegrees=5, swing2LimitDegrees=5, twistLimitDegrees=5}
 {bone1Name=hairFront_01_r, bone2Name=head, swing1LimitDegrees=5, swing2LimitDegrees=5, twistLimitDegrees=5}
 {bone1Name=hairFront_02_r, bone2Name=hairFront_01_r, swing1LimitDegrees=5, swing2LimitDegrees=5, twistLimitDegrees=5}
 {bone1Name=hairFront_03_r, bone2Name=hairFront_02_r, swing1LimitDegrees=5, swing2LimitDegrees=5, twistLimitDegrees=5}
 {
  bone1Name=head
  bone2Name=neck_02
  swing1LimitDegrees=45
  swing1Motion=Free
  swing2LimitDegrees=45
  swing2Motion=Free
  twistLimitDegrees=45
  twistMotion=Free
 }
 {
  bone1Name=spine_04
  bone2Name=spine_03
  swing1LimitDegrees=45
  swing1Motion=Free
  swing2LimitDegrees=45
  swing2Motion=Free
  twistLimitDegrees=45
  twistMotion=Free
 }
 {
  bone1Name=spine_05
  bone2Name=spine_04
  swing1LimitDegrees=45
  swing1Motion=Free
  swing2LimitDegrees=45
  swing2Motion=Free
  twistLimitDegrees=45
  twistMotion=Free
 }
 {
  bone1Name=neck_02
  bone2Name=spine_05
  swing1LimitDegrees=45
  swing1Motion=Free
  swing2LimitDegrees=45
  swing2Motion=Free
  twistLimitDegrees=45
  twistMotion=Free
 }
 {
  bone1Name=hairBack_r_base
  bone2Name=spine_05
  swing1LimitDegrees=45
  swing1Motion=Free
  swing2LimitDegrees=45
  swing2Motion=Free
  twistLimitDegrees=45
  twistMotion=Free
 }
 {bone1Name=hairBack_01_r, bone2Name=hairBack_r_base, swing1LimitDegrees=15, swing2LimitDegrees=15, twistLimitDegrees=5}
 {bone1Name=hairBack_02_r, bone2Name=hairBack_01_r, swing1LimitDegrees=15, swing2LimitDegrees=15, twistLimitDegrees=5}
 {bone1Name=hairBack_03_r, bone2Name=hairBack_02_r, swing1LimitDegrees=15, swing2LimitDegrees=15, twistLimitDegrees=5}
 {
  bone1Name=hairBack_l_base
  bone2Name=spine_05
  swing1LimitDegrees=45
  swing1Motion=Free
  swing2LimitDegrees=45
  swing2Motion=Free
  twistLimitDegrees=45
  twistMotion=Free
 }
 {bone1Name=hairBack_01_l, bone2Name=hairBack_l_base, swing1LimitDegrees=15, swing2LimitDegrees=15, twistLimitDegrees=5}
 {bone1Name=hairBack_02_l, bone2Name=hairBack_01_l, swing1LimitDegrees=15, swing2LimitDegrees=15, twistLimitDegrees=5}
 {bone1Name=hairBack_03_l, bone2Name=hairBack_02_l, swing1LimitDegrees=15, swing2LimitDegrees=15, twistLimitDegrees=5}
 {bone1Name=hairBack_04_l, bone2Name=hairBack_03_l, swing1LimitDegrees=15, swing2LimitDegrees=15, twistLimitDegrees=5}
 {
  bone1Name=clavicle_l
  bone2Name=spine_05
  swing1LimitDegrees=45
  swing1Motion=Free
  swing2LimitDegrees=45
  swing2Motion=Free
  twistLimitDegrees=45
  twistMotion=Free
 }
 {
  bone1Name=clavicle_r
  bone2Name=spine_05
  swing1LimitDegrees=45
  swing1Motion=Free
  swing2LimitDegrees=45
  swing2Motion=Free
  twistLimitDegrees=45
  twistMotion=Free
 }
]
