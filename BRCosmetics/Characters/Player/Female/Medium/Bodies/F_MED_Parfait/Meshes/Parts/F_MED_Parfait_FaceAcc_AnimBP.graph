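# Animation Blueprint for Parfait's face accessory, copying the attached parent's pose and adding RigidBody physics to hair and cape.
# Bool blends select a cape-disabled branch (Stage2/3 variant or backbling), a skydive branch or the default.
# Frontend outside emotes swaps physics assets and adds bone offsets in non-skydive branches.
# Head, clavicle and crouch-targeting alphas, measured against spine_05, drive corrective hair bone transforms; gravity override feeds the physics.
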AnimBlueprint /BRCosmetics/Characters/Player/Female/Medium/Bodies/F_MED_Parfait/Meshes/Parts/F_MED_Parfait_FaceAcc_AnimBP
target_skeleton=&/F_MED_Parfait_FaceAcc_Skeleton
graphs [
 AnimGraph {nodes [
   n1 = Get head_alpha() -> (head_alpha:double)
   "Transform (Modify) Bone - Bone: hairBack_01_l"(Alpha=@n1, Translation="0,0,0", Rotation="0,-10,0", Scale="1,1,1") {
    rotationMode=BMM_Additive
    rotationSpace=BCS_BoneSpace
    translationMode=BMM_Additive
    translationSpace=BCS_BoneSpace
    alphaBoolBlend {blendOption=Linear}
    alphaScaleBiasClamp {outRange {max=0, min=1}}
    boneToModify {boneName=hairBack_01_l}
    rotation {yaw=-10}
   }
   n2 = "Transform (Modify) Bone - Bone: hair_r_raise"(
    ComponentPose=Local To Component(LocalPose=Copy Pose From Mesh() {bUseAttachedParent=true})
    Alpha=Get head_alpha()
    Translation="-0.000,-0.000,1.5"
    Rotation="0,0,0"
    Scale="1,1,1"
   ) {
    translationMode=BMM_Additive
    translationSpace=BCS_BoneSpace
    alphaBoolBlend {blendOption=Linear}
    boneToModify {boneName=hair_r_raise}
    translation {z=1.5}
   }
   n3 = RigidBody!(ComponentPose=@n2, Alpha, OverrideWorldGravity:Vector=Get GravityOverride(), ExternalForce) -> (Pose) {
    bOverrideWorldGravity=true
    overridePhysicsAsset=&/F_MED_Parfait_Skydive_FaceAcc_Physics
    simulationSpace=BaseBoneSpace
    alphaBoolBlend {blendOption=Linear}
    baseBoneRef {boneName=spine_05}
    componentAppliedLinearAccClamp=<2000 2000 2000>
    componentLinearAccScale=<0.35 0.35 0.4>
    componentLinearVelScale=<0.35 0.35 0.4>
    overrideWorldGravity {z=-980}
   }
   n4 = "Transform (Modify) Bone - Bone: hairBack_l_base"(ComponentPose=@n3, Alpha=1, Translation="0,0,0", Rotation="15,-25,-0.000", Scale="1,1,1") {
    rotationMode=BMM_Additive
    rotationSpace=BCS_BoneSpace
    translationMode=BMM_Additive
    translationSpace=BCS_BoneSpace
    alphaBoolBlend {blendOption=Linear}
    boneToModify {boneName=hairBack_l_base}
    rotation {pitch=15, yaw=-25}
   }
   n5 = "Transform (Modify) Bone - Bone: hairBack_r_base"(ComponentPose=@n4, Alpha=1, Translation="0,0,0", Rotation="-15,-45,-0.000", Scale="1,1,1") {
    rotationMode=BMM_Additive
    rotationSpace=BCS_BoneSpace
    translationMode=BMM_Additive
    translationSpace=BCS_BoneSpace
    alphaBoolBlend {blendOption=Linear}
    boneToModify {boneName=hairBack_r_base}
    rotation {pitch=-15, yaw=-45}
   }
   n6 = "Transform (Modify) Bone - Bone: hairBack_01_r"(ComponentPose=@n5, Alpha=1, Translation="0,0,0", Rotation="0,20,-0.000", Scale="1,1,1") {
    rotationMode=BMM_Additive
    rotationSpace=BCS_BoneSpace
    translationMode=BMM_Additive
    translationSpace=BCS_BoneSpace
    alphaBoolBlend {blendOption=Linear}
    boneToModify {boneName=hairBack_01_r}
    rotation {yaw=20}
   }
   n7 = Get clavicle_l_alpha() -> (clavicle_l_alpha:double)
   n8 = "Transform (Modify) Bone - Bone: hair_l_raise"(ComponentPose=@n6, Alpha=@n7, Translation="-2,0,2", Rotation="0,0,0", Scale="1,1,1") {
    rotationMode=BMM_Additive
    rotationSpace=BCS_BoneSpace
    translationMode=BMM_Additive
    translationSpace=BCS_BoneSpace
    alphaBoolBlend {blendOption=Linear}
    boneToModify {boneName=hair_l_raise}
    translation {x=-2, z=2}
   }
   n9 = "Transform (Modify) Bone - Bone: hairBack_l_base"(ComponentPose=@n8, Alpha=@n7, Translation="0,0,0", Rotation="0,-15,-0.000", Scale="1,1,1") {
    rotationMode=BMM_Additive
    rotationSpace=BCS_BoneSpace
    translationMode=BMM_Additive
    translationSpace=BCS_BoneSpace
    alphaBoolBlend {blendOption=Linear}
    boneToModify {boneName=hairBack_l_base}
    rotation {yaw=-15}
   }
   n10 = Component To Local(ComponentPose=@n9)
   n11 = "Transform (Modify) Bone - Bone: hair_r_raise"(
    ComponentPose=Local To Component(LocalPose=Copy Pose From Mesh() {bUseAttachedParent=true})
    Alpha=Get head_alpha()
    Translation="-0.000,-0.000,1.5"
    Rotation="0,0,0"
    Scale="1,1,1"
   ) {
    translationMode=BMM_Additive
    translationSpace=BCS_BoneSpace
    alphaBoolBlend {blendOption=Linear}
    boneToModify {boneName=hair_r_raise}
    translation {z=1.5}
   }
   n12 = Get crouch_head_alpha() -> (crouch_head_alpha:double)
   n13 = "Transform (Modify) Bone - Bone: hairBack_r_base"(ComponentPose=@n11, Alpha=@n12, Translation="0,0,0", Rotation="-0.000,-15,-0.000", Scale="1,1,1") {
    rotationMode=BMM_Additive
    rotationSpace=BCS_BoneSpace
    translationMode=BMM_Additive
    translationSpace=BCS_BoneSpace
    alphaBoolBlend {blendOption=Linear}
    alphaScaleBiasClamp {bInterpResult=true}
    boneToModify {boneName=hairBack_r_base}
    rotation {yaw=-15}
   }
   n14 = "Transform (Modify) Bone - Bone: hairBack_l_base"(ComponentPose=@n13, Alpha=@n12, Translation="0,0,0", Rotation="-0.000,-15,-0.000", Scale="1,1,1") {
    rotationMode=BMM_Additive
    rotationSpace=BCS_BoneSpace
    translationMode=BMM_Additive
    translationSpace=BCS_BoneSpace
    alphaBoolBlend {blendOption=Linear}
    alphaScaleBiasClamp {bInterpResult=true}
    boneToModify {boneName=hairBack_l_base}
    rotation {yaw=-15}
   }
   n15 = Get frontend_not_emote() -> (frontend_not_emote:bool)
   n16 = RigidBody!(ComponentPose=@n14, bAlphaBoolEnabled:bool=NOT Boolean(A=@n15), OverrideWorldGravity=Get GravityOverride(), ExternalForce) -> (Pose) {
    alphaInputType=Bool
    bOverrideWorldGravity=true
    overridePhysicsAsset=&/F_MED_Parfait_FaceAcc_Physics
    simulationSpace=BaseBoneSpace
    alphaBoolBlend {blendOption=Linear}
    baseBoneRef {boneName=spine_05}
    componentAppliedLinearAccClamp=<2000 2000 2000>
    componentLinearAccScale=<0.35 0.35 0.4>
    componentLinearVelScale=<0.35 0.35 0.4>
    overrideWorldGravity {z=-980}
   }
   n17 = RigidBody(ComponentPose=@n16, bAlphaBoolEnabled=@n15, OverrideWorldGravity=Get GravityOverride()) {
    alphaInputType=Bool
    bOverrideWorldGravity=true
    overridePhysicsAsset=&/F_MED_Parfait_FaceAcc_Frontend_Physics
    simulationSpace=BaseBoneSpace
    alphaBoolBlend {blendOption=Linear}
    baseBoneRef {boneName=spine_05}
    componentAppliedLinearAccClamp=<2000 2000 2000>
    componentLinearAccScale=<0.35 0.35 0.4>
    componentLinearVelScale=<0.35 0.35 0.4>
    overrideWorldGravity {z=-980}
   }
   n18 = "Transform (Modify) Bone - Bone: hairBack_01_r"(ComponentPose=@n17, Alpha=Get head_alpha(), Translation="0.000,0.000,1", Rotation="0,15,0", Scale="1,1,1") {
    rotationMode=BMM_Additive
    rotationSpace=BCS_BoneSpace
    translationMode=BMM_Additive
    alphaBoolBlend {blendOption=Linear}
    boneToModify {boneName=hairBack_01_r}
    rotation {yaw=15}
    translation {z=1}
   }
   n19 = "Transform (Modify) Bone - Bone: hairBack_l_base"(ComponentPose=@n18, Alpha=1, Translation="0,0,0", Rotation="-0.000,-5.000,-0.000", Scale="1,1,1") {
    rotationMode=BMM_Additive
    rotationSpace=BCS_BoneSpace
    translationMode=BMM_Additive
    translationSpace=BCS_BoneSpace
    alphaBoolBlend {blendOption=Linear}
    boneToModify {boneName=hairBack_l_base}
    rotation=<rot 0 -4.9999 0>
   }
   n20 = "Transform (Modify) Bone - Bone: hairBack_01_r"(
    ComponentPose=@n19
    Alpha=Get clavicle_r_alpha() -> (clavicle_r_alpha:double)
    Translation="0.000,0.000,2.25"
    Rotation="0,0,0"
    Scale="1,1,1"
   ) {
    rotationMode=BMM_Additive
    rotationSpace=BCS_BoneSpace
    translationMode=BMM_Additive
    alphaBoolBlend {blendOption=Linear}
    boneToModify {boneName=hairBack_01_r}
    translation {z=2.25}
   }
   n21 = "Transform (Modify) Bone - Bone: hair_l_raise"(ComponentPose=@n20, Alpha=Get clavicle_l_alpha(), Translation="-2,0,2", Rotation="0,0,0", Scale="1,1,1") {
    rotationMode=BMM_Additive
    rotationSpace=BCS_BoneSpace
    translationMode=BMM_Additive
    translationSpace=BCS_BoneSpace
    alphaBoolBlend {blendOption=Linear}
    boneToModify {boneName=hair_l_raise}
    translation {x=-2, z=2}
   }
   n22 = Get crouch_head_alpha()
   n23 = "Transform (Modify) Bone - Bone: hairBack_01_r"(ComponentPose=@n21, Alpha=@n22, Translation="0,0,0", Rotation="0,15,0", Scale="1,1,1") {
    rotationMode=BMM_Additive
    rotationSpace=BCS_BoneSpace
    translationMode=BMM_Additive
    alphaBoolBlend {blendOption=Linear}
    alphaScaleBiasClamp {bInterpResult=true}
    boneToModify {boneName=hairBack_01_r}
    rotation {yaw=15}
   }
   n24 = "Transform (Modify) Bone - Bone: hairBack_02_r"(ComponentPose=@n23, Alpha=@n22, Translation="0,0,0", Rotation="0,-15,0", Scale="1,1,1") {
    rotationMode=BMM_Additive
    rotationSpace=BCS_BoneSpace
    translationMode=BMM_Additive
    alphaBoolBlend {blendOption=Linear}
    alphaScaleBiasClamp {bInterpResult=true}
    boneToModify {boneName=hairBack_02_r}
    rotation {yaw=-15}
   }
   n25 = Get head_alpha()
   n26 = "Transform (Modify) Bone - Bone: hairBack_01_l"(ComponentPose=@n24, Alpha=@n25, Translation="0,0,0", Rotation="0,10,0", Scale="1,1,1") {
    rotationMode=BMM_Additive
    rotationSpace=BCS_BoneSpace
    translationMode=BMM_Additive
    translationSpace=BCS_BoneSpace
    alphaBoolBlend {blendOption=Linear}
    alphaScaleBiasClamp {bMapRange=true, outRange {max=0, min=1}}
    boneToModify {boneName=hairBack_01_l}
    rotation {yaw=10}
   }
   n27 = "Transform (Modify) Bone - Bone: hairBack_02_r"(ComponentPose=@n26, Alpha=@n25, Translation="0,0,0", Rotation="-0.000,10,-0.000", Scale="1,1,1") {
    rotationMode=BMM_Additive
    rotationSpace=BCS_BoneSpace
    translationMode=BMM_Additive
    translationSpace=BCS_BoneSpace
    alphaBoolBlend {blendOption=Linear}
    alphaScaleBiasClamp {bMapRange=true, outRange {max=0, min=1}}
    boneToModify {boneName=hairBack_02_r}
    rotation {yaw=10}
   }
   n28 = "Transform (Modify) Bone - Bone: hairBack_02_r"(ComponentPose=@n27, Alpha=@n1, Translation="0,0,0", Rotation="-0.000,-10,-0.000", Scale="1,1,1") {
    rotationMode=BMM_Additive
    rotationSpace=BCS_BoneSpace
    translationMode=BMM_Additive
    translationSpace=BCS_BoneSpace
    alphaBoolBlend {blendOption=Linear}
    alphaScaleBiasClamp {outRange {max=0, min=1}}
    boneToModify {boneName=hairBack_02_r}
    rotation {yaw=-10}
   }
   n29 = Get frontend_not_emote()
   n30 = "Transform (Modify) Bone - Bone: hairBack_02_r"!(ComponentPose=@n28, bAlphaBoolEnabled:bool=@n29, Translation="0,0,0", Rotation="-0.000,-30,-0.000", Scale="1,1,1") -> (Pose) {
    alphaInputType=Bool
    rotationMode=BMM_Additive
    rotationSpace=BCS_BoneSpace
    translationMode=BMM_Additive
    translationSpace=BCS_BoneSpace
    alphaBoolBlend {blendOption=Linear}
    alphaScaleBiasClamp {bInterpResult=true}
    boneToModify {boneName=hairBack_02_r}
    rotation {yaw=-30}
   }
   n31 = "Transform (Modify) Bone - Bone: hairBack_02_l"!(ComponentPose=@n30, bAlphaBoolEnabled:bool=@n29, Translation="0,0,0", Rotation="-0.000,-30,-0.000", Scale="1,1,1") -> (Pose) {
    alphaInputType=Bool
    rotationMode=BMM_Additive
    rotationSpace=BCS_BoneSpace
    translationMode=BMM_Additive
    translationSpace=BCS_BoneSpace
    alphaBoolBlend {blendOption=Linear}
    alphaScaleBiasClamp {bInterpResult=true}
    boneToModify {boneName=hairBack_02_l}
    rotation {yaw=-30}
   }
   n32 = Component To Local(ComponentPose=@n31)
   n33 = Get skydive() -> (skydive:bool)
   n34 = Get frontend_not_emote()
   n35 = RigidBody(
    ComponentPose=Local To Component(LocalPose=Copy Pose From Mesh() {bUseAttachedParent=true})
    bAlphaBoolEnabled=NOT Boolean(A=@n34)
    OverrideWorldGravity=Get GravityOverride()
   ) {
    alphaInputType=Bool
    bOverrideWorldGravity=true
    overridePhysicsAsset=&/F_MED_Parfait_Cape_Disabled_Physics
    simulationSpace=BaseBoneSpace
    alphaBoolBlend {blendOption=Linear}
    baseBoneRef {boneName=spine_05}
    componentAppliedLinearAccClamp=<2000 2000 2000>
    componentLinearAccScale=<0.35 0.35 0.4>
    componentLinearVelScale=<0.35 0.35 0.4>
    overrideWorldGravity {z=-980}
   }
   n36 = RigidBody(ComponentPose=@n35, bAlphaBoolEnabled=@n34, OverrideWorldGravity=Get GravityOverride()) {
    alphaInputType=Bool
    bOverrideWorldGravity=true
    overridePhysicsAsset=&/F_MED_Parfait_FaceAcc_Frontend_Cape_Off_Physics
    simulationSpace=BaseBoneSpace
    alphaBoolBlend {blendOption=Linear}
    baseBoneRef {boneName=spine_05}
    componentAppliedLinearAccClamp=<2000 2000 2000>
    componentLinearAccScale=<0.35 0.35 0.4>
    componentLinearVelScale=<0.35 0.35 0.4>
    overrideWorldGravity {z=-980}
   }
   n37 = Get frontend_not_emote()
   n38 = "Transform (Modify) Bone - Bone: hairBack_01_r"!(ComponentPose=@n36, bAlphaBoolEnabled:bool=@n37, Translation="0,0,0", Rotation="-0.000,-15,-0.000", Scale="1,1,1") -> (Pose) {
    alphaInputType=Bool
    rotationMode=BMM_Additive
    rotationSpace=BCS_BoneSpace
    translationMode=BMM_Additive
    translationSpace=BCS_BoneSpace
    alphaBoolBlend {blendOption=Linear}
    alphaScaleBiasClamp {bInterpResult=true}
    boneToModify {boneName=hairBack_01_r}
    rotation {yaw=-15}
   }
   n39 = "Transform (Modify) Bone - Bone: hairBack_01_l"!(ComponentPose=@n38, bAlphaBoolEnabled:bool=@n37, Translation="0,0,0", Rotation="-0.000,-15,-0.000", Scale="1,1,1") -> (Pose) {
    alphaInputType=Bool
    rotationMode=BMM_Additive
    rotationSpace=BCS_BoneSpace
    translationMode=BMM_Additive
    translationSpace=BCS_BoneSpace
    alphaBoolBlend {blendOption=Linear}
    alphaScaleBiasClamp {bInterpResult=true}
    boneToModify {boneName=hairBack_01_l}
    rotation {yaw=-15}
   }
   Output Pose(Result=Blend Poses by bool(
     BlendPose_0=Component To Local(ComponentPose=@n39)
     BlendPose_1=Blend Poses by bool(BlendPose_0=@n10, BlendPose_1=@n32, BlendTime_0=0.5, BlendTime_1=0.5, bActiveValue=@n33) {blendType=Linear}
     BlendTime_0=0.5
     BlendTime_1=0.5
     bActiveValue=Get cape_disabled() -> (cape_disabled:bool)
    ) {blendType=Linear}) {layerGroup=None}
  ]}
 EventGraph {nodes [
   Event BlueprintUpdateAnimation()
   > n1 = Sequence!(execute) -> (then_0, then_1, then_2:Exec)
   > Set skydive(skydive:bool=OR Boolean!(A=Get bIsSkyDiving(), B=Get bIsSkydiveFloating(), C:bool=Get bIsSkydiveDiveMode()) -> (ReturnValue)) -> (Output_Get:bool)
   > n2 = Set backbling_equipped(backbling_equipped:bool=CheckForBackbling()) -> (Output_Get:bool)
   n3 = Make Array!("[0]"=(TagName="Cosmetics.Variant.Property.Stage2"), "[1]":Gameplay Tag Structure=(TagName="Cosmetics.Variant.Property.Stage3")) -> (Array)
   > Set cape_disabled(cape_disabled:bool=OR Boolean!(A=CharacterItemDefinitionHasAnyMetaTag(Tags=@n3), B=@n2) -> (ReturnValue)) -> (Output_Get:bool)
   n4 = Get bIsPlayingEmote()
   > Set frontend_not_emote(frontend_not_emote:bool=AND Boolean(A=NOT Boolean(A=@n4), B=Get bIsInFrontEnd())) -> (Output_Get:bool)
   n5 = Gravity Override State Function DEPRECATED(
    execute=@n1.then_1
    CharacterPartAnimInstance=Self-Reference()
    GameTimeInSeconds=GetGameTimeInSeconds()
    GravityOverrideParams=Get Gravity Override Params()
   )
   > Set GravityOverride(GravityOverride=@n5)
   n6 = GetOwnerSkeletalMeshComponent()
   n7 = BreakRotator(InRot=BreakTransform(InTransform=GetDeltaTransformFromRefPose(self=@n6, BoneName=head, BaseName=spine_05)).Rotation)
   Set head_alpha(execute=@n1.then_2, head_alpha:double=MapRangeClamped(Value=@n7.Yaw, InRangeA=0, InRangeB=-30)) -> (Output_Get:double)
   n8 = BreakTransform(InTransform=GetDeltaTransformFromRefPose(self=@n6, BoneName=clavicle_r, BaseName=spine_05))
   > Set clavicle_r_alpha(clavicle_r_alpha:double=MapRangeClamped(Value=BreakRotator(InRot=@n8.Rotation).Pitch, InRangeA=-3, InRangeB=15)) -> (Output_Get:double)
   n9 = BreakTransform(InTransform=GetDeltaTransformFromRefPose(self=@n6, BoneName=clavicle_l, BaseName=spine_05))
   > Set clavicle_l_alpha(clavicle_l_alpha:double=MapRangeClamped(Value=BreakRotator(InRot=@n9.Rotation).Pitch, InRangeA=-3, InRangeB=15)) -> (Output_Get:double)
   > n10 = Branch(Condition=AND Boolean(A=Get bIsTargeting(), B=Get bIsCrouching()))
   > Set crouch_head_alpha(crouch_head_alpha:double=MapRangeClamped(Value=@n7.Yaw, InRangeA=0, InRangeB=-30)) -> (Output_Get:double)
   Set crouch_head_alpha(execute=@n10.else, crouch_head_alpha=0)
   CharacterItemDefinitionHasAnyMetaTag(Tags=Make Array("[0]"=(TagName="Cosmetics.Variant.Property.Stage2"), "[1]"=(TagName="Cosmetics.Variant.Property.Stage3")))
  ]}
]
variables [
 backbling_equipped {type=boolean}
 cape_disabled {type=boolean}
 clavicle_l_alpha {type=number}
 clavicle_r_alpha {type=number}
 crouch_head_alpha {type=number}
 frontend_not_emote {type=boolean}
 Gravity Override Params {type=GravityOverrideParamsStruct}
 GravityOverride {type=Vector}
 head_alpha {type=number}
 skydive {type=boolean}
]
